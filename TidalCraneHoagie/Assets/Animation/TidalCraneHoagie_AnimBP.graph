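# Animation Blueprint for the TidalCraneHoagie twin-rotor vehicle, animating its rotor hinges and rotor spin.
# Stacks additives on a reference pose: rotor blendspaces, whole-body and per hinge, then thrust noise and per-hinge adjust noise at differing left/right play rates.
# Transform Bone nodes then add MainRotorRotation to both rotors in bone space.
# The event graph is empty and the Blueprint declares no variables of its own.
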AnimBlueprint /TidalCraneHoagie/Assets/Animation/TidalCraneHoagie_AnimBP
target_skeleton=/TidalCraneHoagie/Assets/Meshes/Chopper_TwinRotor_Veh_Skeleton
graphs [
 AnimGraph {nodes [
   n1 = Apply Additive(
    Base=Apply Additive(
     Base=Sequence Player 'TidalCraneHoagie_Ref_Pose'() {bLoopAnimation=false, sequence=&/TidalCraneHoagie_Ref_Pose}
     Additive=Blendspace Player 'TidalCraneHoagie_Rotor_BS'(X=0, Y=0) {blendSpace=&/TidalCraneHoagie_Rotor_BS}
    ) {alphaBoolBlend {blendOption=Linear}}
    Additive=Layered blend per bone(
     BasePose=Additive Identity Pose()
     BlendPoses_0=Blendspace Player 'TidalCraneHoagie_Rotor_BS'(X=0, Y=0) {blendSpace=&TidalCraneHoagie_Rotor_BS}
    ) {layerSetup [{branchFilters [{boneName=rotor_hinge_l}]}]}
   ) {alphaBoolBlend {blendOption=Linear}}
   n2 = Layered blend per bone(
    BasePose=Additive Identity Pose()
    BlendPoses_0=Blendspace Player 'TidalCraneHoagie_Rotor_BS'(X=0, Y=0) {blendSpace=&TidalCraneHoagie_Rotor_BS}
   ) {layerSetup [{branchFilters [{boneName=rotor_hinge_r}]}]}
   n3 = Apply Additive(
    Base=Apply Additive(
     Base=Apply Additive(Base=@n1, Additive=@n2) {alphaBoolBlend {blendOption=Linear}}
     Additive="Sequence Player (additive) 'TidalCraneHoagie_Noise_Thrust'"!(PlayRate:float=1) -> (Pose) {sequence=&/TidalCraneHoagie_Noise_Thrust}
    ) {alphaBoolBlend {blendOption=Linear}}
    Additive=Layered blend per bone(
     BasePose=Additive Identity Pose()
     BlendPoses_0="Sequence Player (additive) 'TidalCraneHoagie_Noise_Adjust'"!(PlayRate:float=0.4) -> (Pose) {sequence=&/TidalCraneHoagie_Noise_Adjust}
    ) {layerSetup [{branchFilters [{boneName=rotor_hinge_l}]}]}
   ) {alphaBoolBlend {blendOption=Linear}}
   n4 = Layered blend per bone(
    BasePose=Additive Identity Pose()
    BlendPoses_0="Sequence Player (additive) 'TidalCraneHoagie_Noise_Adjust'"(PlayRate=0.5) {sequence=&TidalCraneHoagie_Noise_Adjust}
   ) {layerSetup [{branchFilters [{boneName=rotor_hinge_r}]}]}
   n5 = Get MainRotorRotation() -> (MainRotorRotation:Rotator)
   n6 = "Transform (Modify) Bone - Bone: rotor_l"!(
    ComponentPose=Local To Component(LocalPose=Apply Additive(Base=@n3, Additive=@n4) {alphaBoolBlend {blendOption=Linear}})
    Alpha=1
    Rotation=@n5
   ) -> (Pose) {rotationMode=BMM_Additive, rotationSpace=BCS_BoneSpace, alphaBoolBlend {blendOption=Linear}, boneToModify {boneName=rotor_l}}
   n7 = "Transform (Modify) Bone - Bone: rotor_r"!(ComponentPose=@n6, Alpha=1, Rotation=@n5) -> (Pose) {rotationMode=BMM_Additive, rotationSpace=BCS_BoneSpace, alphaBoolBlend {blendOption=Linear}, boneToModify {boneName=rotor_r}}
   Output Pose(Result=Component To Local(ComponentPose=@n7))
  ]}
 EventGraph {links [], nodes []}
]
variables []
