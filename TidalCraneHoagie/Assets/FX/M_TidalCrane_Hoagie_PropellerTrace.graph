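# Translucent, two-sided, Default Lit Niagara sprite material for the propeller trace effect on the TidalCrane Hoagie.
# Particle color RGB feeds the MF_FX_EmissiveAdjustments function for the glow.
# The red channel of a propeller trace texture, reused from a glider's FX assets, multiplies with particle alpha as a mask, so the particle system drives tint and fade.
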Material /TidalCraneHoagie/Assets/FX/M_TidalCrane_Hoagie_PropellerTrace
graph {
 properties {bUsedWithNiagaraSprites=true, blendMode=BLEND_Translucent, pixelDepthOffsetMode=PDOM_Legacy, twoSided=true, usedShadingModels=Default Lit}
 nodes [
  MaterialFunctionCall(Emissive=ParticleColor().RGB) {materialFunction=&MF_FX_EmissiveAdjustments}
  Multiply(
   A=TextureSample() {samplerType=SAMPLERTYPE_LinearColor, texture=/Game/Gadgets/Assets/VinderTech_GliderChute/Glider_IndieBucket/FX/T_FX_Propeller_Trace}.R
   B=ParticleColor().A
  )
 ]
}
