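# Niagara system for the damage-state effects on the Hoagie, a rotor vehicle, layering smoke, embers, flares, flames and blinking critical lights as damage progresses.
# User DamageScale is compared against staged thresholds per part (engine, rotor, tail, booster, forward) to switch spawning, while isCritical triggers the lights.
# Shared emitters serve multiple parts: spawn groups select part info and skeletal mesh sockets for placement.
# isMobile sets spawn probability for some parts.
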NiagaraSystem /TidalCraneHoagie/Assets/FX/NS_TidalCraneHoagie_DamageState_Optimized
dependencies {
 usedDataInterfaces [
  &NiagaraDataInterfaceCurve
  &NiagaraDataInterfaceArrayInt32
  &NiagaraDataInterfaceColorCurve
  &NiagaraDataInterfaceVectorCurve
 ]
 usedDynamicInputs [
  &MakeFloatFromInt
  &Lerp_Vector
  /Hoagie/FX/DynamicInputs/FN_BuildHoagieFXPartInfo
 ]
 usedModules [
  &SystemState
  /Niagara/Modules/Spawn/Location/CylinderLocation_System
  /Niagara/Modules/Spawn/Location/SphereLocation_System
  &self:SystemSpawnScript.SystemScriptSource.SystemScriptGraph.NiagaraNodeAssignment_0.SetVariables_#1
 ]
}
emitters {
 Critical_Lights {
  emitterUpdateScript [
   &EmitterState(Life Cycle Mode=::NewEnumerator1, Loop Duration=0.5, UseLoopDelay=false)
   &SpawnBurst_Instantaneous(
    Spawn Count=&MakeCustomIntFromBool(Bool=$User.isCritical, True INT=1, False INT=0)
    Spawn Group=6
    Use Spawn Probability=true
   ) {module=SpawnBurst_Instantaneous003}
   &SpawnBurst_Instantaneous(
    Spawn Count=&MakeCustomIntFromBool(Bool=$User.isCritical, True INT=1, False INT=0)
    Spawn Group=7
    Use Spawn Probability=true
   ) {module=SpawnBurst_Instantaneous004}
  ]
  particleSpawnScript [
   &InitializeParticle(
    Color=<rgba 1 0 0 1>
    Sprite Size Mode=::NewEnumerator3
    Uniform Sprite Size=&MaskFloatBySpawnGroup(Masked Float=80, Passthrough Float=300)
    UsePositionOffset=false
    Write Material Random=true
   ) {module=InitializeParticle001}
   &SampleSkeletalMeshSkeleton(
    Mesh:NiagaraDataInterfaceSkeletalMesh=$User.SkeletalMesh
    Spawn Mode:ENiagaraSkeletalSpawning=::NewEnumerator1
    Specific Socket Index:NiagaraInt32=&SpawnGroup()
   )
   &SkeletalMeshLocation(
    Bone / Socket Index=&SpawnGroup()
    MaskSpawnGroup=false
    Mesh Sampling Type=::NewEnumerator5
    Skeletal Mesh=$User.SkeletalMesh
    Socket Sampling Mode=::NewEnumerator5
   )
   &CameraOffset(Camera Offset Amount=-20)
  ]
  particleUpdateScript [
   &ParticleState()
   &SolveForcesAndVelocity(Clamp Velocity=false, Limit Acceleration=false)
   &DynamicMaterialParameters(
    Index 0 Param 1=1
    Module.Param0WriteEnabled=true
    Module.Param1WriteEnabled=true
    Module.Param2WriteEnabled=false
    Module.Param3WriteEnabled=false
   )
   &ScaleColor(
    Scale Alpha=&FloatFromCurve(FloatCurve=DataInterface() {propertyValues {Curve {keys [<at 0 0 RCIM_Cubic>,<at 0.5 1 RCIM_Cubic>,<at 1 0 RCIM_Cubic>]}}})
    Scale RGB=<5 5 5>
    ScaleA=true
    ScaleRGB=true
   )
   &ScaleSpriteSize(Scale Factor=&Vector2DFromFloat(Value=&Lerp_Float(
      A=&FloatFromCurve(FloatCurve=DataInterface() {propertyValues {Curve {keys [<at 0 1 RCIM_Cubic>,<at 0.5 0 RCIM_Cubic>,<at 1 1 RCIM_Cubic>]}}})
      B=&FloatFromCurve(
       FloatCurve=DataInterface() {propertyValues {Curve {keys [<at 0 1 RCIM_Cubic>,<at 0.25 0 RCIM_Cubic>,<at 0.5 1 RCIM_Cubic>,<at 0.75 0 RCIM_Cubic>,<at 1 1 RCIM_Cubic>]}}}
       CurveIndex=&Frac_Float(Float=$Engine.System.Age)
      )
      Alpha=&MaskFloatBySpawnGroup(Masked Float=0, Passthrough Float=1)
     )))
   &FN_Light_Attributes(
    Color=$Particles.Color
    Create Color Attribute=true
    Create Exponent Attribute=true
    Create Position Attribute=true
    Create Radius Attribute=true
    Exponent=1
    Position=&AddVectorToPosition(
     Position=$Particles.Position
     Vector=&MaskVectorBySpawnGroup(Masked Vector=<0 0 0>, Passthrough Vector=<1 0 0>)
    )
    Radius=&MaskFloatBySpawnGroup(Masked Float=0, Passthrough Float=200)
    Scale Color=&FloatFromCurve(
     FloatCurve=DataInterface() {propertyValues {Curve {keys [<at 0 1 RCIM_Cubic>,<at 0.25 0 RCIM_Cubic>,<at 0.5 1 RCIM_Cubic>,<at 0.75 0 RCIM_Cubic>,<at 1 1 RCIM_Cubic>]}}}
     CurveIndex=&Frac_Float(Float=$Engine.System.Age)
    )
   )
  ]
  rendererClasses [&NiagaraSpriteRendererProperties, &NiagaraLightRendererProperties]
  renderers [
   {rendererClass=&NiagaraSpriteRendererProperties}
   {rendererClass=&NiagaraLightRendererProperties, rendererIndex=1}
  ]
 }
 Generator_Flare_Optimized {
  emitterUpdateScript [
   &EmitterState(Life Cycle Mode=::NewEnumerator1, Loop Duration=1, UseLoopDelay=false)
   &SpawnBurst_Instantaneous(
    Spawn Count=&MakeCustomIntFromBool(
     Bool=/Niagara/DynamicInputs/Bool/SetBoolByFloatComparison(A:NiagaraFloat=$User.DamageScale, B:NiagaraFloat=$User.Threshold.Engine.Stage2, ComparisonType:ENiagaraBooleanLogicOps=::NewEnumerator2)
     True INT=&MakeCustomIntFromBool(
      Bool=&SetBoolByFloatComparison(
       A:NiagaraFloat=$User.DamageScale
       B:NiagaraFloat=$User.Threshold.Engine.Stage3Critical
       ComparisonType:ENiagaraBooleanLogicOps=::NewEnumerator2
      )
      True INT=&UniformRangedInt(Minimum=0, Maximum=2)
      False INT=&UniformRangedInt(Minimum=0, Maximum=1)
     )
     False INT=0
    )
    Use Spawn Probability=false
   )
   &SpawnBurst_Instantaneous(
    Spawn Count=&MakeCustomIntFromBool(
     Bool=&SetBoolByFloatComparison(A:NiagaraFloat=$User.DamageScale, B:NiagaraFloat=$User.Threshold.Engine.Stage2, ComparisonType:ENiagaraBooleanLogicOps=::NewEnumerator2)
     True INT=&MakeCustomIntFromBool(
      Bool=&SetBoolByFloatComparison(
       A:NiagaraFloat=$User.DamageScale
       B:NiagaraFloat=$User.Threshold.Engine.Stage3Critical
       ComparisonType:ENiagaraBooleanLogicOps=::NewEnumerator2
      )
      True INT=&UniformRangedInt(Minimum=0, Maximum=2)
      False INT=&UniformRangedInt(Minimum=0, Maximum=1)
     )
     False INT=0
    )
    Spawn Group=1
    Spawn Time=0.88
    Use Spawn Probability=false
   ) {module=SpawnBurst_Instantaneous001}
   &SpawnBurst_Instantaneous(
    Spawn Count=&MakeCustomIntFromBool(
     Bool=&SetBoolByFloatComparison(A:NiagaraFloat=$User.DamageScale, B:NiagaraFloat=$User.Threshold.Engine.Stage2, ComparisonType:ENiagaraBooleanLogicOps=::NewEnumerator2)
     True INT=&MakeCustomIntFromBool(
      Bool=&SetBoolByFloatComparison(
       A:NiagaraFloat=$User.DamageScale
       B:NiagaraFloat=$User.Threshold.Engine.Stage3Critical
       ComparisonType:ENiagaraBooleanLogicOps=::NewEnumerator2
      )
      True INT=&UniformRangedInt(Minimum=0, Maximum=2)
      False INT=&UniformRangedInt(Minimum=0, Maximum=1)
     )
     False INT=0
    )
    Spawn Group=1
    Spawn Time=1.4
    Use Spawn Probability=false
   ) {module=SpawnBurst_Instantaneous002}
   &SpawnBurst_Instantaneous(
    Spawn Count=&MakeCustomIntFromBool(
     Bool=&SetBoolByFloatComparison(A:NiagaraFloat=$User.DamageScale, B:NiagaraFloat=$User.Threshold.Engine.Stage2, ComparisonType:ENiagaraBooleanLogicOps=::NewEnumerator2)
     True INT=&MakeCustomIntFromBool(
      Bool=&SetBoolByFloatComparison(
       A:NiagaraFloat=$User.DamageScale
       B:NiagaraFloat=$User.Threshold.Engine.Stage3Critical
       ComparisonType:ENiagaraBooleanLogicOps=::NewEnumerator2
      )
      True INT=&UniformRangedInt(Minimum=0, Maximum=2)
      False INT=&UniformRangedInt(Minimum=0, Maximum=1)
     )
     False INT=0
    )
    Spawn Group=1
    Spawn Time=0.88
    Use Spawn Probability=false
   ) {module=SpawnBurst_Instantaneous003}
   &SpawnBurst_Instantaneous(
    Spawn Count=&MakeCustomIntFromBool(
     Bool=&SetBoolByFloatComparison(A:NiagaraFloat=$User.DamageScale, B:NiagaraFloat=$User.Threshold.Engine.Stage2, ComparisonType:ENiagaraBooleanLogicOps=::NewEnumerator2)
     True INT=&MakeCustomIntFromBool(
      Bool=&SetBoolByFloatComparison(
       A:NiagaraFloat=$User.DamageScale
       B:NiagaraFloat=$User.Threshold.Engine.Stage3Critical
       ComparisonType:ENiagaraBooleanLogicOps=::NewEnumerator2
      )
      True INT=&UniformRangedInt(Minimum=0, Maximum=2)
      False INT=&UniformRangedInt(Minimum=0, Maximum=1)
     )
     False INT=0
    )
    Spawn Group=1
    Spawn Time=1.4
    Use Spawn Probability=false
   ) {module=SpawnBurst_Instantaneous004}
   &SpawnBurst_Instantaneous(
    Spawn Count=&MakeCustomIntFromBool(
     Bool=&SetBoolByFloatComparison(A:NiagaraFloat=$User.DamageScale, B:NiagaraFloat=$User.Threshold.Engine.Stage2, ComparisonType:ENiagaraBooleanLogicOps=::NewEnumerator2)
     True INT=&MakeCustomIntFromBool(
      Bool=&SetBoolByFloatComparison(
       A:NiagaraFloat=$User.DamageScale
       B:NiagaraFloat=$User.Threshold.Engine.Stage3Critical
       ComparisonType:ENiagaraBooleanLogicOps=::NewEnumerator2
      )
      True INT=&UniformRangedInt(Minimum=0, Maximum=2)
      False INT=&UniformRangedInt(Minimum=0, Maximum=1)
     )
     False INT=0
    )
    Spawn Group=1
    Use Spawn Probability=false
   ) {module=SpawnBurst_Instantaneous005}
   &SpawnBurst_Instantaneous(
    Spawn Count=&MakeCustomIntFromBool(
     Bool=&SetBoolByFloatComparison(A:NiagaraFloat=$User.DamageScale, B:NiagaraFloat=$User.Threshold.Forward.Stage2, ComparisonType:ENiagaraBooleanLogicOps=::NewEnumerator2)
     True INT=&MakeCustomIntFromBool(
      Bool=&SetBoolByFloatComparison(
       A:NiagaraFloat=$User.DamageScale
       B:NiagaraFloat=$User.Threshold.Forward.Stage3Critical
       ComparisonType:ENiagaraBooleanLogicOps=::NewEnumerator2
      )
      True INT=&UniformRangedInt(Minimum=0, Maximum=2)
      False INT=&UniformRangedInt(Minimum=0, Maximum=1)
     )
     False INT=0
    )
    Spawn Group=5
    Use Spawn Probability=false
   ) {module=SpawnBurst_Instantaneous006}
   &SpawnBurst_Instantaneous(
    Spawn Count=&MakeCustomIntFromBool(
     Bool=&SetBoolByFloatComparison(A:NiagaraFloat=$User.DamageScale, B:NiagaraFloat=$User.Threshold.Forward.Stage2, ComparisonType:ENiagaraBooleanLogicOps=::NewEnumerator2)
     True INT=&MakeCustomIntFromBool(
      Bool=&SetBoolByFloatComparison(
       A:NiagaraFloat=$User.DamageScale
       B:NiagaraFloat=$User.Threshold.Forward.Stage3Critical
       ComparisonType:ENiagaraBooleanLogicOps=::NewEnumerator2
      )
      True INT=&UniformRangedInt(Minimum=0, Maximum=2)
      False INT=&UniformRangedInt(Minimum=0, Maximum=1)
     )
     False INT=0
    )
    Spawn Group=5
    Spawn Time=0.88
    Use Spawn Probability=false
   ) {module=SpawnBurst_Instantaneous007}
   &SpawnBurst_Instantaneous(
    Spawn Count=&MakeCustomIntFromBool(
     Bool=&SetBoolByFloatComparison(A:NiagaraFloat=$User.DamageScale, B:NiagaraFloat=$User.Threshold.Forward.Stage2, ComparisonType:ENiagaraBooleanLogicOps=::NewEnumerator2)
     True INT=&MakeCustomIntFromBool(
      Bool=&SetBoolByFloatComparison(
       A:NiagaraFloat=$User.DamageScale
       B:NiagaraFloat=$User.Threshold.Forward.Stage3Critical
       ComparisonType:ENiagaraBooleanLogicOps=::NewEnumerator2
      )
      True INT=&UniformRangedInt(Minimum=0, Maximum=2)
      False INT=&UniformRangedInt(Minimum=0, Maximum=1)
     )
     False INT=0
    )
    Spawn Group=5
    Spawn Time=1.4
    Use Spawn Probability=false
   ) {module=SpawnBurst_Instantaneous008}
   &SpawnBurst_Instantaneous(
    Spawn Count=&MakeCustomIntFromBool(
     Bool=&SetBoolByFloatComparison(A:NiagaraFloat=$User.DamageScale, B:NiagaraFloat=$User.Threshold.Rotor.Stage2, ComparisonType:ENiagaraBooleanLogicOps=::NewEnumerator2)
     True INT=&MakeCustomIntFromBool(
      Bool=&SetBoolByFloatComparison(
       A:NiagaraFloat=$User.DamageScale
       B:NiagaraFloat=$User.Threshold.Rotor.Stage3Critical
       ComparisonType:ENiagaraBooleanLogicOps=::NewEnumerator2
      )
      True INT=&UniformRangedInt(Minimum=0, Maximum=2)
      False INT=&UniformRangedInt(Minimum=0, Maximum=1)
     )
     False INT=0
    )
    Spawn Group=2
    Use Spawn Probability=false
   ) {module=SpawnBurst_Instantaneous009}
   &SpawnBurst_Instantaneous(
    Spawn Count=&MakeCustomIntFromBool(
     Bool=&SetBoolByFloatComparison(A:NiagaraFloat=$User.DamageScale, B:NiagaraFloat=$User.Threshold.Rotor.Stage2, ComparisonType:ENiagaraBooleanLogicOps=::NewEnumerator2)
     True INT=&MakeCustomIntFromBool(
      Bool=&SetBoolByFloatComparison(
       A:NiagaraFloat=$User.DamageScale
       B:NiagaraFloat=$User.Threshold.Rotor.Stage3Critical
       ComparisonType:ENiagaraBooleanLogicOps=::NewEnumerator2
      )
      True INT=&UniformRangedInt(Minimum=0, Maximum=2)
      False INT=&UniformRangedInt(Minimum=0, Maximum=1)
     )
     False INT=0
    )
    Spawn Group=2
    Spawn Time=0.88
    Use Spawn Probability=false
   ) {module=SpawnBurst_Instantaneous010}
   &SpawnBurst_Instantaneous(
    Spawn Count=&MakeCustomIntFromBool(
     Bool=&SetBoolByFloatComparison(A:NiagaraFloat=$User.DamageScale, B:NiagaraFloat=$User.Threshold.Rotor.Stage2, ComparisonType:ENiagaraBooleanLogicOps=::NewEnumerator2)
     True INT=&MakeCustomIntFromBool(
      Bool=&SetBoolByFloatComparison(
       A:NiagaraFloat=$User.DamageScale
       B:NiagaraFloat=$User.Threshold.Rotor.Stage3Critical
       ComparisonType:ENiagaraBooleanLogicOps=::NewEnumerator2
      )
      True INT=&UniformRangedInt(Minimum=0, Maximum=2)
      False INT=&UniformRangedInt(Minimum=0, Maximum=1)
     )
     False INT=0
    )
    Spawn Group=2
    Spawn Time=1.4
    Use Spawn Probability=false
   ) {module=SpawnBurst_Instantaneous011}
   &SpawnBurst_Instantaneous(
    Spawn Count=&MakeCustomIntFromBool(
     Bool=&SetBoolByFloatComparison(A:NiagaraFloat=$User.DamageScale, B:NiagaraFloat=$User.Threshold.Tail.Stage2, ComparisonType:ENiagaraBooleanLogicOps=::NewEnumerator2)
     True INT=&MakeCustomIntFromBool(
      Bool=&SetBoolByFloatComparison(
       A:NiagaraFloat=$User.DamageScale
       B:NiagaraFloat=$User.Threshold.Tail.Stage3Critical
       ComparisonType:ENiagaraBooleanLogicOps=::NewEnumerator2
      )
      True INT=&UniformRangedInt(Minimum=0, Maximum=2)
      False INT=&UniformRangedInt(Minimum=0, Maximum=1)
     )
     False INT=0
    )
    Spawn Group=3
    Use Spawn Probability=false
   ) {module=SpawnBurst_Instantaneous012}
   &SpawnBurst_Instantaneous(
    Spawn Count=&MakeCustomIntFromBool(
     Bool=&SetBoolByFloatComparison(A:NiagaraFloat=$User.DamageScale, B:NiagaraFloat=$User.Threshold.Tail.Stage2, ComparisonType:ENiagaraBooleanLogicOps=::NewEnumerator2)
     True INT=&MakeCustomIntFromBool(
      Bool=&SetBoolByFloatComparison(
       A:NiagaraFloat=$User.DamageScale
       B:NiagaraFloat=$User.Threshold.Tail.Stage3Critical
       ComparisonType:ENiagaraBooleanLogicOps=::NewEnumerator2
      )
      True INT=&UniformRangedInt(Minimum=0, Maximum=2)
      False INT=&UniformRangedInt(Minimum=0, Maximum=1)
     )
     False INT=0
    )
    Spawn Group=3
    Spawn Time=0.88
    Use Spawn Probability=false
   ) {module=SpawnBurst_Instantaneous013}
   &SpawnBurst_Instantaneous(
    Spawn Count=&MakeCustomIntFromBool(
     Bool=&SetBoolByFloatComparison(A:NiagaraFloat=$User.DamageScale, B:NiagaraFloat=$User.Threshold.Tail.Stage2, ComparisonType:ENiagaraBooleanLogicOps=::NewEnumerator2)
     True INT=&MakeCustomIntFromBool(
      Bool=&SetBoolByFloatComparison(
       A:NiagaraFloat=$User.DamageScale
       B:NiagaraFloat=$User.Threshold.Tail.Stage3Critical
       ComparisonType:ENiagaraBooleanLogicOps=::NewEnumerator2
      )
      True INT=&UniformRangedInt(Minimum=0, Maximum=2)
      False INT=&UniformRangedInt(Minimum=0, Maximum=1)
     )
     False INT=0
    )
    Spawn Group=3
    Spawn Time=1.4
    Use Spawn Probability=false
   ) {module=SpawnBurst_Instantaneous014}
   &SpawnBurst_Instantaneous(
    Spawn Count=&MakeCustomIntFromBool(
     Bool=&SetBoolByFloatComparison(A:NiagaraFloat=$User.DamageScale, B:NiagaraFloat=$User.Threshold.Booster.Stage2, ComparisonType:ENiagaraBooleanLogicOps=::NewEnumerator2)
     True INT=&MakeCustomIntFromBool(
      Bool=&SetBoolByFloatComparison(
       A:NiagaraFloat=$User.DamageScale
       B:NiagaraFloat=$User.Threshold.Booster.Stage3Critical
       ComparisonType:ENiagaraBooleanLogicOps=::NewEnumerator2
      )
      True INT=&UniformRangedInt(Minimum=0, Maximum=2)
      False INT=&UniformRangedInt(Minimum=0, Maximum=1)
     )
     False INT=0
    )
    Spawn Group:NiagaraInt32=&RandomRangeInt(Minimum=0, Maximum=1)
    Use Spawn Probability=false
   ) {module=SpawnBurst_Instantaneous015}
   &SpawnBurst_Instantaneous(
    Spawn Count=&MakeCustomIntFromBool(
     Bool=&SetBoolByFloatComparison(A:NiagaraFloat=$User.DamageScale, B:NiagaraFloat=$User.Threshold.Booster.Stage2, ComparisonType:ENiagaraBooleanLogicOps=::NewEnumerator2)
     True INT=&MakeCustomIntFromBool(
      Bool=&SetBoolByFloatComparison(
       A:NiagaraFloat=$User.DamageScale
       B:NiagaraFloat=$User.Threshold.Booster.Stage3Critical
       ComparisonType:ENiagaraBooleanLogicOps=::NewEnumerator2
      )
      True INT=&UniformRangedInt(Minimum=0, Maximum=2)
      False INT=&UniformRangedInt(Minimum=0, Maximum=1)
     )
     False INT=0
    )
    Spawn Group:NiagaraInt32=&RandomRangeInt(Minimum=0, Maximum=1)
    Spawn Time=0.88
    Use Spawn Probability=false
   ) {module=SpawnBurst_Instantaneous016}
   &SpawnBurst_Instantaneous(
    Spawn Count=&MakeCustomIntFromBool(
     Bool=&SetBoolByFloatComparison(A:NiagaraFloat=$User.DamageScale, B:NiagaraFloat=$User.Threshold.Booster.Stage2, ComparisonType:ENiagaraBooleanLogicOps=::NewEnumerator2)
     True INT=&MakeCustomIntFromBool(
      Bool=&SetBoolByFloatComparison(
       A:NiagaraFloat=$User.DamageScale
       B:NiagaraFloat=$User.Threshold.Booster.Stage3Critical
       ComparisonType:ENiagaraBooleanLogicOps=::NewEnumerator2
      )
      True INT=&UniformRangedInt(Minimum=0, Maximum=2)
      False INT=&UniformRangedInt(Minimum=0, Maximum=1)
     )
     False INT=0
    )
    Spawn Group:NiagaraInt32=&RandomRangeInt(Minimum=0, Maximum=1)
    Spawn Time=1.4
    Use Spawn Probability=false
   ) {module=SpawnBurst_Instantaneous017}
  ]
  particleSpawnScript [
   /SetVariables(Particles.SpawnGroup=&SpawnGroup()) {isSetParameters=true}
   /Hoagie/FX/DynamicInputs/FN_SelectHoagiePartInfo(
    BoostPartInfo:HoagieFXPartInfo=$System.Booster
    EnginePartInfo_L:HoagieFXPartInfo=$System.EnginePartInfo
    EnginePartInfo_R:HoagieFXPartInfo=$System.EngineR
    ForwardPartInfo:HoagieFXPartInfo=$System.Forward
    RotorPartInfo:HoagieFXPartInfo=$System.Rotor
    TailPartInfo:HoagieFXPartInfo=$System.Tail
   ) {module=SelectHoagiePartInfo}
   /SetVariables(Particles.SpawnGroup=&MergeHelper_IntSwitch(
     0=&SpawnGroup()
     1=&SelectIntFromArray(
      Array Sampling Mode=::NewEnumerator1
      Direct Array Index=?
      Int Selection Array=DataInterface() {propertyValues {IntData [0,1]}}
      Random Seed=0
     )
     2=&SelectIntFromArray(
      Array Sampling Mode=::NewEnumerator1
      Direct Array Index=?
      Int Selection Array=DataInterface() {propertyValues {IntData [2,8]}}
      Random Seed=0
     )
     3=&SpawnGroup()
     4=&SpawnGroup()
     5=&SpawnGroup()
     6=&SpawnGroup()
     Index=&SpawnGroup()
     Num=::NewEnumerator5
    )) {isSetParameters=true}
   /Niagara/Modules/Spawn/Initialization/InitializeParticle(
    Color=<rgba 1 0.244 0 1>
    Lifetime=0.15
    Module.Write Color=true
    Module.Write Lifetime=true
    Module.Write Mass=false
    Module.Write Material Random=true
    Module.Write Position=false
    Module.Write Scale=false
    Module.Write SpriteRotation=false
    Module.Write SpriteSize=true
    Position=&SimulationPosition()
    Sprite Size=&Vector2DFromFloat(Value=&UniformRangedFloat(Minimum=70, Maximum=100))
   )
   &SampleSkeletalMeshSkeleton(
    Mesh:NiagaraDataInterfaceSkeletalMesh=$User.SkeletalMesh
    Spawn Mode:ENiagaraSkeletalSpawning=::NewEnumerator1
    Specific Socket Index:NiagaraInt32=$Particles.SpawnGroup
   )
   &SkeletalMeshLocation(
    Bone / Socket Index=&SpawnGroup()
    MaskSpawnGroup=false
    Mesh Sampling Type=::NewEnumerator5
    Sampled Position Offset=&Add_Vector(A=$Transient.HoagiePosition, B=$Transient.HoagieOffset)
    Skeletal Mesh=$User.SkeletalMesh
    Socket Sampling Mode=::NewEnumerator5
   )
   &CameraOffset(Camera Offset Amount=20)
   &ApplyInitialForces()
   &GenerateLocationEvent(Event Send Rate=2, Module.Use Event Send Rate=true)
   &FN_Light_v2(Light Exponent=2, Light Radius=150, LightEnabled=true, Scale Color=1)
  ]
  particleUpdateScript [
   &ParticleState()
   &SolveForcesAndVelocity(Clamp Velocity=false, Limit Acceleration=false)
   &ScaleColor(
    Scale Alpha=&FloatFromCurve(FloatCurve=DataInterface() {propertyValues {Curve {keys [
         {arriveTangent=60.2604, interpMode=RCIM_Cubic, leaveTangent=60.2604, tangentMode=RCTM_User, time=0, value=0}
         <at 0.05 1 RCIM_Cubic>
         <at 1 0 RCIM_Cubic>
        ]}}})
    Scale RGB=&VectorFromFloat(Value=&FloatFromCurve(FloatCurve=DataInterface() {propertyValues {Curve {keys [
          {arriveTangent=97.6859, interpMode=RCIM_Cubic, leaveTangent=97.6859, tangentMode=RCTM_User, time=0, value=1}
          <at 0.1 6 RCIM_Cubic>
          <at 1 1 RCIM_Cubic>
         ]}}}))
    ScaleA=true
    ScaleRGB=true
   )
  ]
  rendererClasses [&NiagaraSpriteRendererProperties, &NiagaraLightRendererProperties]
  renderers [
   {rendererClass=&NiagaraSpriteRendererProperties}
   {rendererClass=&NiagaraLightRendererProperties, rendererIndex=1}
  ]
 }
 Local_Embers_Optimized {
  emitterUpdateScript [
   &EmitterState(Life Cycle Mode=::NewEnumerator1, Loop Duration=1, UseLoopDelay=false)
   &SpawnRate(
    SpawnGroup=1
    SpawnRate=&MakeCustomFloatFromBool(
     Bool=&SetBoolByFloatComparison(A:NiagaraFloat=$User.DamageScale, B:NiagaraFloat=$User.Threshold.Engine.Stage2, ComparisonType:ENiagaraBooleanLogicOps=::NewEnumerator2)
     True Float=50
     False Float=0
    )
    Use Spawn Probability=false
   )
   &SpawnRate(
    SpawnGroup=1
    SpawnRate=&MakeCustomFloatFromBool(
     Bool=&SetBoolByFloatComparison(A:NiagaraFloat=$User.DamageScale, B:NiagaraFloat=$User.Threshold.Engine.Stage2, ComparisonType:ENiagaraBooleanLogicOps=::NewEnumerator2)
     True Float=50
     False Float=0
    )
    Use Spawn Probability=false
   ) {module=SpawnRate001}
   &SpawnRate(
    SpawnGroup=2
    SpawnRate=&MakeCustomFloatFromBool(
     Bool=&SetBoolByFloatComparison(A:NiagaraFloat=$User.DamageScale, B:NiagaraFloat=$User.Threshold.Rotor.Stage2, ComparisonType:ENiagaraBooleanLogicOps=::NewEnumerator2)
     True Float=100
     False Float=0
    )
    Use Spawn Probability=false
   ) {module=SpawnRate002}
   &SpawnRate(
    SpawnGroup=2
    SpawnRate=&MakeCustomFloatFromBool(
     Bool=&SetBoolByFloatComparison(A:NiagaraFloat=$User.DamageScale, B:NiagaraFloat=$User.Threshold.Rotor.Stage2, ComparisonType:ENiagaraBooleanLogicOps=::NewEnumerator2)
     True Float=100
     False Float=0
    )
    Use Spawn Probability=false
   ) {module=SpawnRate006}
   &SpawnRate(
    Spawn Probability=&MakeCustomFloatFromBool(Bool=$User.isMobile, True Float=1, False Float=0)
    SpawnGroup=3
    SpawnRate=&MakeCustomFloatFromBool(
     Bool=&SetBoolByFloatComparison(A:NiagaraFloat=$User.DamageScale, B:NiagaraFloat=$User.Threshold.Tail.Stage2, ComparisonType:ENiagaraBooleanLogicOps=::NewEnumerator0)
     True Float=50
     False Float=0
    )
    Use Spawn Probability=true
   ) {module=SpawnRate003}
   &SpawnRate(
    Spawn Probability=&MakeCustomFloatFromBool(Bool=$User.isMobile, True Float=1, False Float=0)
    SpawnGroup=4
    SpawnRate=&MakeCustomFloatFromBool(
     Bool=&SetBoolByFloatComparison(A:NiagaraFloat=$User.DamageScale, B:NiagaraFloat=$User.Threshold.Booster.Stage2, ComparisonType:ENiagaraBooleanLogicOps=::NewEnumerator2)
     True Float=100
     False Float=0
    )
    Use Spawn Probability=true
   ) {module=SpawnRate004}
   &SpawnRate(
    Spawn Probability=&MakeCustomFloatFromBool(Bool=$User.isMobile, True Float=1, False Float=0)
    SpawnGroup=5
    SpawnRate=&MakeCustomFloatFromBool(
     Bool=&SetBoolByFloatComparison(A:NiagaraFloat=$User.DamageScale, B:NiagaraFloat=$User.Threshold.Forward.Stage2, ComparisonType:ENiagaraBooleanLogicOps=::NewEnumerator2)
     True Float=50
     False Float=0
    )
    Use Spawn Probability=true
   ) {module=SpawnRate005}
  ]
  particleSpawnScript [
   /SetVariables(Particles.SpawnGroup=&SpawnGroup()) {isSetParameters=true}
   &FN_SelectHoagiePartInfo(
    BoostPartInfo:HoagieFXPartInfo=$System.Booster
    EnginePartInfo_L:HoagieFXPartInfo=$System.EnginePartInfo
    EnginePartInfo_R:HoagieFXPartInfo=$System.EngineR
    ForwardPartInfo:HoagieFXPartInfo=$System.Forward
    RotorPartInfo:HoagieFXPartInfo=$System.Rotor
    TailPartInfo:HoagieFXPartInfo=$System.Tail
   ) {module=SelectHoagiePartInfo}
   /SetVariables(Particles.SpawnGroup=&MergeHelper_IntSwitch(
     0=&SpawnGroup()
     1=&SelectIntFromArray(
      Array Sampling Mode=::NewEnumerator1
      Direct Array Index=?
      Int Selection Array=DataInterface() {propertyValues {IntData [0,1]}}
      Random Seed=0
     )
     2=&SelectIntFromArray(
      Array Sampling Mode=::NewEnumerator1
      Direct Array Index=?
      Int Selection Array=DataInterface() {propertyValues {IntData [2,8]}}
      Random Seed=0
     )
     3=&SpawnGroup()
     4=&SpawnGroup()
     5=&SpawnGroup()
     6=&SpawnGroup()
     Index=&SpawnGroup()
     Num=::NewEnumerator5
    )) {isSetParameters=true}
   &InitializeParticle(
    Color=<rgba 2 0.2858 0.0464 1>
    Lifetime=&UniformRangedFloat(Minimum=0.15, Maximum=0.5)
    Module.Write Color=true
    Module.Write Lifetime=true
    Module.Write Mass=true
    Module.Write Material Random=true
    Module.Write Position=false
    Module.Write Scale=false
    Module.Write SpriteRotation=false
    Module.Write SpriteSize=true
    Position=&SimulationPosition()
    Sprite Size=&UniformRangedVector2D(Minimum=<1 1>, Maximum=<2 2>)
   )
   &SampleSkeletalMeshSkeleton(
    Mesh:NiagaraDataInterfaceSkeletalMesh=$User.SkeletalMesh
    Spawn Mode:ENiagaraSkeletalSpawning=::NewEnumerator1
    Specific Socket Index:NiagaraInt32=$Particles.SpawnGroup
   )
   /Niagara/Modules/Spawn/Location/SkeletalMeshSkeletonLocation(
    Module.Inherit Velocity=false
    Module.Use Sampled Bone as Ribbon ID=true
    Sampled Position Offset:Vector3f=&Add_Vector(
     A=&RotateVectorByQuaternion(Vector To Rotate=$Transient.HoagiePosition, Quaternion=$Particles.SampleSkeletalMeshSkeleton.SampledRotation)
     B=$Transient.HoagieOffset
    )
   ) {enabled=false}
   &SkeletalMeshLocation(
    Bone / Socket Index=&SpawnGroup()
    MaskSpawnGroup=false
    Mesh Sampling Type=::NewEnumerator5
    Sampled Position Offset=&Add_Vector(A=$Transient.HoagiePosition, B=$Transient.HoagieOffset)
    Skeletal Mesh=$User.SkeletalMesh
    Socket Sampling Mode=::NewEnumerator5
   )
   &AddVelocity(Velocity=&Multiply_Vector(
     A=$Transient.HoagieNormal
     B=&VectorFromFloat(Value=&UniformRangedFloat(Minimum=50, Maximum=100))
    )) {module=AddVelocity002}
   &AddVelocity(Velocity=&UniformRangedVector(Minimum=$Transient.HoagieVelocityMin, Maximum=$Transient.HoagieVelocityMax)) {module=AddVelocity001}
   &CameraOffset(Camera Offset Amount=&UniformRangedFloat(Minimum=10, Maximum=25))
  ]
  particleUpdateScript [
   &ParticleState()
   &AccelerationForce(Acceleration=<0 0 -980>)
   &SolveForcesAndVelocity(Clamp Velocity=false, Limit Acceleration=false)
   &ScaleSpriteSize(Scale Factor=&Vector2DFromFloat(Value=&FloatFromCurve(FloatCurve=DataInterface() {propertyValues {Curve {keys [
          {arriveTangent=56.1861, interpMode=RCIM_Cubic, leaveTangent=56.1859, tangentMode=RCTM_User, time=0, value=0}
          <at 0.03 1 RCIM_Cubic>
          <at 1 0 RCIM_Cubic>
         ]}}})))
   &ScaleSpriteSizeBySpeed(Max Scale Factor=<1 15>)
   &ScaleColor(
    Scale RGB=&VectorFromCurve(VectorCurve=DataInterface() {propertyValues {
       XCurve {keys [<at 0 16 RCIM_Cubic>,<at 0.8 1 RCIM_Cubic>]}
       YCurve {keys [<at 0 16 RCIM_Cubic>,<at 0.8 0.1249 RCIM_Cubic>]}
       ZCurve {keys [<at 0 16 RCIM_Cubic>,<at 0.8 0.2547 RCIM_Cubic>]}
      }})
    ScaleA=true
    ScaleRGB=true
   )
  ]
  rendererClasses [&NiagaraSpriteRendererProperties]
  renderers [{rendererClass=&NiagaraSpriteRendererProperties}]
 }
 Local_Flames_Optimized {
  emitterUpdateScript [
   &EmitterState(Life Cycle Mode=::NewEnumerator1, Loop Duration=1, UseLoopDelay=false)
   &SpawnRate(
    SpawnGroup=1
    SpawnRate=&MakeCustomFloatFromBool(
     Bool=&SetBoolByFloatComparison(
      A:NiagaraFloat=$User.DamageScale
      B:NiagaraFloat=$User.Threshold.Engine.Stage3Critical
      ComparisonType:ENiagaraBooleanLogicOps=::NewEnumerator2
     )
     True Float=10
     False Float=0
    )
    Use Spawn Probability=false
   )
   &SpawnRate(
    SpawnGroup=1
    SpawnRate=&MakeCustomFloatFromBool(
     Bool=&SetBoolByFloatComparison(
      A:NiagaraFloat=$User.DamageScale
      B:NiagaraFloat=$User.Threshold.Engine.Stage3Critical
      ComparisonType:ENiagaraBooleanLogicOps=::NewEnumerator2
     )
     True Float=10
     False Float=0
    )
    Use Spawn Probability=false
   ) {module=SpawnRate001}
   &SpawnRate(
    SpawnGroup=2
    SpawnRate=&MakeCustomFloatFromBool(
     Bool=&SetBoolByFloatComparison(
      A:NiagaraFloat=$User.DamageScale
      B:NiagaraFloat=$User.Threshold.Rotor.Stage3Critical
      ComparisonType:ENiagaraBooleanLogicOps=::NewEnumerator2
     )
     True Float=10
     False Float=0
    )
    Use Spawn Probability=false
   ) {module=SpawnRate002}
   &SpawnRate(
    SpawnGroup=2
    SpawnRate=&MakeCustomFloatFromBool(
     Bool=&SetBoolByFloatComparison(
      A:NiagaraFloat=$User.DamageScale
      B:NiagaraFloat=$User.Threshold.Rotor.Stage3Critical
      ComparisonType:ENiagaraBooleanLogicOps=::NewEnumerator2
     )
     True Float=10
     False Float=0
    )
    Use Spawn Probability=false
   ) {module=SpawnRate006}
   &SpawnRate(
    Spawn Probability=&MakeCustomFloatFromBool(Bool=$User.isMobile, True Float=1, False Float=0)
    SpawnGroup=3
    SpawnRate=&MakeCustomFloatFromBool(
     Bool=&SetBoolByFloatComparison(
      A:NiagaraFloat=$User.DamageScale
      B:NiagaraFloat=$User.Threshold.Tail.Stage3Critical
      ComparisonType:ENiagaraBooleanLogicOps=::NewEnumerator0
     )
     True Float=10
     False Float=0
    )
    Use Spawn Probability=true
   ) {module=SpawnRate003}
   &SpawnRate(
    Spawn Probability=&MakeCustomFloatFromBool(Bool=$User.isMobile, True Float=1, False Float=0)
    SpawnGroup=4
    SpawnRate=&MakeCustomFloatFromBool(
     Bool=&SetBoolByFloatComparison(
      A:NiagaraFloat=$User.DamageScale
      B:NiagaraFloat=$User.Threshold.Booster.Stage3Critical
      ComparisonType:ENiagaraBooleanLogicOps=::NewEnumerator2
     )
     True Float=10
     False Float=0
    )
    Use Spawn Probability=true
   ) {module=SpawnRate004}
   &SpawnRate(
    Spawn Probability=&MakeCustomFloatFromBool(Bool=$User.isMobile, True Float=1, False Float=0)
    SpawnGroup=5
    SpawnRate=&MakeCustomFloatFromBool(
     Bool=&SetBoolByFloatComparison(
      A:NiagaraFloat=$User.DamageScale
      B:NiagaraFloat=$User.Threshold.Forward.Stage3Critical
      ComparisonType:ENiagaraBooleanLogicOps=::NewEnumerator2
     )
     True Float=10
     False Float=0
    )
    Use Spawn Probability=true
   ) {module=SpawnRate005}
  ]
  particleSpawnScript [
   /SetVariables(Particles.SpawnGroup=&SpawnGroup()) {isSetParameters=true}
   &FN_SelectHoagiePartInfo(
    BoostPartInfo:HoagieFXPartInfo=$System.Booster
    EnginePartInfo_L:HoagieFXPartInfo=$System.EnginePartInfo
    EnginePartInfo_R:HoagieFXPartInfo=$System.EngineR
    ForwardPartInfo:HoagieFXPartInfo=$System.Forward
    RotorPartInfo:HoagieFXPartInfo=$System.Rotor
    TailPartInfo:HoagieFXPartInfo=$System.Tail
   ) {module=SelectHoagiePartInfo}
   /SetVariables(Particles.SpawnGroup=&MergeHelper_IntSwitch(
     0=&SpawnGroup()
     1=&SelectIntFromArray(
      Array Sampling Mode=::NewEnumerator1
      Direct Array Index=?
      Int Selection Array=DataInterface() {propertyValues {IntData [0,1]}}
      Random Seed=0
     )
     2=&SelectIntFromArray(
      Array Sampling Mode=::NewEnumerator1
      Direct Array Index=?
      Int Selection Array=DataInterface() {propertyValues {IntData [2,8]}}
      Random Seed=0
     )
     3=&SpawnGroup()
     4=&SpawnGroup()
     5=&SpawnGroup()
     6=&SpawnGroup()
     Index=&SpawnGroup()
     Num=::NewEnumerator5
    )) {isSetParameters=true}
   &InitializeParticle(
    Color=<rgba 2 0.9857 0.2812 1>
    Lifetime=&UniformRangedFloat(Minimum=0.4, Maximum=0.75)
    Module.Write Color=true
    Module.Write Lifetime=true
    Module.Write Mass=true
    Module.Write Material Random=true
    Module.Write Position=false
    Module.Write Scale=false
    Module.Write SpriteRotation=true
    Module.Write SpriteSize=true
    Position=&SimulationPosition()
    Sprite Rotation=&NormalizedAngleToDegrees(NormalizedAngle=&UniformRangedFloat(Minimum=0, Maximum=1))
    Sprite Size=&Vector2DFromFloat(Value=&UniformRangedFloat(Minimum=80, Maximum=125))
   )
   &SampleSkeletalMeshSkeleton(
    Mesh:NiagaraDataInterfaceSkeletalMesh=$User.SkeletalMesh
    Spawn Mode:ENiagaraSkeletalSpawning=::NewEnumerator1
    Specific Socket Index:NiagaraInt32=$Particles.SpawnGroup
   )
   &SkeletalMeshSkeletonLocation(
    Module.Inherit Velocity=false
    Module.Use Sampled Bone as Ribbon ID=true
    Sampled Position Offset:Vector3f=&Add_Vector(
     A=&RotateVectorByQuaternion(Vector To Rotate=$Transient.HoagiePosition, Quaternion=$Particles.SampleSkeletalMeshSkeleton.SampledRotation)
     B=$Transient.HoagieOffset
    )
   ) {enabled=false}
   &SkeletalMeshLocation(
    Bone / Socket Index=&SpawnGroup()
    MaskSpawnGroup=false
    Mesh Sampling Type=::NewEnumerator5
    Sampled Position Offset=&Add_Vector(A=$Transient.HoagiePosition, B=$Transient.HoagieOffset)
    Skeletal Mesh=$User.SkeletalMesh
    Socket Sampling Mode=::NewEnumerator5
   )
   &AddVelocity(Velocity=&UniformRangedVector(Minimum=$Transient.HoagieVelocityMin, Maximum=$Transient.HoagieVelocityMax)) {module=AddVelocity001}
   &AddVelocity(Velocity=&Multiply_Vector(
     A=$Transient.HoagieNormal
     B=&VectorFromFloat(Value=&UniformRangedFloat(Minimum=50, Maximum=100))
    )) {module=AddVelocity002}
   &CameraOffset(Camera Offset Amount=25)
   &DynamicMaterialParameters(
    Index 0 Param 2=&UniformRangedFloat(Minimum=0.333, Maximum=1)
    Index 0 Param 4=&UniformRangedFloat(Minimum=1.333, Maximum=2)
    Module.Param0WriteEnabled=false
    Module.Param1WriteEnabled=true
    Module.Param2WriteEnabled=false
    Module.Param3WriteEnabled=true
   )
  ]
  particleUpdateScript [
   &ParticleState()
   &AccelerationForce(Acceleration=<0 0 100>)
   &Drag(Drag=2, Use Linear Drag=true, Use Rotational Drag=false)
   &SolveForcesAndVelocity(Clamp Velocity=false, Limit Acceleration=false)
   &DynamicMaterialParameters(
    Index 0 Param 1=&FloatFromCurve(FloatCurve=DataInterface() {propertyValues {Curve {keys [{arriveTangent=2.3188, interpMode=RCIM_Cubic, time=0, value=0}, <at 0.5 1 RCIM_Cubic>]}}})
    Index 0 Param 3=&FloatFromCurve(FloatCurve=DataInterface() {propertyValues {Curve {keys [{arriveTangent=1.4224, interpMode=RCIM_Cubic, time=0.5, value=1}, <at 1 2 RCIM_Cubic>]}}})
    Module.Param0WriteEnabled=true
    Module.Param1WriteEnabled=false
    Module.Param2WriteEnabled=true
    Module.Param3WriteEnabled=false
   ) {module=DynamicMaterialParameters001}
   &ScaleColor(
    Scale Alpha=&FloatFromCurve(FloatCurve=DataInterface() {propertyValues {Curve {keys [
         {arriveTangent=6.233, interpMode=RCIM_Cubic, leaveTangent=6.233, tangentMode=RCTM_User, time=0, value=0}
         {arriveTangent=3.1104, interpMode=RCIM_Cubic, leaveTangent=3.1104, time=0.2, value=1.2445}
         {arriveTangent=-1.5556, interpMode=RCIM_Cubic, leaveTangent=-1.5556, time=0.6833, value=2.1254}
         {interpMode=RCIM_Cubic, leaveTangent=-3.4385, time=1, value=0}
        ]}}})
    Scale RGB=&MakeVectorFromLinearColorRGB(Color=&ColorFromCurve(ColorCurve=DataInterface() {propertyValues {
        AlphaCurve {keys [<at -0.0023 1 RCIM_Cubic>,<at 1 1 RCIM_Cubic>]}
        BlueCurve {keys [<at -0.002 0 RCIM_Cubic>,<at 0.998 0 RCIM_Cubic>]}
        GreenCurve {keys [<at -0.002 1.3836 RCIM_Cubic>,<at 0.998 0.0572 RCIM_Cubic>]}
        RedCurve {keys [<at -0.002 4 RCIM_Cubic>,<at 0.998 1 RCIM_Cubic>]}
       }}))
    ScaleA=true
    ScaleRGB=true
   )
   &ScaleSpriteSize(Scale Factor=&Vector2DFromFloat(Value=&FloatFromCurve(FloatCurve=DataInterface() {propertyValues {Curve {keys [
          {arriveTangent=10.5064, interpMode=RCIM_Cubic, leaveTangent=10.5064, tangentMode=RCTM_User, time=0, value=0.0211}
          {arriveTangent=0.6433, interpMode=RCIM_Cubic, leaveTangent=0.6433, tangentMode=RCTM_User, time=0.1836, value=1}
          <at 1 0.5>
         ]}}})))
   /SetVariables(Particles.SubImageIndex=&FloatFromCurve(FloatCurve=DataInterface() {propertyValues {Curve {keys [{arriveTangent=87.6094, time=0, value=0}, {arriveTangent=33.5534, leaveTangent=33.5534, time=1, value=63}]}}})) {isSetParameters=true}
  ]
  rendererClasses [&NiagaraSpriteRendererProperties]
  renderers [{rendererClass=&NiagaraSpriteRendererProperties}]
 }
 Local_Smoke_Optimized {
  emitterUpdateScript [
   &EmitterState()
   &SpawnRate(
    SpawnGroup=1
    SpawnRate=&MakeCustomFloatFromBool(
     Bool=&SetBoolByFloatComparison(A:NiagaraFloat=$User.DamageScale, B:NiagaraFloat=$User.Threshold.Engine.Stage1, ComparisonType:ENiagaraBooleanLogicOps=::NewEnumerator2)
     True Float=&MakeCustomFloatFromBool(
      Bool=&SetBoolByFloatComparison(A:NiagaraFloat=$User.DamageScale, B:NiagaraFloat=$User.Threshold.Engine.Stage2, ComparisonType:ENiagaraBooleanLogicOps=::NewEnumerator2)
      True Float=10
      False Float=5
     )
     False Float=0
    )
    Use Spawn Probability=false
   )
   &SpawnRate(
    SpawnGroup=1
    SpawnRate=&MakeCustomFloatFromBool(
     Bool=&SetBoolByFloatComparison(A:NiagaraFloat=$User.DamageScale, B:NiagaraFloat=$User.Threshold.Engine.Stage1, ComparisonType:ENiagaraBooleanLogicOps=::NewEnumerator2)
     True Float=&MakeCustomFloatFromBool(
      Bool=&SetBoolByFloatComparison(A:NiagaraFloat=$User.DamageScale, B:NiagaraFloat=$User.Threshold.Engine.Stage2, ComparisonType:ENiagaraBooleanLogicOps=::NewEnumerator2)
      True Float=10
      False Float=5
     )
     False Float=0
    )
    Use Spawn Probability=false
   ) {module=SpawnRate001}
   &SpawnRate(
    SpawnGroup=2
    SpawnRate=&MakeCustomFloatFromBool(
     Bool=&SetBoolByFloatComparison(A:NiagaraFloat=$User.DamageScale, B:NiagaraFloat=$User.Threshold.Rotor.Stage1, ComparisonType:ENiagaraBooleanLogicOps=::NewEnumerator2)
     True Float=&MakeCustomFloatFromBool(
      Bool=&SetBoolByFloatComparison(A:NiagaraFloat=$User.DamageScale, B:NiagaraFloat=$User.Threshold.Rotor.Stage2, ComparisonType:ENiagaraBooleanLogicOps=::NewEnumerator2)
      True Float=20
      False Float=10
     )
     False Float=0
    )
    Use Spawn Probability=false
   ) {module=SpawnRate002}
   &SpawnRate(
    SpawnGroup=2
    SpawnRate=&MakeCustomFloatFromBool(
     Bool=&SetBoolByFloatComparison(A:NiagaraFloat=$User.DamageScale, B:NiagaraFloat=$User.Threshold.Rotor.Stage1, ComparisonType:ENiagaraBooleanLogicOps=::NewEnumerator2)
     True Float=&MakeCustomFloatFromBool(
      Bool=&SetBoolByFloatComparison(A:NiagaraFloat=$User.DamageScale, B:NiagaraFloat=$User.Threshold.Rotor.Stage2, ComparisonType:ENiagaraBooleanLogicOps=::NewEnumerator2)
      True Float=20
      False Float=10
     )
     False Float=0
    )
    Use Spawn Probability=false
   ) {module=SpawnRate006}
   &SpawnRate(
    Spawn Probability=&MakeCustomFloatFromBool(Bool=$User.isMobile, True Float=1, False Float=0)
    SpawnGroup=3
    SpawnRate=&MakeCustomFloatFromBool(
     Bool=&SetBoolByFloatComparison(A:NiagaraFloat=$User.DamageScale, B:NiagaraFloat=$User.Threshold.Tail.Stage1, ComparisonType:ENiagaraBooleanLogicOps=::NewEnumerator0)
     True Float=&MakeCustomFloatFromBool(
      Bool=&SetBoolByFloatComparison(A:NiagaraFloat=$User.DamageScale, B:NiagaraFloat=$User.Threshold.Tail.Stage2, ComparisonType:ENiagaraBooleanLogicOps=::NewEnumerator0)
      True Float=20
      False Float=10
     )
     False Float=0
    )
    Use Spawn Probability=true
   ) {module=SpawnRate003}
   &SpawnRate(
    Spawn Probability=&MakeCustomFloatFromBool(Bool=$User.isMobile, True Float=1, False Float=0)
    SpawnGroup=4
    SpawnRate=&MakeCustomFloatFromBool(
     Bool=&SetBoolByFloatComparison(A:NiagaraFloat=$User.DamageScale, B:NiagaraFloat=$User.Threshold.Booster.Stage1, ComparisonType:ENiagaraBooleanLogicOps=::NewEnumerator0)
     True Float=&MakeCustomFloatFromBool(
      Bool=&SetBoolByFloatComparison(A:NiagaraFloat=$User.DamageScale, B:NiagaraFloat=$User.Threshold.Booster.Stage2, ComparisonType:ENiagaraBooleanLogicOps=::NewEnumerator0)
      True Float=20
      False Float=10
     )
     False Float=0
    )
    Use Spawn Probability=true
   ) {module=SpawnRate004}
   &SpawnRate(
    Spawn Probability=&MakeCustomFloatFromBool(Bool=$User.isMobile, True Float=1, False Float=0)
    SpawnGroup=5
    SpawnRate=&MakeCustomFloatFromBool(
     Bool=&SetBoolByFloatComparison(A:NiagaraFloat=$User.DamageScale, B:NiagaraFloat=$User.Threshold.Forward.Stage1, ComparisonType:ENiagaraBooleanLogicOps=::NewEnumerator2)
     True Float=&MakeCustomFloatFromBool(
      Bool=&SetBoolByFloatComparison(A:NiagaraFloat=$User.DamageScale, B:NiagaraFloat=$User.Threshold.Forward.Stage2, ComparisonType:ENiagaraBooleanLogicOps=::NewEnumerator2)
      True Float=10
      False Float=5
     )
     False Float=0
    )
    Use Spawn Probability=true
   ) {module=SpawnRate005}
  ]
  particleSpawnScript [
   /SetVariables(Particles.SpawnGroup=&SpawnGroup()) {isSetParameters=true}
   &FN_SelectHoagiePartInfo(
    BoostPartInfo:HoagieFXPartInfo=$System.Booster
    EnginePartInfo_L:HoagieFXPartInfo=$System.EnginePartInfo
    EnginePartInfo_R:HoagieFXPartInfo=$System.EngineR
    ForwardPartInfo:HoagieFXPartInfo=$System.Forward
    RotorPartInfo:HoagieFXPartInfo=$System.Rotor
    TailPartInfo:HoagieFXPartInfo=$System.Tail
   ) {module=SelectHoagiePartInfo}
   /SetVariables(Particles.SpawnGroup=&MergeHelper_IntSwitch(
     0=&SpawnGroup()
     1=&SelectIntFromArray(
      Array Sampling Mode=::NewEnumerator1
      Direct Array Index=?
      Int Selection Array=DataInterface() {propertyValues {IntData [0,1]}}
      Random Seed=0
     )
     2=&SelectIntFromArray(
      Array Sampling Mode=::NewEnumerator1
      Direct Array Index=?
      Int Selection Array=DataInterface() {propertyValues {IntData [2,8]}}
      Random Seed=0
     )
     3=&SpawnGroup()
     4=&SpawnGroup()
     5=&SpawnGroup()
     6=&SpawnGroup()
     Index=&SpawnGroup()
     Num=::NewEnumerator5
    )) {isSetParameters=true}
   &InitializeParticle(
    Lifetime=&UniformRangedFloat(Minimum=0.45, Maximum=0.9)
    Module.Write Color=true
    Module.Write Lifetime=true
    Module.Write Mass=false
    Module.Write Material Random=true
    Module.Write Position=false
    Module.Write Scale=false
    Module.Write SpriteRotation=true
    Module.Write SpriteSize=true
    Sprite Rotation=&NormalizedAngleToDegrees(NormalizedAngle=&UniformRangedFloat(Minimum=0, Maximum=1))
    Sprite Size=&Vector2DFromFloat(Value=&UniformRangedFloat(Minimum=$Transient.HoagieSpriteMin, Maximum=$Transient.HoagieSpriteMax))
   )
   &SampleSkeletalMeshSkeleton(
    Mesh:NiagaraDataInterfaceSkeletalMesh=$User.SkeletalMesh
    Spawn Mode:ENiagaraSkeletalSpawning=::NewEnumerator1
    Specific Socket Index:NiagaraInt32=$Particles.SpawnGroup
   )
   &SkeletalMeshSkeletonLocation(
    Module.Inherit Velocity=false
    Module.Use Sampled Bone as Ribbon ID=true
    Sampled Position Offset:Vector3f=&Add_Vector(
     A=&RotateVectorByQuaternion(Vector To Rotate=$Transient.HoagiePosition, Quaternion=$Particles.SampleSkeletalMeshSkeleton.SampledRotation)
     B=$Transient.HoagieOffset
    )
   ) {enabled=false}
   &SkeletalMeshLocation(
    Bone / Socket Index=&SpawnGroup()
    MaskSpawnGroup=false
    Mesh Sampling Type=::NewEnumerator5
    Sampled Position Offset=&Add_Vector(A=$Transient.HoagiePosition, B=$Transient.HoagieOffset)
    Skeletal Mesh=$User.SkeletalMesh
    Socket Sampling Mode=::NewEnumerator5
   )
   &AddVelocity(Velocity=&Multiply_Vector(
     A=$Transient.HoagieNormal
     B=&VectorFromFloat(Value=&UniformRangedFloat(Minimum=50, Maximum=100))
    )) {module=AddVelocity002}
   &AddVelocity(Velocity=&UniformRangedVector(Minimum=$Transient.HoagieVelocityMin, Maximum=$Transient.HoagieVelocityMax)) {module=AddVelocity001}
   &CameraOffset(Camera Offset Amount=25)
  ]
  particleUpdateScript [
   &ParticleState()
   &AccelerationForce(Acceleration=/Hoagie/FX/DynamicInputs/FN_SelectHoagieVectorbySpawnGroup(
     "(0)_Engine_L":Vector3f=<0 0 100>
     "(1)_Engine_R":Vector3f=<0 0 100>
     "(2)_Rotor":Vector3f=<0 0 100>
     "(3)_Tail":Vector3f=<0 0 200>
     "(4)_BoosterEngine":Vector3f=<0 0 200>
     "(5)_Forward":Vector3f=<0 0 100>
    ))
   &Drag(Drag=2, Use Linear Drag=true, Use Rotational Drag=false)
   &SolveForcesAndVelocity(Clamp Velocity=false, Limit Acceleration=false)
   /SetVariables() {isSetParameters=true}
   &ScaleColor(
    Scale RGB=&MakeVectorFromLinearColorRGB(Color=&ColorFromCurve(
      ColorCurve=DataInterface() {propertyValues {
        AlphaCurve {keys [<at 0 1 RCIM_Constant>,<at 1 1 RCIM_Constant>]}
        BlueCurve {keys [<at 0 0.1 RCIM_Constant>,<at 0.45 0.05 RCIM_Constant>,<at 1 0.0464 RCIM_Constant>]}
        GreenCurve {keys [<at 0 0.1 RCIM_Constant>,<at 0.45 0.05 RCIM_Constant>,<at 1 0.2858 RCIM_Constant>]}
        RedCurve {keys [<at 0 0.1 RCIM_Constant>,<at 0.45 0.05 RCIM_Constant>,<at 1 2 RCIM_Constant>]}
       }}
      CurveIndex=$User.DamageScale
     ))
    ScaleA=false
    ScaleRGB=true
   ) {module=ScaleColor001}
   &ScaleColor(
    Scale Alpha=&FloatFromCurve(FloatCurve=DataInterface() {propertyValues {Curve {keys [
         {interpMode=RCIM_Cubic, tangentMode=RCTM_User, time=0, value=0}
         {interpMode=RCIM_Cubic, tangentMode=RCTM_User, time=0.1, value=1}
         {arriveTangent=-2.0134, interpMode=RCIM_Cubic, leaveTangent=-2.0134, tangentMode=RCTM_User, time=0.554, value=0.7238}
         {interpMode=RCIM_Cubic, tangentMode=RCTM_User, time=1, value=0}
        ]}}})
    Scale RGB=&VectorFromFloat(Value=&FloatFromCurve(FloatCurve=DataInterface() {propertyValues {Curve {keys [{arriveTangent=-3.2082, interpMode=RCIM_Cubic, leaveTangent=-1.8602, tangentMode=RCTM_User, time=0, value=1.25}, <at 1 0.015 RCIM_Cubic>]}}}))
    ScaleA=true
    ScaleRGB=true
   )
   &DynamicMaterialParameters(
    Index 0 Param 1=&FloatFromCurve(FloatCurve=DataInterface() {propertyValues {Curve {keys [<at 0 1>,<at 1 -1>]}}})
    Index 0 Param 4=1
    Module.Param0WriteEnabled=true
    Module.Param1WriteEnabled=true
    Module.Param2WriteEnabled=true
    Module.Param3WriteEnabled=true
   )
   &ScaleSpriteSize(Scale Factor=&Vector2DFromFloat(Value=&MakeCustomFloatFromBool(
      Bool=/Game/Effects/Niagara/DynamicInputs/Logic/FN_SetBoolByIntComparison(A:NiagaraInt32=&SpawnGroup(), B:NiagaraInt32=1, ComparisonType:ENiagaraBooleanLogicOps=::NewEnumerator0)
      True Float=&FloatFromCurve(FloatCurve=DataInterface() {propertyValues {Curve {keys [{arriveTangent=2.8178, interpMode=RCIM_Cubic, leaveTangent=2.8178, tangentMode=RCTM_User, time=0, value=0.5}, <at 1 2 RCIM_Cubic>]}}})
      False Float=&FloatFromCurve(FloatCurve=DataInterface() {propertyValues {Curve {keys [{arriveTangent=4.3948, interpMode=RCIM_Cubic, leaveTangent=4.3948, tangentMode=RCTM_User, time=0, value=0.5}, <at 1 3 RCIM_Cubic>]}}})
     )))
   &SpriteRotationRate(Rotation Rate=&UniformRangedFloat(Minimum=-1, Maximum=1))
  ]
  rendererClasses [&NiagaraSpriteRendererProperties]
  renderers [{rendererClass=&NiagaraSpriteRendererProperties}]
 }
 World_BubbleSmoke_Optimized {
  emitterUpdateScript [
   &EmitterState(Life Cycle Mode=::NewEnumerator1, Loop Duration=1, UseLoopDelay=false)
   &SpawnRate(
    SpawnGroup=1
    SpawnRate=&MakeCustomFloatFromBool(
     Bool=&SetBoolByFloatComparison(A:NiagaraFloat=$User.DamageScale, B:NiagaraFloat=$User.Threshold.Engine.Stage2, ComparisonType:ENiagaraBooleanLogicOps=::NewEnumerator2)
     True Float=&MakeCustomFloatFromBool(
      Bool=&SetBoolByFloatComparison(
       A:NiagaraFloat=$User.DamageScale
       B:NiagaraFloat=$User.Threshold.Engine.Stage3Critical
       ComparisonType:ENiagaraBooleanLogicOps=::NewEnumerator2
      )
      True Float=10
      False Float=5
     )
     False Float=0
    )
    Use Spawn Probability=true
   )
   &SpawnRate(
    SpawnGroup=1
    SpawnRate=&MakeCustomFloatFromBool(
     Bool=&SetBoolByFloatComparison(A:NiagaraFloat=$User.DamageScale, B:NiagaraFloat=$User.Threshold.Engine.Stage2, ComparisonType:ENiagaraBooleanLogicOps=::NewEnumerator2)
     True Float=&MakeCustomFloatFromBool(
      Bool=&SetBoolByFloatComparison(
       A:NiagaraFloat=$User.DamageScale
       B:NiagaraFloat=$User.Threshold.Engine.Stage3Critical
       ComparisonType:ENiagaraBooleanLogicOps=::NewEnumerator2
      )
      True Float=10
      False Float=5
     )
     False Float=0
    )
    Use Spawn Probability=true
   ) {module=SpawnRate001}
   &SpawnRate(
    SpawnGroup=2
    SpawnRate=&MakeCustomFloatFromBool(
     Bool=&SetBoolByFloatComparison(A:NiagaraFloat=$User.DamageScale, B:NiagaraFloat=$User.Threshold.Rotor.Stage2, ComparisonType:ENiagaraBooleanLogicOps=::NewEnumerator2)
     True Float=&MakeCustomFloatFromBool(
      Bool=&SetBoolByFloatComparison(
       A:NiagaraFloat=$User.DamageScale
       B:NiagaraFloat=$User.Threshold.Rotor.Stage3Critical
       ComparisonType:ENiagaraBooleanLogicOps=::NewEnumerator2
      )
      True Float=10
      False Float=5
     )
     False Float=0
    )
    Use Spawn Probability=true
   ) {module=SpawnRate002}
   &SpawnRate(
    SpawnGroup=2
    SpawnRate=&MakeCustomFloatFromBool(
     Bool=&SetBoolByFloatComparison(A:NiagaraFloat=$User.DamageScale, B:NiagaraFloat=$User.Threshold.Rotor.Stage2, ComparisonType:ENiagaraBooleanLogicOps=::NewEnumerator2)
     True Float=&MakeCustomFloatFromBool(
      Bool=&SetBoolByFloatComparison(
       A:NiagaraFloat=$User.DamageScale
       B:NiagaraFloat=$User.Threshold.Rotor.Stage3Critical
       ComparisonType:ENiagaraBooleanLogicOps=::NewEnumerator2
      )
      True Float=10
      False Float=5
     )
     False Float=0
    )
    Use Spawn Probability=true
   ) {module=SpawnRate003}
  ]
  particleSpawnScript [
   /SetVariables(Particles.SpawnGroup=&SpawnGroup()) {isSetParameters=true}
   &FN_SelectHoagiePartInfo(
    BoostPartInfo:HoagieFXPartInfo=$System.Booster
    EnginePartInfo_L:HoagieFXPartInfo=$System.EnginePartInfo
    EnginePartInfo_R:HoagieFXPartInfo=$System.EngineR
    ForwardPartInfo:HoagieFXPartInfo=$System.Forward
    RotorPartInfo:HoagieFXPartInfo=$System.Rotor
    TailPartInfo:HoagieFXPartInfo=$System.Tail
   ) {module=SelectHoagiePartInfo}
   /SetVariables(Particles.SpawnGroup=&MergeHelper_IntSwitch(
     0=&SpawnGroup()
     1=&SelectIntFromArray(
      Array Sampling Mode=::NewEnumerator1
      Direct Array Index=?
      Int Selection Array=DataInterface() {propertyValues {IntData [0,1]}}
      Random Seed=0
     )
     2=&SelectIntFromArray(
      Array Sampling Mode=::NewEnumerator1
      Direct Array Index=?
      Int Selection Array=DataInterface() {propertyValues {IntData [2,8]}}
      Random Seed=0
     )
     3=&SpawnGroup()
     4=&SpawnGroup()
     5=&SpawnGroup()
     6=&SpawnGroup()
     Index=&SpawnGroup()
     Num=::NewEnumerator5
    )) {isSetParameters=true}
   &InitializeParticle(
    Color=<rgba 1 1 1 0.75>
    Lifetime=&UniformRangedFloat(Minimum=0.75, Maximum=1.75)
    Module.Write Color=true
    Module.Write Lifetime=true
    Module.Write Mass=true
    Module.Write Material Random=true
    Module.Write Position=false
    Module.Write Scale=false
    Module.Write SpriteRotation=true
    Module.Write SpriteSize=true
    Position=&SimulationPosition()
    Sprite Rotation=&NormalizedAngleToDegrees(NormalizedAngle=&UniformRangedFloat(Minimum=0, Maximum=1))
    Sprite Size=&Vector2DFromFloat(Value=&UniformRangedFloat(Minimum=150, Maximum=200))
   )
   &SampleSkeletalMeshSkeleton(
    Mesh:NiagaraDataInterfaceSkeletalMesh=$User.SkeletalMesh
    Spawn Mode:ENiagaraSkeletalSpawning=::NewEnumerator1
    Specific Socket Index:NiagaraInt32=$Particles.SpawnGroup
   )
   &SkeletalMeshSkeletonLocation(
    Module.Inherit Velocity=false
    Module.Use Sampled Bone as Ribbon ID=true
    Sampled Position Offset:Vector3f=&Add_Vector(
     A=&RotateVectorByQuaternion(Vector To Rotate=$Transient.HoagiePosition, Quaternion=$Particles.SampleSkeletalMeshSkeleton.SampledRotation)
     B=$Transient.HoagieOffset
    )
   ) {enabled=false}
   &SkeletalMeshLocation(
    Bone / Socket Index=&SpawnGroup()
    MaskSpawnGroup=false
    Mesh Sampling Type=::NewEnumerator5
    Sampled Position Offset=&Add_Vector(A=$Transient.HoagiePosition, B=$Transient.HoagieOffset)
    Skeletal Mesh=$User.SkeletalMesh
    Socket Sampling Mode=::NewEnumerator5
   )
   &AddVelocity(Velocity=&Multiply_Vector(
     A=$Transient.HoagieNormal
     B=&VectorFromFloat(Value=&UniformRangedFloat(Minimum=50, Maximum=100))
    )) {module=AddVelocity002}
   &AddVelocity(Velocity=&UniformRangedVector(Minimum=$Transient.HoagieVelocityMin, Maximum=$Transient.HoagieVelocityMax)) {module=AddVelocity001}
   &CameraOffset(Camera Offset Amount=10)
   /SetVariables() {isSetParameters=true}
  ]
  particleUpdateScript [
   &ParticleState()
   &AccelerationForce(Acceleration=<0 0 500>)
   /Niagara/Modules/Update/Forces/CurlNoiseForce(
    Noise Frequency=0.5
    Noise Strength=&Lerp_Float(A=0, B=5000, Alpha=&Power(a=$Particles.NormalizedAge, b=0.25))
    Pan Noise=true
    Pan Noise Field=<0 0 100>
   )
   &Drag(Drag=5, Use Linear Drag=true, Use Rotational Drag=false)
   &SolveForcesAndVelocity(Clamp Velocity=false, Limit Acceleration=false)
   &DynamicMaterialParameters(
    Index 0 Param 1=&FloatFromCurve(FloatCurve=DataInterface() {propertyValues {Curve {keys [{interpMode=RCIM_Cubic, tangentMode=RCTM_User, time=0, value=1}, {interpMode=RCIM_Cubic, tangentMode=RCTM_User, time=1, value=-1}]}}})
    Index 0 Param 4=1
    Module.Param0WriteEnabled=true
    Module.Param1WriteEnabled=false
    Module.Param2WriteEnabled=false
    Module.Param3WriteEnabled=true
   ) {module=DynamicMaterialParameters001}
   &ScaleColor(
    Scale Alpha=&FloatFromCurve(FloatCurve=DataInterface() {propertyValues {Curve {keys [
         <at 0 0 RCIM_Cubic>
         {interpMode=RCIM_Cubic, tangentMode=RCTM_User, time=0.1, value=1}
         {arriveTangent=-2.0134, interpMode=RCIM_Cubic, leaveTangent=-2.0134, tangentMode=RCTM_User, time=0.554, value=0.7238}
         {interpMode=RCIM_Cubic, tangentMode=RCTM_User, time=1, value=0}
        ]}}})
    Scale RGB=&VectorFromFloat(Value=&FloatFromCurve(FloatCurve=DataInterface() {propertyValues {Curve {keys [{arriveTangent=-2.1466, interpMode=RCIM_Cubic, leaveTangent=-2.1466, tangentMode=RCTM_User, time=0.05, value=1}, <at 1 0.015 RCIM_Cubic>]}}}))
    ScaleA=true
    ScaleRGB=true
   )
   &ScaleColor(
    Scale Alpha=&MakeCustomFloatFromBool(
     Bool=&SetBoolByFloatComparison(A:NiagaraFloat=$User.DamageScale, B:NiagaraFloat=0.499, ComparisonType:ENiagaraBooleanLogicOps=::NewEnumerator0)
     True Float=&MakeCustomFloatFromBool(
      Bool=&SetBoolByFloatComparison(A:NiagaraFloat=$User.DamageScale, B:NiagaraFloat=0.799, ComparisonType:ENiagaraBooleanLogicOps=::NewEnumerator0)
      True Float=1
      False Float=0.5
     )
     False Float=0
    )
    Scale RGB=&MakeVectorFromLinearColorRGB(Color=<rgba 0.05 0.05 0.05 1>)
    ScaleA=true
    ScaleRGB=true
   ) {module=ScaleColor001}
   &ScaleSpriteSize(Scale Factor=&Vector2DFromFloat(Value=&FloatFromCurve(FloatCurve=DataInterface() {propertyValues {Curve {keys [{arriveTangent=4.3948, interpMode=RCIM_Cubic, leaveTangent=4.3948, tangentMode=RCTM_User, time=0, value=1}, <at 1 3 RCIM_Cubic>]}}})))
   &SpriteRotationRate(Rotation Rate=&UniformRangedFloat(Minimum=-0.25, Maximum=0.25))
  ]
  rendererClasses [&NiagaraSpriteRendererProperties]
  renderers [{rendererClass=&NiagaraSpriteRendererProperties}]
 }
 World_Smoke_Optimized {
  emitterUpdateScript [
   &EmitterState(Life Cycle Mode=::NewEnumerator1, Loop Duration=1, UseLoopDelay=false)
   &SpawnRate(
    SpawnGroup=1
    SpawnRate=&MakeCustomFloatFromBool(
     Bool=&SetBoolByFloatComparison(A:NiagaraFloat=$User.DamageScale, B:NiagaraFloat=$User.Threshold.Engine.Stage2, ComparisonType:ENiagaraBooleanLogicOps=::NewEnumerator2)
     True Float=&MakeCustomFloatFromBool(
      Bool=&SetBoolByFloatComparison(
       A:NiagaraFloat=$User.DamageScale
       B:NiagaraFloat=$User.Threshold.Engine.Stage3Critical
       ComparisonType:ENiagaraBooleanLogicOps=::NewEnumerator2
      )
      True Float=10
      False Float=5
     )
     False Float=0
    )
    Use Spawn Probability=true
   )
   &SpawnRate(
    SpawnGroup=1
    SpawnRate=&MakeCustomFloatFromBool(
     Bool=&SetBoolByFloatComparison(A:NiagaraFloat=$User.DamageScale, B:NiagaraFloat=$User.Threshold.Engine.Stage2, ComparisonType:ENiagaraBooleanLogicOps=::NewEnumerator2)
     True Float=&MakeCustomFloatFromBool(
      Bool=&SetBoolByFloatComparison(
       A:NiagaraFloat=$User.DamageScale
       B:NiagaraFloat=$User.Threshold.Engine.Stage3Critical
       ComparisonType:ENiagaraBooleanLogicOps=::NewEnumerator2
      )
      True Float=10
      False Float=5
     )
     False Float=0
    )
    Use Spawn Probability=true
   ) {module=SpawnRate001}
   &SpawnRate(
    SpawnGroup=2
    SpawnRate=&MakeCustomFloatFromBool(
     Bool=&SetBoolByFloatComparison(A:NiagaraFloat=$User.DamageScale, B:NiagaraFloat=$User.Threshold.Rotor.Stage2, ComparisonType:ENiagaraBooleanLogicOps=::NewEnumerator2)
     True Float=&MakeCustomFloatFromBool(
      Bool=&SetBoolByFloatComparison(
       A:NiagaraFloat=$User.DamageScale
       B:NiagaraFloat=$User.Threshold.Rotor.Stage3Critical
       ComparisonType:ENiagaraBooleanLogicOps=::NewEnumerator2
      )
      True Float=20
      False Float=10
     )
     False Float=0
    )
    Use Spawn Probability=false
   ) {module=SpawnRate002}
   &SpawnRate(
    SpawnGroup=2
    SpawnRate=&MakeCustomFloatFromBool(
     Bool=&SetBoolByFloatComparison(A:NiagaraFloat=$User.DamageScale, B:NiagaraFloat=$User.Threshold.Rotor.Stage2, ComparisonType:ENiagaraBooleanLogicOps=::NewEnumerator2)
     True Float=&MakeCustomFloatFromBool(
      Bool=&SetBoolByFloatComparison(
       A:NiagaraFloat=$User.DamageScale
       B:NiagaraFloat=$User.Threshold.Rotor.Stage3Critical
       ComparisonType:ENiagaraBooleanLogicOps=::NewEnumerator2
      )
      True Float=20
      False Float=10
     )
     False Float=0
    )
    Use Spawn Probability=false
   ) {module=SpawnRate006}
   &SpawnRate(
    Spawn Probability=&MakeCustomFloatFromBool(Bool=$User.isMobile, True Float=1, False Float=0)
    SpawnGroup=3
    SpawnRate=&MakeCustomFloatFromBool(
     Bool=&SetBoolByFloatComparison(A:NiagaraFloat=$User.DamageScale, B:NiagaraFloat=$User.Threshold.Tail.Stage2, ComparisonType:ENiagaraBooleanLogicOps=::NewEnumerator0)
     True Float=&MakeCustomFloatFromBool(
      Bool=&SetBoolByFloatComparison(
       A:NiagaraFloat=$User.DamageScale
       B:NiagaraFloat=$User.Threshold.Tail.Stage3Critical
       ComparisonType:ENiagaraBooleanLogicOps=::NewEnumerator0
      )
      True Float=20
      False Float=10
     )
     False Float=0
    )
    Use Spawn Probability=true
   ) {module=SpawnRate003}
   &SpawnRate(
    Spawn Probability=&MakeCustomFloatFromBool(Bool=$User.isMobile, True Float=1, False Float=0)
    SpawnGroup=4
    SpawnRate=&MakeCustomFloatFromBool(
     Bool=&SetBoolByFloatComparison(A:NiagaraFloat=$User.DamageScale, B:NiagaraFloat=$User.Threshold.Booster.Stage2, ComparisonType:ENiagaraBooleanLogicOps=::NewEnumerator2)
     True Float=&MakeCustomFloatFromBool(
      Bool=&SetBoolByFloatComparison(
       A:NiagaraFloat=$User.DamageScale
       B:NiagaraFloat=$User.Threshold.Booster.Stage3Critical
       ComparisonType:ENiagaraBooleanLogicOps=::NewEnumerator2
      )
      True Float=20
      False Float=10
     )
     False Float=0
    )
    Use Spawn Probability=true
   ) {module=SpawnRate004}
   &SpawnRate(
    Spawn Probability=&MakeCustomFloatFromBool(Bool=$User.isMobile, True Float=1, False Float=0)
    SpawnGroup=5
    SpawnRate=&MakeCustomFloatFromBool(
     Bool=&SetBoolByFloatComparison(A:NiagaraFloat=$User.DamageScale, B:NiagaraFloat=$User.Threshold.Forward.Stage2, ComparisonType:ENiagaraBooleanLogicOps=::NewEnumerator2)
     True Float=&MakeCustomFloatFromBool(
      Bool=&SetBoolByFloatComparison(
       A:NiagaraFloat=$User.DamageScale
       B:NiagaraFloat=$User.Threshold.Forward.Stage3Critical
       ComparisonType:ENiagaraBooleanLogicOps=::NewEnumerator2
      )
      True Float=10
      False Float=5
     )
     False Float=0
    )
    Use Spawn Probability=true
   ) {module=SpawnRate005}
  ]
  particleSpawnScript [
   /SetVariables(Particles.SpawnGroup=&SpawnGroup()) {isSetParameters=true}
   &FN_SelectHoagiePartInfo(
    BoostPartInfo:HoagieFXPartInfo=$System.Booster
    EnginePartInfo_L:HoagieFXPartInfo=$System.EnginePartInfo
    EnginePartInfo_R:HoagieFXPartInfo=$System.EngineR
    ForwardPartInfo:HoagieFXPartInfo=$System.Forward
    RotorPartInfo:HoagieFXPartInfo=$System.Rotor
    TailPartInfo:HoagieFXPartInfo=$System.Tail
   ) {module=SelectHoagiePartInfo}
   /SetVariables(Particles.SpawnGroup=&MergeHelper_IntSwitch(
     0=&SpawnGroup()
     1=&SelectIntFromArray(
      Array Sampling Mode=::NewEnumerator1
      Direct Array Index=?
      Int Selection Array=DataInterface() {propertyValues {IntData [0,1]}}
      Random Seed=0
     )
     2=&SelectIntFromArray(
      Array Sampling Mode=::NewEnumerator1
      Direct Array Index=?
      Int Selection Array=DataInterface() {propertyValues {IntData [2,8]}}
      Random Seed=0
     )
     3=&SpawnGroup()
     4=&SpawnGroup()
     5=&SpawnGroup()
     6=&SpawnGroup()
     Index=&SpawnGroup()
     Num=::NewEnumerator5
    )) {isSetParameters=true}
   &InitializeParticle(
    Color=<rgba 1 1 1 0.7>
    Lifetime=&UniformRangedFloat(Minimum=0.45, Maximum=0.9)
    Module.Write Color=true
    Module.Write Lifetime=true
    Module.Write Mass=true
    Module.Write Material Random=true
    Module.Write Position=false
    Module.Write Scale=false
    Module.Write SpriteRotation=true
    Module.Write SpriteSize=true
    Position=&SimulationPosition()
    Sprite Rotation=&NormalizedAngleToDegrees(NormalizedAngle=&UniformRangedFloat(Minimum=0, Maximum=1))
    Sprite Size=&Vector2DFromFloat(Value=/Hoagie/FX/DynamicInputs/FN_SelectHoagieFloatbySpawnGroup(
      "(0)_Engine_L":NiagaraFloat=&UniformRangedFloat(Minimum=150, Maximum=170)
      "(1)_Engine_R":NiagaraFloat=&UniformRangedFloat(Minimum=150, Maximum=175)
      "(2)_Rotor":NiagaraFloat=&UniformRangedFloat(Minimum=150, Maximum=170)
      "(3)_Tail":NiagaraFloat=&UniformRangedFloat(Minimum=125, Maximum=175)
      "(4)_BoosterEngine":NiagaraFloat=&UniformRangedFloat(Minimum=100, Maximum=175)
      "(5)_Forward":NiagaraFloat=&UniformRangedFloat(Minimum=150, Maximum=170)
     ))
   )
   &SampleSkeletalMeshSkeleton(
    Mesh:NiagaraDataInterfaceSkeletalMesh=$User.SkeletalMesh
    Spawn Mode:ENiagaraSkeletalSpawning=::NewEnumerator1
    Specific Socket Index:NiagaraInt32=$Particles.SpawnGroup
   )
   &SkeletalMeshSkeletonLocation(
    Module.Inherit Velocity=false
    Module.Use Sampled Bone as Ribbon ID=true
    Sampled Position Offset:Vector3f=&Add_Vector(
     A=&RotateVectorByQuaternion(Vector To Rotate=$Transient.HoagiePosition, Quaternion=$Particles.SampleSkeletalMeshSkeleton.SampledRotation)
     B=$Transient.HoagieOffset
    )
   ) {enabled=false}
   &SkeletalMeshLocation(
    Bone / Socket Index=&SpawnGroup()
    MaskSpawnGroup=false
    Mesh Sampling Type=::NewEnumerator5
    Sampled Position Offset=&Add_Vector(A=$Transient.HoagiePosition, B=$Transient.HoagieOffset)
    Skeletal Mesh=$User.SkeletalMesh
    Socket Sampling Mode=::NewEnumerator5
   )
   &AddVelocity(Velocity=&UniformRangedVector(Minimum=$Transient.HoagieVelocityMin, Maximum=$Transient.HoagieVelocityMax)) {module=AddVelocity001}
   &AddVelocity(Velocity=&Multiply_Vector(
     A=$Transient.HoagieNormal
     B=&VectorFromFloat(Value=&UniformRangedFloat(Minimum=25, Maximum=50))
    )) {module=AddVelocity002}
   &CameraOffset(Camera Offset Amount=25)
  ]
  particleUpdateScript [
   &ParticleState()
   &AccelerationForce(Acceleration=&FN_SelectHoagieVectorbySpawnGroup(
     "(0)_Engine_L":Vector3f=<0 0 500>
     "(1)_Engine_R":Vector3f=<0 0 500>
     "(2)_Rotor":Vector3f=<0 0 100>
     "(3)_Tail":Vector3f=<0 0 200>
     "(4)_BoosterEngine":Vector3f=<0 0 200>
     "(5)_Forward":Vector3f=<0 0 500>
    ))
   &CurlNoiseForce(
    Noise Frequency=0.5
    Noise Strength=&Lerp_Float(
     A=0
     B=5000
     Alpha=&Power(
      a=$Particles.NormalizedAge
      b=&FN_SelectHoagieFloatbySpawnGroup(
       "(0)_Engine_L":NiagaraFloat=0.25
       "(1)_Engine_R":NiagaraFloat=0.25
       "(2)_Rotor":NiagaraFloat=0.25
       "(3)_Tail":NiagaraFloat=2
       "(4)_BoosterEngine":NiagaraFloat=0.5
       "(5)_Forward":NiagaraFloat=0.25
      )
     )
    )
    Pan Noise=true
    Pan Noise Field=<0 0 100>
   )
   &Drag(Drag=5, Use Linear Drag=true, Use Rotational Drag=false)
   &SolveForcesAndVelocity(Clamp Velocity=false, Limit Acceleration=false)
   /SetVariables() {isSetParameters=true}
   &ScaleColor(
    Scale Alpha=&FloatFromCurve(FloatCurve=DataInterface() {propertyValues {Curve {keys [
         {arriveTangent=7.2485, interpMode=RCIM_Cubic, leaveTangent=7.2485, tangentMode=RCTM_User, time=0, value=0}
         {arriveTangent=-2.0134, interpMode=RCIM_Cubic, leaveTangent=-2.0134, tangentMode=RCTM_User, time=0.554, value=0.7238}
         {interpMode=RCIM_Cubic, tangentMode=RCTM_User, time=1, value=0}
        ]}}})
    Scale RGB=&VectorFromFloat(Value=&FloatFromCurve(FloatCurve=DataInterface() {propertyValues {Curve {keys [
          {arriveTangent=-156.3847, interpMode=RCIM_Cubic, leaveTangent=-101.9614, tangentMode=RCTM_User, time=0, value=5}
          {arriveTangent=-7.9923, interpMode=RCIM_Cubic, leaveTangent=-7.9923, tangentMode=RCTM_User, time=0.05, value=1}
          <at 0.5 0.015 RCIM_Cubic>
         ]}}}))
    ScaleA=true
    ScaleRGB=true
   )
   &ScaleColor(
    Scale RGB=&MakeVectorFromLinearColorRGB(Color=<rgba 0.0156 0.0156 0.0156 1>)
    ScaleA=false
    ScaleRGB=true
   ) {module=ScaleColor001}
   &ScaleSpriteSize(Scale Factor=&Vector2DFromFloat(Value=&FloatFromCurve(FloatCurve=DataInterface() {propertyValues {Curve {keys [{arriveTangent=4.0099, interpMode=RCIM_Cubic, leaveTangent=4.0099, tangentMode=RCTM_User, time=0, value=1}, <at 1 3 RCIM_Cubic>]}}})))
   &SpriteRotationRate(Rotation Rate=&UniformRangedFloat(Minimum=-1, Maximum=1))
  ]
  rendererClasses [&NiagaraSpriteRendererProperties]
  renderers [{rendererClass=&NiagaraSpriteRendererProperties}]
 }
}
summary {
 systemName=NS_TidalCraneHoagie_DamageState_Optimized
 emitters [
  {
   emitterName=Critical_Lights
   rendererClasses [&NiagaraSpriteRendererProperties, &NiagaraLightRendererProperties]
  }
  {emitterName=Local_Smoke_Optimized, rendererClasses [&NiagaraSpriteRendererProperties]}
  {emitterName=Local_Embers_Optimized, rendererClasses [&NiagaraSpriteRendererProperties]}
  {emitterName=World_Smoke_Optimized, rendererClasses [&NiagaraSpriteRendererProperties]}
  {emitterName=Local_Flames_Optimized, rendererClasses [&NiagaraSpriteRendererProperties]}
  {
   emitterName=Generator_Flare_Optimized
   rendererClasses [&NiagaraSpriteRendererProperties, &NiagaraLightRendererProperties]
  }
  {emitterName=World_BubbleSmoke_Optimized, rendererClasses [&NiagaraSpriteRendererProperties]}
 ]
 userVariables [
  User.DamageScale {type=NiagaraFloat, defaultValue=1, description=""}
  User.isCritical {type=NiagaraBool}
  User.isMobile {type=NiagaraBool}
  User.SkeletalMesh {
   type=NiagaraDataInterfaceSkeletalMesh
   defaultValue {
    struct=NiagaraExt_VariableValue_DataInterface
    value {dataInterface=&self:NiagaraDataInterfaceSkeletalMesh_0}
   }
  }
  User.Threshold.Booster.Stage1 {type=NiagaraFloat, defaultValue=0.33, description=""}
  User.Threshold.Booster.Stage2 {type=NiagaraFloat, defaultValue=0.66, description=""}
  User.Threshold.Booster.Stage3Critical {type=NiagaraFloat, defaultValue=1, description=""}
  User.Threshold.Engine.Stage1 {type=NiagaraFloat, defaultValue=0.33, description=""}
  User.Threshold.Engine.Stage2 {type=NiagaraFloat, defaultValue=0.66, description=""}
  User.Threshold.Engine.Stage3Critical {type=NiagaraFloat, defaultValue=1, description=""}
  User.Threshold.Forward.Stage1 {type=NiagaraFloat, defaultValue=0.33, description=""}
  User.Threshold.Forward.Stage2 {type=NiagaraFloat, defaultValue=0.66, description=""}
  User.Threshold.Forward.Stage3Critical {type=NiagaraFloat, defaultValue=1, description=""}
  User.Threshold.Rotor.Stage1 {type=NiagaraFloat, defaultValue=0.33, description=""}
  User.Threshold.Rotor.Stage2 {type=NiagaraFloat, defaultValue=0.66, description=""}
  User.Threshold.Rotor.Stage3Critical {type=NiagaraFloat, defaultValue=1, description=""}
  User.Threshold.Tail.Stage1 {type=NiagaraFloat, defaultValue=0.33, description=""}
  User.Threshold.Tail.Stage2 {type=NiagaraFloat, defaultValue=0.66, description=""}
  User.Threshold.Tail.Stage3Critical {type=NiagaraFloat, defaultValue=1, description=""}
 ]
}
