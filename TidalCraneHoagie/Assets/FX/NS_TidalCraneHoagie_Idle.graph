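# Niagara system for the TidalCraneHoagie rotor aircraft's ambient effects: thruster and rotor flames with lights, rotor motion blur, smoke, GPU sparks and speed lines.
# Rotor, flame and exhaust emitters follow skeletal mesh sockets or bones, driven by rotor speed and smoothed forward velocity.
# A system Timeline scales thruster intensity, smoke and spark rates, and a local-only component-renderer (PP) effect.
# User speed curves drive local-only speed lines; camera distance fades nearby exhaust.
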NiagaraSystem /TidalCraneHoagie/Assets/FX/NS_TidalCraneHoagie_Idle
dependencies {
 usedDataInterfaces [
  &NiagaraDataInterfaceArrayInt32
  &NiagaraDataInterfaceArrayFloat3
  &NiagaraDataInterfaceCurve
  &NiagaraDataInterfaceColorCurve
  &NiagaraDataInterfaceCamera
  &NiagaraDataInterfaceVector2DCurve
 ]
 usedDynamicInputs []
 usedModules [
  &self:SystemSpawnScript.SystemScriptSource.SystemScriptGraph.NiagaraNodeAssignment_0.SetVariables_#1
  &SystemState
  &Timeline
  &self:SystemSpawnScript.SystemScriptSource.SystemScriptGraph.NiagaraNodeAssignment_1.SetVariables_#2
 ]
}
emitters {
 FN_Sprite_SpeedLines001 {
  emitterUpdateScript [
   &SpawnBurst_Instantaneous(Spawn Count=5, Use Spawn Probability=false) {enabled=false}
   &EmitterState()
   &SpawnRate(
    SpawnRate=&Multiply_Float(
     A=&FloatFromCurve(
      FloatCurve=DataInterface() {propertyValues {Curve {keys [
          {arriveTangent=0.4214, interpMode=RCIM_Cubic, leaveTangent=0.4225, tangentMode=RCTM_User, time=25, value=0}
          {arriveTangent=17.7333, interpMode=RCIM_Cubic, leaveTangent=17.7333, tangentMode=RCTM_User, time=31.9282, value=72.4651}
          {arriveTangent=7.1685, interpMode=RCIM_Cubic, leaveTangent=7.1685, tangentMode=RCTM_User, time=39.6367, value=353.7616}
          {arriveTangent=16.7131, interpMode=RCIM_Cubic, leaveTangent=16.7131, tangentMode=RCTM_User, time=70, value=500}
          <at 90 800>
         ]}}}
      CurveIndex=$User.Speed
     )
     B=&MakeCustomFloatFromBool(Bool=$User.IsLocal, True Float=1, False Float=0)
    )
    Use Spawn Probability=false
   )
   &EmitterState(Enable Distance Culling=true, MaxDistance=2500, Module.bUseMaxDistance=true, Module.bUseMinDistance=false, Scalability Mode=::NewEnumerator1) {module=EmitterState001}
  ]
  particleSpawnScript [
   /Niagara/Modules/Spawn/Initialization/InitializeParticle(
    Color=&MakeLinearColor(
     R:NiagaraFloat=1
     G:NiagaraFloat=1
     B:NiagaraFloat=1
     A:NiagaraFloat=&UniformRangedFloat(Minimum=1, Maximum=0.4)
    )
    Lifetime=&UniformRangedFloat(Minimum=0.4, Maximum=0.5)
    Module.Write Color=true
    Module.Write Lifetime=true
    Module.Write Mass=true
    Module.Write Material Random=true
    Module.Write Position=false
    Module.Write Scale=false
    Module.Write SpriteRotation=false
    Module.Write SpriteSize=true
    Position=&SimulationPosition()
    Sprite Rotation=&UniformRangedFloat(Minimum=0, Maximum=360)
    Sprite Size=&Multiply_Vector2(
     A=<0.1 1>
     B=&Vector2DFromFloat(Value=&Multiply_Float(
       A=&FloatFromCurve(
        FloatCurve=DataInterface() {propertyValues {Curve {keys [
            {arriveTangent=0.4214, interpMode=RCIM_Cubic, leaveTangent=0.4225, tangentMode=RCTM_User, time=25, value=0}
            {arriveTangent=7.1685, interpMode=RCIM_Cubic, leaveTangent=7.1685, tangentMode=RCTM_User, time=30, value=50}
            {arriveTangent=7.5, interpMode=RCIM_Cubic, leaveTangent=7.5, tangentMode=RCTM_User, time=69.9874, value=151.101}
            {arriveTangent=10.2726, interpMode=RCIM_Cubic, leaveTangent=10.2726, tangentMode=RCTM_User, time=90.1617, value=288.3149}
           ]}}}
        CurveIndex=$User.Speed
       )
       B=1
      ))
    )
   )
   &SystemLocation(MaskSpawnGroup=false, Offset=<100 0 100>)
   /Niagara/Modules/Spawn/Location/SphereLocation(MaskSpawnGroup=false, Module.Override Randomness=false, Sphere Radius=1000, Surface Only=false)
   &TorusLocation(
    Align Torus With Axis=true
    Handle Radius=400
    Large Radius=1200
    MaskSpawnGroup=false
    Module.Override Randomness=false
    Offset=&Multiply_VectorByFloat(Vector=$Engine.Owner.SystemXAxis, Float=40)
    Torus Axis=<1 0 0>
   )
   &AddVelocity(Velocity=&MakeVector(X=&UniformRangedFloat(Minimum=-150, Maximum=-200), Y=0, Z=0))
   &InheritVelocity(
    Module.Limit Speed=true
    Velocity Limit=1000
    Velocity Scale=&Multiply_VectorByFloat(Vector=<-0.9 -0.9 -0.9>, Float=$User.Speed)
   )
   /SetVariables(Particles.Initial.SpriteSize:Vector2f=&Multiply_Vector2(
     A=<1 1>
     B=&Vector2DFromFloat(Value=&Multiply_Float(
       A=1
       B=&FloatFromCurve(
        FloatCurve=DataInterface() {propertyValues {Curve {keys [
            {arriveTangent=0.4214, interpMode=RCIM_Cubic, leaveTangent=0.4225, tangentMode=RCTM_User, time=25, value=0}
            {arriveTangent=12.969, interpMode=RCIM_Cubic, leaveTangent=12.969, tangentMode=RCTM_User, time=31.9201, value=25.8333}
            {arriveTangent=7.1685, interpMode=RCIM_Cubic, leaveTangent=7.1685, tangentMode=RCTM_User, time=41.1282, value=209.1666}
            {arriveTangent=16.7131, interpMode=RCIM_Cubic, leaveTangent=16.7131, tangentMode=RCTM_User, time=70, value=500}
            <at 90 1500>
           ]}}}
        CurveIndex=$User.Speed
       )
      ))
    )) {isSetParameters=true}
  ]
  particleUpdateScript [
   &ParticleState()
   &Color(
    Color=&ScaleLinearColorByCurve(
     InitialValue=<rgba 1 1 1 0.075>
     ScaleCurve=DataInterface() {propertyValues {
       AlphaCurve {keys [<at 0.0384 1>,<at 0.3631 1>]}
       BlueCurve {keys [<at 0.0281 1>]}
       GreenCurve {keys [<at 0.0281 1>]}
       RedCurve {keys [<at 0.0281 1>]}
      }}
     CurveIndex=$Particles.NormalizedAge
    )
    Scale Alpha=&FloatFromCurve(FloatCurve=DataInterface() {propertyValues {Curve {keys [<at 0 0 RCIM_Cubic>,<at 0.5 1 RCIM_Cubic>,<at 1 0 RCIM_Cubic>]}}})
   ) {enabled=false}
   &ScaleColor(
    Scale Alpha=&FloatFromCurve(FloatCurve=DataInterface() {propertyValues {Curve {keys [<at -0.0046 1.0113>,<at 1.0011 0.002>]}}}, Scale Curve=3)
    ScaleA=true
    ScaleRGB=true
   )
   &ScaleColor(
    Scale Alpha=&FloatFromCurve(FloatCurve=DataInterface() {propertyValues {Curve {keys [<at 0 0>,<at 1 1>]}}}, CurveIndex=$User.Speed, Scale Curve=0.1)
    ScaleA=true
    ScaleRGB=true
   ) {module=ScaleColor001}
   &SolveForcesAndVelocity(Clamp Velocity=false, Limit Acceleration=false)
   &ScaleSpriteSizeBySpeed(Max Scale Factor=<0.5 1>, Min Scale Factor=<0 0>)
   &ScaleSpriteSize(Scale Factor=&Vector2DFromCurve(
     Vector2Curve=DataInterface() {propertyValues {
       XCurve {keys [<at -0.0069 0.5 RCIM_Cubic>,<at 0.4918 1 RCIM_Cubic>,<at 1 0.5 RCIM_Cubic>]}
       YCurve {keys [<at 0 0.0245 RCIM_Cubic>, {arriveTangent=-0.0245, interpMode=RCIM_Cubic, leaveTangent=-0.0245, time=0.4975, value=1}, <at 1 0 RCIM_Cubic>]}
      }}
     ScaleCurve=<1 1>
    ))
   &CameraOffset(Camera Offset Amount=10)
  ]
  rendererClasses [&NiagaraSpriteRendererProperties]
  renderers [{rendererClass=&NiagaraSpriteRendererProperties}]
 }
 NE_RotorMotionBlur {
  emitterSpawnScript [/SetVariables() {isSetParameters=true}]
  emitterUpdateScript [
   /SetVariables(Emitter.CustomTime:NiagaraFloat=&Add_Float(A=$Emitter.CustomTime, B=&Lerp_Float(A=-0.04, B=-0.08, Alpha=$User.RotorSpeedAlpha))) {isSetParameters=true}
   &EmitterState(
    Inactive Response=::NewEnumerator1
    Life Cycle Mode=::NewEnumerator1
    Loop Behavior=::NewEnumerator1
    Loop Duration Mode=::NewEnumerator1
    UseLoopDelay=false
   )
   &SpawnBurst_Instantaneous(Use Spawn Probability=false, UseLoopCountLimit=false)
   &SpawnBurst_Instantaneous(Spawn Group=1, Use Spawn Probability=false, UseLoopCountLimit=false) {module=SpawnBurst_Instantaneous001}
  ]
  particleSpawnScript [/Niagara/Modules/Spawn/Initialization/V2/InitializeParticle(
    Color=<rgba 0.2 0.2 0.2 0.5>
    Mesh Scale Mode=::NewEnumerator3
    Mesh Uniform Scale=10
    Sprite Rotation Angle=&MaskFloatBySpawnGroup(Masked Float=0.05, Passthrough Float=0.16)
    Sprite Rotation Mode=::NewEnumerator1
    Sprite Size Mode=::NewEnumerator3
    Uniform Sprite Size=246.9375
    UsePositionOffset=true
    Write Material Random=true
   )]
  particleUpdateScript [
   &ParticleState(Kill Particles When Lifetime Has Elapsed=false)
   &ScaleColor(
    Scale Alpha=&Multiply_Float(
     A=&Multiply_Float(
      A=&FloatFromCurve(FloatCurve=DataInterface() {propertyValues {Curve {keys [<at 0 0 RCIM_Cubic>,<at 0.2 1 RCIM_Cubic>]}}})
      B=&FNTimeOfDayScalar(Night Scale=0.2, Morning Scale=0.3, Evening Scale=0.2)
     )
     B=&Divide_Float(A=$User.RotorSpeedAlpha, B=1)
    )
    Scale RGB=&VectorFromFloat(Value=1)
    ScaleA=true
    ScaleRGB=false
   )
   &SkeletalMeshLocation(
    Bone / Socket Index=$Particles.SpawnGroup
    Bone Sampling Mode=::NewEnumerator4
    MaskSpawnGroup=false
    Orientation Sampling=::NewEnumerator1
    Sampled Position Offset=<0 0 0.5>
    Skeletal Mesh=$User.SkeletalMesh
   )
   &SpriteFacingAndAlignment(
    Sprite Alignment=&RotateVectorByQuaternion(Vector To Rotate=<1 0 0>, Quaternion=$Particles.SkeletalMeshLocation.SampledRotation)
    Sprite Facing=&RotateVectorByQuaternion(Vector To Rotate=<0 0 1>, Quaternion=$Particles.SkeletalMeshLocation.SampledRotation)
    WriteAlignment=true
    WriteFacing=true
   )
   /SetVariables(Particles.SecondRotorRotation:NiagaraFloat=&Add_Float(
     A=$Particles.SpriteRotation
     B=&AngleConversion(Angle:NiagaraFloat=0.33, Angle Input:ENiagara_AngleInput=::NewEnumerator1, Angle Output:ENiagara_AngleInput=::NewEnumerator0)
    )) {isSetParameters=true}
   /SetVariables(Particles.ThirdRotorRotation:NiagaraFloat=&Add_Float(
     A=$Particles.SpriteRotation
     B=&AngleConversion(Angle:NiagaraFloat=0.66, Angle Input:ENiagara_AngleInput=::NewEnumerator1, Angle Output:ENiagara_AngleInput=::NewEnumerator0)
    )) {isSetParameters=true}
   /SetVariables(Particles.RotorBlurColor:LinearColor=&MultiplyLinearColor(Linear Color=$Particles.Color, Scale Factor=<1 1 1 1>)) {isSetParameters=true}
   &FN_Mesh_Attributes(
    Color=&MultiplyLinearColor(
     Linear Color=<rgba 0.3229 0.3229 0.3229 0.1>
     Scale Factor=&MakeVector4FromVectorAndFloat(
      "Vector (RGB)"=<1 1 1>
      "Float (Alpha)"=&Multiply_Float(
       A=$User.RotorSpeedAlpha
       B=&FN_Select_Float_from_Time_Of_Day(Morning=0.6, Day=1, Evening=0.6, Night=0.4)
      )
     )
    )
    Create Color Attribute=true
    Create Orientation Attribute=true
    Create Scale Attribute=true
    Orientation:Quat4f=&MultiplyQuaternion(
     Quaternion A=&OrientMeshToVector(
      Look At Direction=&RotateVectorByQuaternion(Vector To Rotate=<0 0 1>, Quaternion=$Particles.SkeletalMeshLocation.SampledRotation)
      Reference Axis Vector=<1 0 0>
     )
     Quaternion B=&MakeQuaternion(
      Quaternion From=0
      Axis=<0 1 0>
      Angle=&MaskFloatBySpawnGroup(Masked Float=0.25, Passthrough Float=-0.25)
      From=?
      To=?
      Angle Type=::NewEnumerator1
      Normalize Vectors=false
     )
    )
    Scale=&Multiply_Vector(
     A=&Multiply_VectorByFloat(Vector=<0.95 0.95 1>, Float=1)
     B=&Lerp_Vector(A=<1 1 1>, B=<1 1 1.2>, Alpha=$User.RotorSpeedAlpha)
    )
    Write to Module Namespace=true
   )
   &FN_Mesh_Attributes(
    Color=&MultiplyLinearColor(
     Linear Color=<rgba 0.3229 0.3229 0.3229 0>
     Scale Factor=&MakeVector4FromVectorAndFloat("Vector (RGB)"=<1 1 1>, "Float (Alpha)"=$User.RotorSpeedAlpha)
    )
    Create Color Attribute=true
    Create Scale Attribute=true
    Scale=&Multiply_Vector(
     A=&Multiply_VectorByFloat(Vector=<0.95 0.95 0.7>, Float=1)
     B=&Lerp_Vector(A=<1 1 1>, B=<1 1 1>, Alpha=$User.RotorSpeedAlpha)
    )
    Write to Module Namespace=true
   ) {module=FN_Mesh_Attributes001}
   &DynamicMaterialParameters(
    Index 2 Param 1=&Lerp_Float(A=1, B=0.985, Alpha=$User.RotorSpeedAlpha)
    Index 3 Param 1=$Emitter.CustomTime
    Module.Index2Float0Write=true
    Module.Index2Float1Write=true
    Module.Index2Float2Write=true
    Module.Index2Float3Write=true
    Module.Index3Float0Write=true
    Module.Index3Float1Write=true
    Module.Index3Float2Write=true
    Module.Index3Float3Write=true
    Write Parameter Index 0=::NewEnumerator0
    Write Parameter Index 2=::NewEnumerator1
    Write Parameter Index 3=::NewEnumerator1
   )
  ]
  rendererClasses [&NiagaraSpriteRendererProperties, &NiagaraMeshRendererProperties]
  renderers [
   {rendererClass=&NiagaraSpriteRendererProperties}
   {rendererClass=&NiagaraSpriteRendererProperties, rendererIndex=1}
   {rendererClass=&NiagaraSpriteRendererProperties, rendererIndex=2}
   {rendererClass=&NiagaraSpriteRendererProperties, rendererIndex=3}
   {rendererClass=&NiagaraMeshRendererProperties, rendererIndex=4}
   {rendererClass=&NiagaraMeshRendererProperties, rendererIndex=5}
  ]
 }
 PP {
  emitterUpdateScript [
   &EmitterState(Life Cycle Mode=::NewEnumerator1, Loop Behavior=::NewEnumerator1, Loop Duration Mode=::NewEnumerator1, UseLoopDelay=false)
   &SpawnBurst_Instantaneous(Use Spawn Probability=true, UseLoopCountLimit=false)
  ]
  particleSpawnScript [&InitializeParticle(Color Mode=::NewEnumerator0, Sprite Size Mode=::NewEnumerator3, Uniform Sprite Size=200, UsePositionOffset=false, Write Material Random=true)]
  particleUpdateScript [
   &ParticleState(Kill Particles When Lifetime Has Elapsed=false, Let Infinitely Lived Particles Die When Emitter Deactivates=true)
   /SetVariables(Particles.SpeedNormalizedB:NiagaraFloat=&Multiply_Float(A=0.3, B=$System.Timeline.NormalizedTime)) {isSetParameters=true}
   /SetVariables(Particles.ComponentsEnabled:NiagaraBool=&Boolean_AndOperation(Bool1=&SetBoolByFloatComparison(A=$System.Timeline.NormalizedTime, B=0.05), Bool2=$User.IsLocal)) {isSetParameters=true}
  ]
  rendererClasses [&NiagaraComponentRendererProperties]
  renderers [{rendererClass=&NiagaraComponentRendererProperties, rendererIndex=0}]
 }
 RotorFlames {
  emitterSpawnScript [/SetVariables() {isSetParameters=true}]
  emitterUpdateScript [
   &EmitterState(
    Inactive Response=::NewEnumerator1
    Life Cycle Mode=::NewEnumerator1
    Loop Behavior=::NewEnumerator1
    Loop Duration Mode=::NewEnumerator1
    UseLoopDelay=false
   )
   &SpawnBurst_Instantaneous(Spawn Count=2, Use Spawn Probability=false)
   /SetVariables(Emitter.SpeedAlpha:NiagaraFloat=&ClampFloat(Float=&Divide_Float(A=&Subtract_Float(A=$User.Speed, B=5), B=50))) {isSetParameters=true}
   /SetVariables(Emitter.SmoothVelocity:Vector3f=/Niagara/DynamicInputs/Math/Clamp_Vector(
     Vector:Vector3f=&Lerp_Vector(
      A=$Emitter.SmoothVelocity
      B=&TransformVector(Vector=$Engine.Owner.Velocity, Source Space=::World, Destination Space=::Local)
      Alpha=0.06
     )
     Min:Vector3f=<0 0 0>
     Max:Vector3f=<1000 1000 1000>
    )) {isSetParameters=true}
   /SetVariables(Emitter.ForwardDot:NiagaraFloat=&ClampFloat(Float=&Divide_Float(
      A=&Subtract_Float(A=&MakeFloatFromVector(Vector=$Emitter.SmoothVelocity, Channel=::NewEnumerator0), B=50)
      B=500
     ))) {isSetParameters=true}
  ]
  particleSpawnScript [
   &InitializeParticle(
    Lifetime=0.2
    Mesh Scale=<0.5 2 1>
    Mesh Scale Mode=::NewEnumerator4
    Sprite Size Mode=::NewEnumerator3
    Uniform Sprite Size=550
    UsePositionOffset=false
    Write Material Random=true
   ) {module=InitializeParticle001}
   &InitialMeshOrientation(Mesh Orientation Mode=::NewEnumerator7, Rotation=<0 0 0.5>, Use Rotation Vector=true) {module=InitialMeshOrientation001}
  ]
  particleUpdateScript [
   &ParticleState(Kill Particles When Lifetime Has Elapsed=false)
   /SetVariables(Particles.Oscillation:NiagaraFloat=&Add_Float(
     A=&Waveform(
      Waveform Count=::NewEnumerator1
      Global Phase=$Particles.Age
      "Waveform [1]"=::NewEnumerator0
      "[1] Frequency"=2.256
      "[1] Amplitude Min/Max"=<0 1>
      "[1] Scale"=0.5
      "[1] Bias"=1
      "Waveform [2] Blend"=::NewEnumerator2
      "[2] Input Position"=0
      "[2] Phase"=$Emitter.Age
      "[2] Phase Scale"=1
      "[2] Compound Frequency"=2
      "[2] Frequency"=7.2563
      "[2] Amplitude Min/Max"=<0 1>
      "[2] Scale"=0.5
      "[2] Bias"=1
      "[2] Exponent"=1
      "[2] Pulse Duration"=0.1
      "[2] Clip Triangle Min/Max"=<0 0>
      "[2] Random Seed"=0
     )
     B=0
    )) {isSetParameters=true}
   &SkeletalMeshLocation(
    Bone / Socket Index=&SelectIntFromArray(Direct Array Index=&ReturnExecIndex(), Int Selection Array=DataInterface() {propertyValues {IntData [2,3]}})
    MaskSpawnGroup=false
    Mesh Sampling Type=::NewEnumerator5
    Sampled Position Offset=&Multiply_Vector(
     A=&SelectVectorFromArray(
      Direct Array Index=&ReturnExecIndex()
      Vector Selection Array=DataInterface() {propertyValues {FloatData [<0 -3 0>, <0 -1 0>]}}
     )
     B=<1 1 1>
    )
    Skeletal Mesh=$User.SkeletalMesh
    Socket Sampling Mode=::NewEnumerator5
   ) {module=SkeletalMeshLocation001}
   &ScaleMeshSize(Scale Factor=&Multiply_Vector(
     A=<8 2.1 0.25>
     B=&Multiply_Vector(
      A=&Lerp_Vector(A=<0 1 0>, B=<1 1 1>, Alpha=$Emitter.ForwardDot)
      B=&Lerp_Vector(A=<1 1 1>, B=<3 1 1>, Alpha=$System.Timeline.NormalizedTime)
     )
    ))
   &SolveForcesAndVelocity(Clamp Velocity=false, Limit Acceleration=false)
   &DynamicMaterialParameters(
    Index 0 Param 1=&RandomRangeFloat(Minimum=0, Maximum=1)
    Index 0 Param 2=5
    Index 0 Param 3=5
    Module.Param0WriteEnabled=true
    Module.Param0WriteEnabled3=true
    Module.Param1WriteEnabled=true
    Module.Param1WriteEnabled3=true
    Module.Param2WriteEnabled=true
    Module.Param2WriteEnabled3=true
    Module.Param3WriteEnabled=true
    Module.Param3WriteEnabled3=true
    Write Parameter Index 3=true
   )
   &ScaleColor(
    Scale RGB=&MakeVectorFromLinearColorRGB(Color=&Multiply_LinearColorByFloat(
      Linear Color=&ColorFromCurve(
       ColorCurve=DataInterface() {propertyValues {
         AlphaCurve {keys []}
         BlueCurve {keys [<at 0 1 RCIM_Cubic>,<at 1 3 RCIM_Cubic>]}
         GreenCurve {keys [<at 0 0.3134 RCIM_Cubic>,<at 1 2.3639 RCIM_Cubic>]}
         RedCurve {keys [<at 0 0.1088 RCIM_Cubic>,<at 1 0.2188 RCIM_Cubic>]}
        }}
       CurveIndex=$Emitter.ForwardDot
      )
      Float=1
     ))
    ScaleA=true
    ScaleRGB=true
   )
   &DynamicMaterialParameters(
    Index 3 Param 1=$Emitter.ForwardDot
    Module.Index0Float0Write=true
    Module.Index0Float1Write=true
    Module.Index0Float2Write=true
    Module.Index0Float3Write=true
    Module.Index3Float0Write=true
    Module.Index3Float1Write=true
    Module.Index3Float2Write=true
    Module.Index3Float3Write=true
    Write Parameter Index 3=::NewEnumerator1
   ) {module=DynamicMaterialParameters001}
   &FN_Light_v2(
    Light Exponent=1
    Light Radius:NiagaraFloat=&Multiply_Float(A=6, B=$User.BoostReady)
    LightEnabled:NiagaraBool=&SetBoolByIntComparison(A=&ModuloInt(A=&ReturnExecIndex(), B=3), B=1, Comparison Type=::NewEnumerator4)
    Scale Color=1
   )
   &FN_Sprite_Attributes(
    Alignment:Vector3f=&RotateVectorByQuaternion(Vector To Rotate=<0 0 1>, Quaternion=$Particles.SkeletalMeshLocation001.SampledRotation)
    Color=&MultiplyLinearColor(Linear Color=$Particles.Color, Scale Factor=&MakeVector4_FromLinearColor(LinearColor=<rgba 5 5 5 1>))
    Create Alignment Attribute=true
    Create Camera Offset Attribute=true
    Create Color Attribute=true
    Create Facing Attribute=true
    Create Position Attribute=true
    Create SpriteSize Attribute=true
    Facing:Vector3f=&RotateVectorByQuaternion(Vector To Rotate=<-1 0.2 0>, Quaternion=$Particles.SkeletalMeshLocation001.SampledRotation)
    Position:NiagaraPosition=&AddVectorToPosition(
     Position=$Particles.Position
     Vector=&RotateVectorByQuaternion(Vector To Rotate=<-15 0 0>, Quaternion=$Particles.SkeletalMeshLocation001.SampledRotation)
    )
    SpriteSize=&Multiply_Vector2D_ByFloat(Float=$Emitter.ForwardDot, Vector2D=<90 15>)
    Write to Module Namespace=true
   )
  ]
  rendererClasses [
   &NiagaraMeshRendererProperties
   &NiagaraSpriteRendererProperties
   &NiagaraLightRendererProperties
  ]
  renderers [
   {rendererClass=&NiagaraMeshRendererProperties, rendererIndex=0}
   {rendererClass=&NiagaraSpriteRendererProperties, rendererIndex=1}
   {rendererClass=&NiagaraLightRendererProperties, rendererIndex=2}
  ]
 }
 Smoke {
  emitterUpdateScript [
   &EmitterState()
   &SpawnRate(SpawnRate=&Multiply_Float(A=60, B=$System.Timeline.NormalizedTime), Use Spawn Probability=false)
  ]
  particleSpawnScript [
   &InitializeParticle(
    Color=<rgba 0.2 0.2 0.2 0.3>
    Lifetime=&UniformRangedFloat(Minimum=0.6, Maximum=1)
    Sprite Rotation Mode=::NewEnumerator2
    Sprite Size Mode=::NewEnumerator1
    Uniform Sprite Size Max=200
    Uniform Sprite Size Min=100
    UsePositionOffset=false
    Write Material Random=true
   ) {module=InitializeParticle001}
   &PartitionParticles(Particles Per Partition=3, Partition Mode:ENiagara_PartitionMode_v2=::NewEnumerator1)
   &SkeletalMeshLocation(
    Bone / Socket Index=$Particles.PartitionParticles.Partition
    MaskSpawnGroup=false
    Mesh Sampling Type=::NewEnumerator5
    Sampled Position Offset=&Multiply_Vector(
     A=&SelectVectorFromArray(
      Direct Array Index=$Particles.PartitionParticles.Index In Partition
      Vector Selection Array=DataInterface() {propertyValues {FloatData [<-10 6 -4>, <-20 31 -4>, <-30 55 -4>]}}
     )
     B=&MaskVector(Mask Source=1, Mask Parameter=$Particles.PartitionParticles.Partition, Mask=0, Masked Vector=<1 1 1>, Passthrough Vector=<1 -1 1>)
    )
    Skeletal Mesh=$User.SkeletalMesh
    Socket Sampling Mode=::NewEnumerator5
   )
   &AddVelocity(
    Cone Angle=10
    Cone Axis=<-1 0 0>
    Module.UseConeFalloff=false
    Module.UseVelDistribution=false
    Override Seed=false
    Velocity Mode=::NewEnumerator2
    Velocity Speed=&RandomRangeFloat(Minimum=800, Maximum=1200)
   ) {module=AddVelocity001}
  ]
  particleUpdateScript [
   &ParticleState()
   /SetVariables(Particles.DistanceScale:NiagaraFloat=&ClampFloat(Float=&Divide_Float(
      A=&Subtract_Float(
       A=&VectorLength(Vector=&Subtract_Position(
         A=$Particles.Position
         B=&ReturnCameraPosition(Coordinate Space:ENiagaraCoordinateSpace=::Simulation, Camera Query:NiagaraDataInterfaceCamera=DataInterface())
        ))
       B=150
      )
      B=150
     ))) {isSetParameters=true}
   &AccelerationForce(Acceleration=<0 0 50>)
   &Drag(Drag=3, Use Linear Drag=true, Use Rotational Drag=false)
   &SolveForcesAndVelocity(Clamp Velocity=false, Limit Acceleration=false)
   /Niagara/Modules/Update/SubUV/SubUVAnimation(Number Of Frames=64)
   &DynamicMaterialParameters(
    Index 0 Param 1=&Lerp_Float(
     A=1
     B=-1
     Alpha=&FloatFromCurve(FloatCurve=DataInterface() {propertyValues {Curve {keys [<at 0 0>,<at 1 1>]}}})
    )
    Index 0 Param 2=1
    Index 0 Param 3=1
    Index 0 Param 4=1
    Module.Param0WriteEnabled=true
    Module.Param1WriteEnabled=true
    Module.Param2WriteEnabled=true
    Module.Param3WriteEnabled=true
   )
   &ScaleSpriteSize(Scale Factor=&Vector2DFromFloat(Value=&FloatFromCurve(FloatCurve=DataInterface() {propertyValues {Curve {keys [<at 0 0.7 RCIM_Cubic>,<at 1 2 RCIM_Cubic>]}}})))
   &ScaleColor(
    Scale Alpha=$Particles.DistanceScale
    Scale RGB=&MakeVectorFromLinearColorRGB(Color=&ColorFromCurve(ColorCurve=DataInterface() {propertyValues {
        AlphaCurve {keys [<at 0 1 RCIM_Cubic>,<at 1 1 RCIM_Cubic>]}
        BlueCurve {keys [<at 0 7 RCIM_Cubic>,<at 0.3634 1 RCIM_Cubic>]}
        GreenCurve {keys [<at 0 4.2411 RCIM_Cubic>,<at 0.3634 1 RCIM_Cubic>]}
        RedCurve {keys [<at 0 0 RCIM_Cubic>,<at 0.3634 1 RCIM_Cubic>]}
       }}))
    ScaleA=true
    ScaleRGB=true
   ) {module=ScaleColor001}
  ]
  rendererClasses [&NiagaraSpriteRendererProperties]
  renderers [{rendererClass=&NiagaraSpriteRendererProperties}]
 }
 Sparks {
  simTarget=GPUComputeSim
  emitterUpdateScript [
   &EmitterState()
   &SpawnRate(SpawnRate=&Multiply_Float(A=60, B=$System.Timeline.NormalizedTime), Use Spawn Probability=false)
   &SpawnBurst_Instantaneous(Spawn Count=&UniformRangedInt(Minimum=10, Maximum=15), Use Spawn Probability=false) {enabled=false}
  ]
  particleSpawnScript [
   &InitializeParticle(
    Color=<rgba 0.0313 2.6218 5 1>
    Lifetime=&UniformRangedFloat(Minimum=0.1, Maximum=0.35)
    Sprite Size Max=<4 30>
    Sprite Size Min=<3 20>
    Sprite Size Mode=::NewEnumerator2
    UsePositionOffset=false
    Write Material Random=true
   ) {module=InitializeParticle001}
   &PartitionParticles(Particles Per Partition=3, Partition Mode:ENiagara_PartitionMode_v2=::NewEnumerator1)
   &SkeletalMeshLocation(
    Bone / Socket Index=$Particles.PartitionParticles.Partition
    MaskSpawnGroup=false
    Mesh Sampling Type=::NewEnumerator5
    Sampled Position Offset=&Multiply_Vector(
     A=&SelectVectorFromArray(
      Direct Array Index=$Particles.PartitionParticles.Index In Partition
      Vector Selection Array=DataInterface() {propertyValues {FloatData [<-10 6 -4>, <-20 31 -4>, <-30 55 -4>]}}
     )
     B=&MaskVector(Mask Source=1, Mask Parameter=$Particles.PartitionParticles.Partition, Mask=0, Masked Vector=<1 1 1>, Passthrough Vector=<1 -1 1>)
    )
    Skeletal Mesh=$User.SkeletalMesh
    Socket Sampling Mode=::NewEnumerator5
   )
   &SphereLocation(
    Hemisphere X=true
    MaskSpawnGroup=false
    Module.Override Randomness=false
    Non Uniform Scale=<-1 1 1>
    Sphere Radius=&UniformRangedFloat(Minimum=10, Maximum=20)
    Surface Only=false
   )
   &AddVelocityFromPoint(
    Module.Use Falloff Distance=false
    Velocity Origin:Vector3f=&SimulationPosition()
    Velocity Strength=&UniformRangedFloat(Minimum=15, Maximum=60)
   )
   &AddVelocity(
    Cone Angle=10
    Cone Axis=<-1 0 0>
    Module.UseConeFalloff=false
    Module.UseVelDistribution=false
    Override Seed=false
    Velocity Mode=::NewEnumerator2
    Velocity Speed=&RandomRangeFloat(Minimum=800, Maximum=1600)
   ) {module=AddVelocity001}
  ]
  particleUpdateScript [
   &ParticleState()
   /SetVariables(Particles.DistanceScale:NiagaraFloat=&ClampFloat(Float=&Divide_Float(
      A=&Subtract_Float(
       A=&VectorLength(Vector=&Subtract_Position(
         A=$Particles.Position
         B=&ReturnCameraPosition(Coordinate Space:ENiagaraCoordinateSpace=::Simulation, Camera Query:NiagaraDataInterfaceCamera=DataInterface())
        ))
       B=150
      )
      B=150
     ))) {isSetParameters=true}
   &AccelerationForce(Acceleration=<0 0 -250>)
   &CurlNoiseForce(Bias Noise=false, Noise Strength=550, Pan Noise=false)
   &SolveForcesAndVelocity(Clamp Velocity=false, Limit Acceleration=false)
   &ScaleColor(
    Scale Alpha=&FloatFromCurve(FloatCurve=DataInterface() {propertyValues {Curve {keys [{arriveTangent=-7.383, interpMode=RCIM_Cubic, leaveTangent=-7.383, tangentMode=RCTM_User, time=0, value=2}, <at 0.2 1 RCIM_Cubic>]}}})
    Scale RGB=&VectorFromFloat(Value=&FloatFromCurve(FloatCurve=DataInterface() {propertyValues {Curve {keys [<at 0 2 RCIM_Cubic>,<at 0.8 1 RCIM_Cubic>]}}}))
    ScaleA=true
    ScaleRGB=true
   )
   &ScaleSpriteSize(Scale Factor=&Vector2DFromFloat(Value=&FloatFromCurve(
      FloatCurve=DataInterface() {propertyValues {Curve {keys [
          {arriveTangent=50.3486, interpMode=RCIM_Cubic, leaveTangent=50.3492, tangentMode=RCTM_User, time=0, value=0}
          <at 0.03 1 RCIM_Cubic>
          {arriveTangent=-3.0852, interpMode=RCIM_Cubic, leaveTangent=-3.0852, tangentMode=RCTM_User, time=1, value=0}
         ]}}}
      Scale Curve=$Particles.DistanceScale
     )))
   &FN_Sprite_Attributes(
    Color=&MultiplyLinearColor(
     Linear Color=$Particles.Color
     Scale Factor=&MakeVector4_FromLinearColor(LinearColor=<rgba 0.5 0.5 0.5 0.03>)
    )
    Create Color Attribute=true
    Create SpriteSize Attribute=true
    SpriteSize=&Multiply_Vector2D(A=$Particles.SpriteSize, B=<10 3>)
    Write to Module Namespace=true
   )
   &DynamicMaterialParameters(
    Index 0 Param 1=1
    Index 0 Param 2=1
    Index 0 Param 3=1
    Module.Index0Float0Write=true
    Module.Index0Float1Write=true
    Module.Index0Float2Write=true
    Module.Index0Float3Write=true
   )
  ]
  rendererClasses [&NiagaraSpriteRendererProperties]
  renderers [
   {rendererClass=&NiagaraSpriteRendererProperties}
   {rendererClass=&NiagaraSpriteRendererProperties, rendererIndex=1}
  ]
 }
 ThrusterFlames {
  emitterSpawnScript [/SetVariables() {isSetParameters=true}]
  emitterUpdateScript [
   &EmitterState(
    Inactive Response=::NewEnumerator1
    Life Cycle Mode=::NewEnumerator1
    Loop Behavior=::NewEnumerator1
    Loop Duration Mode=::NewEnumerator1
    UseLoopDelay=false
   )
   &SpawnBurst_Instantaneous(Spawn Count=6, Use Spawn Probability=false)
   /SetVariables(Emitter.SpeedAlpha:NiagaraFloat=&ClampFloat(Float=&Divide_Float(A=&Subtract_Float(A=$User.Speed, B=5), B=50))) {isSetParameters=true}
   /SetVariables(Emitter.SmoothVelocity:Vector3f=&Clamp_Vector(
     Vector:Vector3f=&Lerp_Vector(
      A=$Emitter.SmoothVelocity
      B=&TransformVector(Vector=$Engine.Owner.Velocity, Source Space=::World, Destination Space=::Local)
      Alpha=0.06
     )
     Min:Vector3f=<0 0 0>
     Max:Vector3f=<1000 1000 1000>
    )) {isSetParameters=true}
   /SetVariables(Emitter.ForwardDot:NiagaraFloat=&ClampFloat(Float=&Divide_Float(
      A=&Subtract_Float(A=&MakeFloatFromVector(Vector=$Emitter.SmoothVelocity, Channel=::NewEnumerator0), B=50)
      B=500
     ))) {isSetParameters=true}
  ]
  particleSpawnScript [
   &InitializeParticle(Lifetime=0.2, Mesh Scale=<0.5 2 1>, Mesh Scale Mode=::NewEnumerator4, UsePositionOffset=false, Write Material Random=true) {module=InitializeParticle001}
   &InitialMeshOrientation(Mesh Orientation Mode=::NewEnumerator7, Rotation=<0 0 0.5>, Use Rotation Vector=true) {module=InitialMeshOrientation001}
  ]
  particleUpdateScript [
   &ParticleState(Kill Particles When Lifetime Has Elapsed=false)
   /SetVariables(Particles.Oscillation:NiagaraFloat=&Add_Float(
     A=&Waveform(
      Waveform Count=::NewEnumerator1
      Global Phase=$Particles.Age
      "Waveform [1]"=::NewEnumerator0
      "[1] Frequency"=2.256
      "[1] Amplitude Min/Max"=<0 1>
      "[1] Scale"=0.5
      "[1] Bias"=1
      "Waveform [2] Blend"=::NewEnumerator2
      "[2] Input Position"=0
      "[2] Phase"=$Emitter.Age
      "[2] Phase Scale"=1
      "[2] Compound Frequency"=2
      "[2] Frequency"=7.2563
      "[2] Amplitude Min/Max"=<0 1>
      "[2] Scale"=0.5
      "[2] Bias"=1
      "[2] Exponent"=1
      "[2] Pulse Duration"=0.1
      "[2] Clip Triangle Min/Max"=<0 0>
      "[2] Random Seed"=0
     )
     B=0
    )) {isSetParameters=true}
   &SkeletalMeshLocation(
    Bone / Socket Index=&SelectIntFromArray(
     Direct Array Index=&ReturnExecIndex()
     Int Selection Array=DataInterface() {propertyValues {IntData [0,0,0,1,1,1]}}
    )
    MaskSpawnGroup=false
    Mesh Sampling Type=::NewEnumerator5
    Sampled Position Offset=&Multiply_Vector(
     A=&SelectVectorFromArray(
      Direct Array Index=&ReturnExecIndex()
      Vector Selection Array=DataInterface() {propertyValues {FloatData [<-10 6 -4>, <-20 31 -4>, <-30 55 -4>, <-10 -6 -4>, <-20 -31 -4>, <-30 -55 -4>]}}
     )
     B=<1 1 1>
    )
    Skeletal Mesh=$User.SkeletalMesh
    Socket Sampling Mode=::NewEnumerator5
   ) {module=SkeletalMeshLocation001}
   &ScaleMeshSize(Scale Factor=&Multiply_Vector(
     A=<7 0.6 0.7>
     B=&Multiply_Vector(
      A=&Lerp_Vector(A=<0.75 1 1>, B=<1.75 1 1>, Alpha=$Emitter.ForwardDot)
      B=&Lerp_Vector(
       A=<1 1 0.5>
       B=&SelectVectorFromArray(
        Direct Array Index=&ModuloInt(A=&ReturnExecIndex(), B=3)
        Vector Selection Array=DataInterface() {propertyValues {FloatData [<2 1 1>, <3 1 1>, <4 1 1>]}}
       )
       Alpha=&Multiply_Float(A=$System.Timeline.NormalizedTime, B=&Lerp_Float(A=0.8, B=1, Alpha=$Particles.Oscillation))
      )
     )
    ))
   &SolveForcesAndVelocity(Clamp Velocity=false, Limit Acceleration=false)
   &DynamicMaterialParameters(
    Index 0 Param 1=&FloatFromCurve(FloatCurve=DataInterface() {propertyValues {Curve {keys [{arriveTangent=-2.9083, interpMode=RCIM_Cubic, leaveTangent=-2.9083, tangentMode=RCTM_User, time=0, value=1}, <at 1 0 RCIM_Cubic>]}}})
    Index 0 Param 2=5
    Index 0 Param 3=5
    Module.Param0WriteEnabled=true
    Module.Param0WriteEnabled3=true
    Module.Param1WriteEnabled=true
    Module.Param1WriteEnabled3=true
    Module.Param2WriteEnabled=true
    Module.Param2WriteEnabled3=true
    Module.Param3WriteEnabled=true
    Module.Param3WriteEnabled3=true
    Write Parameter Index 3=true
   )
   &ScaleColor(
    Scale Alpha=&Lerp_Float(
     A=0
     B=1
     Alpha=/Niagara/DynamicInputs/Camera/CameraFade(
      Camera Query:NiagaraDataInterfaceCamera=DataInterface()
      Fade Offset:NiagaraFloat=150
      Fade Distance:NiagaraFloat=100
      Use Owner Position:NiagaraBool=true
     )
    )
    Scale RGB=&Multiply_Vector(
     A=&MakeVectorFromLinearColorRGB(Color=<rgba 0.15 1.575 3 1>)
     B=&MakeVectorFromLinearColorRGB(Color=&Multiply_LinearColorByFloat(
       Linear Color=&ColorFromCurve(
        ColorCurve=DataInterface() {propertyValues {
          AlphaCurve {keys []}
          BlueCurve {keys [<at 0 1 RCIM_Cubic>,<at 1 10 RCIM_Cubic>]}
          GreenCurve {keys [<at 0 1 RCIM_Cubic>,<at 1 3.1464 RCIM_Cubic>]}
          RedCurve {keys [<at 0 1 RCIM_Cubic>,<at 1 0 RCIM_Cubic>]}
         }}
        CurveIndex=$System.Timeline.NormalizedTime
       )
       Float=1
      ))
    )
    ScaleA=true
    ScaleRGB=true
   )
   &DynamicMaterialParameters(
    Index 0 Param 1=&RandomRangeFloat(Minimum=0, Maximum=1)
    Index 3 Param 1=3
    Index 3 Param 2=50
    Index 3 Param 3=1
    Module.Index0Float0Write=true
    Module.Index0Float1Write=true
    Module.Index0Float2Write=true
    Module.Index0Float3Write=true
    Module.Index3Float0Write=true
    Module.Index3Float1Write=true
    Module.Index3Float2Write=true
    Module.Index3Float3Write=true
    Write Parameter Index 3=::NewEnumerator1
   ) {module=DynamicMaterialParameters001}
   /NiagaraFN/Sprite/FN_SpriteAlignmentAndFacing(SpriteAlignment:Vector3f=<0 0 1>, SpriteFacing:Vector3f=<-1 0 0>)
   &FN_Sprite_Attributes(
    Create Camera Offset Attribute=true
    Create Position Attribute=true
    Create SpriteSize Attribute=true
    Position:NiagaraPosition=&AddVectorToPosition(
     Position=$Particles.Position
     Vector=&RotateVectorByQuaternion(Vector To Rotate=<-17 0 0>, Quaternion=$Particles.SkeletalMeshLocation001.SampledRotation)
    )
    SpriteSize=<50 30>
    Write to Module Namespace=true
   )
   &FN_Sprite_Attributes(
    Color=&ColorFromCurve(
     ColorCurve=DataInterface() {propertyValues {
       AlphaCurve {keys [<at 0.0055 0>,<at 1 1 RCIM_Cubic>]}
       BlueCurve {keys [<at 0 1 RCIM_Cubic>,<at 1 0.7 RCIM_Cubic>]}
       GreenCurve {keys [<at 0 0.458 RCIM_Cubic>,<at 1 0.383 RCIM_Cubic>]}
       RedCurve {keys [<at 0 0 RCIM_Cubic>,<at 1 0.1057 RCIM_Cubic>]}
      }}
     CurveIndex=&Max_Float(A:NiagaraFloat=$System.Timeline.NormalizedTime, B:NiagaraFloat=$System.TakeOffAlpha)
    )
    Create Camera Offset Attribute=true
    Create Color Attribute=true
    Create Position Attribute=true
    Create SpriteSize Attribute=true
    Position:NiagaraPosition=&AddVectorToPosition(
     Position=$Particles.Position
     Vector=&RotateVectorByQuaternion(Vector To Rotate=<-30 0 0>, Quaternion=$Particles.SkeletalMeshLocation001.SampledRotation)
    )
    SpriteSize=&Multiply_Vector2D_ByFloat(
     Float=&Multiply_Float(A=$System.Timeline.NormalizedTime, B=&Lerp_Float(A=1, B=1.2, Alpha=$Particles.Oscillation))
     Vector2D=<150 150>
    )
    Write to Module Namespace=true
   ) {module=FN_Sprite_Attributes001}
   &FN_Light_Attributes(
    Color=&ColorFromCurve(
     ColorCurve=DataInterface() {propertyValues {
       AlphaCurve {keys [<at 0 1 RCIM_Cubic>,<at 1 1 RCIM_Cubic>]}
       BlueCurve {keys [<at 0 0.2 RCIM_Cubic>,<at 1 2 RCIM_Cubic>]}
       GreenCurve {keys [<at 0 0.1368 RCIM_Cubic>,<at 1 1.2592 RCIM_Cubic>]}
       RedCurve {keys [<at 0 0 RCIM_Cubic>,<at 1 0 RCIM_Cubic>]}
      }}
     CurveIndex=&Max_Float(A:NiagaraFloat=$System.Timeline.NormalizedTime, B:NiagaraFloat=$System.TakeOffAlpha)
    )
    Create Color Attribute=true
    Create Exponent Attribute=true
    Create Radius Attribute=true
    Enable Particle Light Enable=true
    Exponent=1
    Light Enabled:NiagaraBool=&SetBoolByIntComparison(A=&ModuloInt(A=&ReturnExecIndex(), B=3), B=1, Comparison Type=::NewEnumerator4)
    Radius=&MakeCustomFloatFromBool(Bool=$User.IsBoosting, True Float=&Multiply_Float(A=95, B=$System.Timeline.NormalizedTime), False Float=60)
    Scale Color=10
   )
  ]
  rendererClasses [
   &NiagaraMeshRendererProperties
   &NiagaraSpriteRendererProperties
   &NiagaraLightRendererProperties
  ]
  renderers [
   {rendererClass=&NiagaraMeshRendererProperties, rendererIndex=0}
   {rendererClass=&NiagaraSpriteRendererProperties, rendererIndex=1}
   {rendererClass=&NiagaraLightRendererProperties, rendererIndex=2}
   {rendererClass=&NiagaraSpriteRendererProperties, rendererIndex=3}
   {rendererClass=&NiagaraSpriteRendererProperties, rendererIndex=4}
  ]
 }
}
summary {
 systemName=NS_TidalCraneHoagie_Idle
 emitters [
  {
   emitterName=ThrusterFlames
   rendererClasses [
    &NiagaraMeshRendererProperties
    &NiagaraSpriteRendererProperties
    &NiagaraLightRendererProperties
   ]
  }
  {emitterName=FN_Sprite_SpeedLines001, rendererClasses [&NiagaraSpriteRendererProperties]}
  {
   emitterName=RotorFlames
   rendererClasses [
    &NiagaraMeshRendererProperties
    &NiagaraSpriteRendererProperties
    &NiagaraLightRendererProperties
   ]
  }
  {
   emitterName=NE_RotorMotionBlur
   rendererClasses [&NiagaraSpriteRendererProperties, &NiagaraMeshRendererProperties]
  }
  {emitterName=Smoke, rendererClasses [&NiagaraSpriteRendererProperties]}
  {emitterName=Sparks, simTarget=GPUComputeSim, rendererClasses [&NiagaraSpriteRendererProperties]}
  {emitterName=PP, rendererClasses [&NiagaraComponentRendererProperties]}
 ]
 userVariables [
  User.BoostReady {type=NiagaraFloat, defaultValue=0, description=""}
  User.CameraFade {type=NiagaraFloat, defaultValue=0, description=""}
  User.FX_Scale {type=NiagaraFloat, defaultValue=1.2, description=""}
  User.IsBoosting {type=NiagaraBool}
  User.IsLocal {type=NiagaraBool, defaultValue=true}
  User.NightMultiplier {type=NiagaraFloat, defaultValue=1, description=""}
  User.RotorSpeedAlpha {type=NiagaraFloat, defaultValue=1, description=""}
  User.SkeletalMesh {
   type=NiagaraDataInterfaceSkeletalMesh
   defaultValue {
    struct=NiagaraExt_VariableValue_DataInterface
    value {dataInterface=&self:NiagaraDataInterfaceSkeletalMesh_0}
   }
  }
  User.Speed {type=NiagaraFloat, defaultValue=0, description=""}
 ]
}
