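# Niagara system for TidalCrane Hoagie rotor wash, kicking up surface-specific effects where the downwash hits.
# PhysMatIndex, driver and active flags gate stone, grass, dirt and water emitters (bubbles, foam, mist, splashes); generic clouds ignore these gates.
# Particles spawn on a torus aligned to ImpactNormal, angled by Theta and widening with Altitude, then fly along its normal.
# RotorPlaySpeed scales spawning and velocity; Altitude thins debris, splashes and clouds.
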NiagaraSystem /TidalCraneHoagie/Assets/FX/NS_TidalCrane_Hoagie_RotorWashBurst
dependencies {
 usedDataInterfaces [&NiagaraDataInterfaceCurve, &NiagaraDataInterfaceColorCurve, &NiagaraDataInterfaceVectorCurve]
 usedDynamicInputs []
 usedModules [&SystemState]
}
emitters {
 Bubbles {
  emitterUpdateScript [
   &EmitterState()
   &EmitterScalability(
    MaxDistance=2000
    Module.bApplySpawnCountScale=true
    Module.bApplySpawnCountScaleByDistanceFraction=true
    Module.bUseMaxDistance=true
    Module.bUseMinDistance=false
    SpawnCountScaleByDistanceFraction=DataInterface() {propertyValues {Curve {keys [<at 0 1>,<at 1 0.5>]}}}
   ) {enabled=false}
   &EmitterLifeCycle() {enabled=false}
   &SpawnPerFrame(
    Spawn:NiagaraBool=&Boolean_AndOperation(
     Bool1=/Hoagie/FX/DynamicInputs/FN_PhysicalMaterialBoolComparison(Index:NiagaraInt32=2, Physical Material Index:NiagaraInt32=$User.PhysMatIndex)
     Bool2=&Boolean_AndOperation(Bool1=$User.hasDriver, Bool2=$User.isRotorWashActive)
    )
    Spawn Count:NiagaraInt32=&MakeIntFromFloat(Float=&Multiply_Float(
      A=$User.RotorPlaySpeed
      B=&Multiply_Float(A=5, B=&OneMinusFloat(Float=$User.Altitude))
     ))
    Use Spawn Probability=false
   ) {enabled=false}
   &SpawnRate(
    Spawn Probability=&MakeCustomFloatFromBool(
     Bool=&Boolean_AndOperation(
      Bool1=&FN_PhysicalMaterialBoolComparison(Index:NiagaraInt32=2, Physical Material Index:NiagaraInt32=$User.PhysMatIndex)
      Bool2=&Boolean_AndOperation(Bool1=$User.hasDriver, Bool2=$User.isRotorWashActive)
     )
     True Float=1
     False Float=0
    )
    SpawnRate=&Multiply_Float(A=$User.RotorPlaySpeed, B=60)
    Use Spawn Probability=true
   )
  ]
  particleSpawnScript [
   /Niagara/Modules/Spawn/Initialization/InitializeParticle(
    Lifetime=&UniformRangedFloat(Minimum=0.5, Maximum=0.8)
    Module.Write Color=true
    Module.Write Lifetime=true
    Module.Write Mass=true
    Module.Write Material Random=true
    Module.Write Position=true
    Module.Write Scale=false
    Module.Write SpriteRotation=false
    Module.Write SpriteSize=false
    Position=&Multiply_VectorByFloat(Vector=&UniformRangedVector(Minimum=<-200 0 -10>, Maximum=<200 0 -10>), Float=$User.Width)
   )
   /Niagara/Modules/Spawn/Orientation/V2/InitialMeshOrientation(Use Orientation Vector=true, Use Rotation Vector=true) {enabled=false}
   &TorusLocation(
    Align Torus With Axis=true
    Coordinate Space:ENiagaraCoordinateSpace=::World
    Large Radius:NiagaraFloat=&Lerp_Float(A=$User.TorusInnerRadius, B=&Multiply_Float(A=$User.TorusInnerRadius, B=2), Alpha=$User.Altitude)
    MaskSpawnGroup=false
    Radius Position=0.2752
    Torus Axis=$User.ImpactNormal
    Torus Distribution Mode:ENiagaraTorusDistributionMode=::NewEnumerator1
    Torus Origin:NiagaraPosition=/Niagara/DynamicInputs/Helpers/SimulationPosition()
    U Position:NiagaraFloat=&Add_Float(
     A=$User.Theta
     B=/Niagara/DynamicInputs/Angles/DegreesToRadians(AngleInDegrees:NiagaraFloat=&UniformRangedFloat(Minimum=-2, Maximum=2))
    )
    V Position=0.05
   )
   /Hoagie/FX/Modules/FN_AlignTorusPosition()
   &AddVelocity(Velocity=&Multiply_VectorByFloat(
     Vector=&TransformVector(Vector=$Output.TorusLocation.TorusNormal, Source Space=::World, Destination Space=::Local)
     Float=&Multiply_Float(A=&UniformRangedFloat(Minimum=750, Maximum=1500), B=$User.RotorPlaySpeed)
    ))
  ]
  particleUpdateScript [
   &ParticleState()
   &Drag(Drag=&UniformRangedFloat(Minimum=5, Maximum=7), Use Linear Drag=true, Use Rotational Drag=false)
   &GravityForce(Gravity=<0 0 -1000>)
   &SolveForcesAndVelocity(Clamp Velocity=false, Limit Acceleration=false)
   &DynamicMaterialParameters(
    Index 0 Param 1=&FloatFromCurve(FloatCurve=DataInterface() {propertyValues {Curve {keys [{arriveTangent=-10.0518, interpMode=RCIM_Cubic, leaveTangent=-6.7075, tangentMode=RCTM_User, time=-0.0046, value=3.2841}, <at 1.0011 0.002>]}}})
    Index 0 Param 2=&FloatFromCurve(FloatCurve=DataInterface() {propertyValues {Curve {keys [<at 0.0645 8.5803>,<at 1.0011 1>]}}})
    Index 0 Param 3=&FloatFromCurve(FloatCurve=DataInterface() {propertyValues {Curve {keys [<at -0.0046 60.8837>,<at 1.0048 166.0627>]}}})
    Index 0 Param 4=HlslExpression() {hlslExpression="rand(8.0) + 8"}
    Module.Param0WriteEnabled=true
    Module.Param1WriteEnabled=true
    Module.Param2WriteEnabled=true
    Module.Param3WriteEnabled=true
   )
   &ScaleColor(
    Scale Alpha=&FloatFromCurve(FloatCurve=DataInterface() {propertyValues {Curve {keys [<at 0 -1>,<at 0.05 0>,<at 0.7904 0.0132>,<at 1.0011 -0.9961>]}}})
    ScaleA=true
    ScaleRGB=true
   )
   &CameraOffset(Camera Offset Amount=50)
  ]
  rendererClasses [&NiagaraMeshRendererProperties]
  renderers [{rendererClass=&NiagaraMeshRendererProperties, rendererIndex=0}]
 }
 DebirsDirt {
  emitterUpdateScript [
   &EmitterState()
   &SpawnRate(
    Spawn Probability=&MakeCustomFloatFromBool(
     Bool=&Boolean_AndOperation(
      Bool1=&FN_PhysicalMaterialBoolComparison(Index:NiagaraInt32=4, Physical Material Index:NiagaraInt32=$User.PhysMatIndex)
      Bool2=&Boolean_AndOperation(Bool1=$User.hasDriver, Bool2=$User.isRotorWashActive)
     )
     True Float=1
     False Float=0
    )
    SpawnRate=&Multiply_Float(
     A=&Multiply_Float(A=$User.RotorPlaySpeed, B=&OneMinusFloat(Float=$User.Altitude))
     B=25
    )
    Use Spawn Probability=true
   )
  ]
  particleSpawnScript [
   &InitializeParticle(
    Color=<rgba 0.2448 0.1904 0.1313 0.8>
    Lifetime=&UniformRangedFloat(Minimum=0.8, Maximum=2)
    Module.Write Color=true
    Module.Write Lifetime=true
    Module.Write Mass=true
    Module.Write Material Random=true
    Module.Write Position=false
    Module.Write Scale=false
    Module.Write SpriteRotation=true
    Module.Write SpriteSize=true
    Position=/Niagara/Functions/Localspace/SimulationPosition()
    Sprite Rotation=&NormalizedAngleToDegrees(NormalizedAngle=&UniformRangedFloat(Minimum=0, Maximum=1))
    Sprite Size=&Vector2DFromFloat(Value=&UniformRangedFloat(Minimum=1, Maximum=10))
   )
   &TorusLocation(
    Align Torus With Axis=true
    Coordinate Space:ENiagaraCoordinateSpace=::World
    Large Radius:NiagaraFloat=&Lerp_Float(A=$User.TorusInnerRadius, B=&Multiply_Float(A=$User.TorusInnerRadius, B=2), Alpha=$User.Altitude)
    MaskSpawnGroup=false
    Radius Position=0.2752
    Torus Axis=$User.ImpactNormal
    Torus Distribution Mode:ENiagaraTorusDistributionMode=::NewEnumerator1
    Torus Origin:NiagaraPosition=/Niagara/DynamicInputs/Helpers/SimulationPosition()
    U Position:NiagaraFloat=&Add_Float(
     A=$User.Theta
     B=&DegreesToRadians(AngleInDegrees:NiagaraFloat=&UniformRangedFloat(Minimum=-2, Maximum=2))
    )
    V Position=0.05
   )
   &FN_AlignTorusPosition() {module=AlignTorusPosition}
   &AddVelocity(Velocity=&Multiply_VectorByFloat(
     Vector=&TransformVector(Vector=$Output.TorusLocation.TorusNormal, Source Space=::World, Destination Space=::Local)
     Float=&Multiply_Float(A=&UniformRangedFloat(Minimum=250, Maximum=500), B=$User.RotorPlaySpeed)
    ))
   /Niagara/Modules/Update/SubUV/SubUVAnimation(SubUV Animation Mode=::NewEnumerator1)
  ]
  particleUpdateScript [
   &ParticleState()
   &SpriteRotationRate(Rotation Rate=&UniformRangedFloat(Minimum=-10, Maximum=10))
   &AccelerationForce(Acceleration=<0 0 -980>)
   &SolveForcesAndVelocity(Clamp Velocity=false, Limit Acceleration=false)
  ]
  rendererClasses [&NiagaraSpriteRendererProperties]
  renderers [{rendererClass=&NiagaraSpriteRendererProperties}]
 }
 Debris_Grass {
  emitterUpdateScript [
   &EmitterState()
   &SpawnRate(
    Spawn Probability=&MakeCustomFloatFromBool(
     Bool=&Boolean_AndOperation(
      Bool1=&FN_PhysicalMaterialBoolComparison(Index:NiagaraInt32=3, Physical Material Index:NiagaraInt32=$User.PhysMatIndex)
      Bool2=&Boolean_AndOperation(Bool1=$User.hasDriver, Bool2=$User.isRotorWashActive)
     )
     True Float=1
     False Float=0
    )
    SpawnRate=&Multiply_Float(
     A=&Multiply_Float(A=$User.RotorPlaySpeed, B=&OneMinusFloat(Float=$User.Altitude))
     B=25
    )
    Use Spawn Probability=true
   )
  ]
  particleSpawnScript [
   &InitializeParticle(
    Color=<rgba 0.117 0.3325 0 1>
    Lifetime=&UniformRangedFloat(Minimum=0.8, Maximum=2)
    Mesh Scale=&UniformRangedVector(Minimum=<0.25 0.25 0.25>, Maximum=<1 0.25 1>)
    Module.Write Color=true
    Module.Write Lifetime=true
    Module.Write Mass=true
    Module.Write Material Random=true
    Module.Write Position=false
    Module.Write Scale=true
    Module.Write SpriteRotation=true
    Module.Write SpriteSize=false
    Position=/Niagara/Functions/Localspace/SimulationPosition()
    Sprite Rotation=&NormalizedAngleToDegrees(NormalizedAngle=&UniformRangedFloat(Minimum=0, Maximum=1))
   )
   &TorusLocation(
    Align Torus With Axis=true
    Coordinate Space:ENiagaraCoordinateSpace=::World
    Large Radius:NiagaraFloat=&Lerp_Float(A=$User.TorusInnerRadius, B=&Multiply_Float(A=$User.TorusInnerRadius, B=2), Alpha=$User.Altitude)
    MaskSpawnGroup=false
    Radius Position=0.2752
    Torus Axis=$User.ImpactNormal
    Torus Distribution Mode:ENiagaraTorusDistributionMode=::NewEnumerator1
    Torus Origin:NiagaraPosition=/Niagara/DynamicInputs/Helpers/SimulationPosition()
    U Position:NiagaraFloat=&Add_Float(
     A=$User.Theta
     B=&DegreesToRadians(AngleInDegrees:NiagaraFloat=&UniformRangedFloat(Minimum=-2, Maximum=2))
    )
    V Position=0.05
   )
   &FN_AlignTorusPosition() {module=AlignTorusPosition}
   &AddVelocity(Velocity=&Multiply_VectorByFloat(
     Vector=&TransformVector(Vector=$Output.TorusLocation.TorusNormal, Source Space=::World, Destination Space=::Local)
     Float=&Multiply_Float(A=&UniformRangedFloat(Minimum=250, Maximum=500), B=$User.RotorPlaySpeed)
    ))
   &SubUVAnimation(SubUV Animation Mode=::NewEnumerator1)
  ]
  particleUpdateScript [
   &ParticleState()
   &AccelerationForce(Acceleration=<0 0 -450>)
   /Niagara/Modules/Update/Forces/CurlNoiseForce(Noise Frequency=1, Noise Strength=1000, Pan Noise=true, Pan Noise Field=<0 0 10>)
   &SolveForcesAndVelocity(Clamp Velocity=false, Limit Acceleration=false)
  ]
  rendererClasses [&NiagaraMeshRendererProperties]
  renderers [{rendererClass=&NiagaraMeshRendererProperties, rendererIndex=0}]
 }
 Debris_Stone {
  emitterUpdateScript [
   &EmitterState()
   &SpawnRate(
    Spawn Probability=&MakeCustomFloatFromBool(
     Bool=&Boolean_AndOperation(
      Bool1=&FN_PhysicalMaterialBoolComparison(Index:NiagaraInt32=1, Physical Material Index:NiagaraInt32=$User.PhysMatIndex)
      Bool2=&Boolean_AndOperation(Bool1=$User.hasDriver, Bool2=$User.isRotorWashActive)
     )
     True Float=1
     False Float=0
    )
    SpawnRate=&Multiply_Float(
     A=&Multiply_Float(A=$User.RotorPlaySpeed, B=&OneMinusFloat(Float=$User.Altitude))
     B=25
    )
    Use Spawn Probability=true
   )
  ]
  particleSpawnScript [
   &InitializeParticle(
    Lifetime=&UniformRangedFloat(Minimum=0.8, Maximum=2)
    Module.Write Color=true
    Module.Write Lifetime=true
    Module.Write Mass=true
    Module.Write Material Random=true
    Module.Write Position=false
    Module.Write Scale=false
    Module.Write SpriteRotation=true
    Module.Write SpriteSize=true
    Position=/Niagara/Functions/Localspace/SimulationPosition()
    Sprite Rotation=&NormalizedAngleToDegrees(NormalizedAngle=&UniformRangedFloat(Minimum=0, Maximum=1))
    Sprite Size=&Vector2DFromFloat(Value=&UniformRangedFloat(Minimum=8, Maximum=12))
   )
   &TorusLocation(
    Align Torus With Axis=true
    Coordinate Space:ENiagaraCoordinateSpace=::World
    Large Radius:NiagaraFloat=&Lerp_Float(A=$User.TorusInnerRadius, B=&Multiply_Float(A=$User.TorusInnerRadius, B=2), Alpha=$User.Altitude)
    MaskSpawnGroup=false
    Radius Position=0.2752
    Torus Axis=$User.ImpactNormal
    Torus Distribution Mode:ENiagaraTorusDistributionMode=::NewEnumerator1
    Torus Origin:NiagaraPosition=/Niagara/DynamicInputs/Helpers/SimulationPosition()
    U Position:NiagaraFloat=&Add_Float(
     A=$User.Theta
     B=&DegreesToRadians(AngleInDegrees:NiagaraFloat=&UniformRangedFloat(Minimum=-2, Maximum=2))
    )
    V Position=0.05
   )
   &FN_AlignTorusPosition() {module=AlignTorusPosition}
   &AddVelocity(Velocity=&Multiply_VectorByFloat(
     Vector=&TransformVector(Vector=$Output.TorusLocation.TorusNormal, Source Space=::World, Destination Space=::Local)
     Float=&Multiply_Float(A=&UniformRangedFloat(Minimum=850, Maximum=1250), B=$User.RotorPlaySpeed)
    ))
   &SubUVAnimation(SubUV Animation Mode=::NewEnumerator1)
  ]
  particleUpdateScript [
   &ParticleState()
   &SpriteRotationRate(Rotation Rate=&Multiply_Float(
     A=&UniformRangedFloat(Minimum=250, Maximum=350)
     B=&FloatFromCurve(FloatCurve=DataInterface() {propertyValues {Curve {keys [
          {interpMode=RCIM_Cubic, tangentMode=RCTM_User, time=0, value=1}
          {arriveTangent=-1.1683, interpMode=RCIM_Cubic, leaveTangent=-1.1683, tangentMode=RCTM_User, time=1, value=0.2}
         ]}}})
    ))
   &AccelerationForce(Acceleration=<0 0 -300>)
   &Drag(Drag=3, Use Linear Drag=true, Use Rotational Drag=true)
   &SolveForcesAndVelocity(Clamp Velocity=false, Limit Acceleration=false)
  ]
  rendererClasses [&NiagaraSpriteRendererProperties]
  renderers [{rendererClass=&NiagaraSpriteRendererProperties}]
 }
 Foam_Detail {
  emitterUpdateScript [
   &EmitterState()
   &EmitterScalability(
    MaxDistance=2000
    Module.bApplySpawnCountScale=false
    Module.bApplySpawnCountScaleByDistanceFraction=false
    Module.bUseMaxDistance=false
    Module.bUseMinDistance=false
    SpawnCountScaleByDistanceFraction=DataInterface() {propertyValues {Curve {keys [<at 0 1>,<at 1 0.5>]}}}
   ) {enabled=false}
   &SpawnPerUnit(
    Module.Use Max Movement Threshold=true
    Module.Use Movement Tolerance=true
    Movement Tolerance=5
    Spawn Per Unit=200
    Use Spawn Probability=false
   ) {enabled=false}
   &EmitterLifeCycle() {enabled=false}
   &SpawnPerFrame(
    Spawn:NiagaraBool=&Boolean_AndOperation(
     Bool1=&FN_PhysicalMaterialBoolComparison(Index:NiagaraInt32=2, Physical Material Index:NiagaraInt32=$User.PhysMatIndex)
     Bool2=&Boolean_AndOperation(Bool1=$User.hasDriver, Bool2=$User.isRotorWashActive)
    )
    Spawn Count:NiagaraInt32=&MakeIntFromFloat(Float=&Multiply_Float(
      A=$User.RotorPlaySpeed
      B=&Multiply_Float(A=3, B=&OneMinusFloat(Float=$User.Altitude))
     ))
    Use Spawn Probability=false
   ) {enabled=false}
   &SpawnRate(
    Spawn Probability=&MakeCustomFloatFromBool(
     Bool=&Boolean_AndOperation(
      Bool1=&FN_PhysicalMaterialBoolComparison(Index:NiagaraInt32=2, Physical Material Index:NiagaraInt32=$User.PhysMatIndex)
      Bool2=&Boolean_AndOperation(Bool1=$User.hasDriver, Bool2=$User.isRotorWashActive)
     )
     True Float=1
     False Float=0
    )
    SpawnRate=&Multiply_Float(A=$User.RotorPlaySpeed, B=60)
    Use Spawn Probability=true
   )
  ]
  particleSpawnScript [
   &InitializeParticle(
    Color=<rgba 0.8 0.8 0.8 1>
    Lifetime=&UniformRangedFloat(Minimum=0.8, Maximum=1.2)
    Mass=&UniformRangedFloat(Minimum=0.6, Maximum=1.2)
    Mesh Scale=&VectorFromFloat(Value=&UniformRangedFloat(Minimum=200, Maximum=400))
    Module.Write Color=true
    Module.Write Lifetime=true
    Module.Write Mass=true
    Module.Write Material Random=true
    Module.Write Position=false
    Module.Write Scale=true
    Module.Write SpriteRotation=true
    Module.Write SpriteSize=false
    Position=/Niagara/Functions/Localspace/SimulationPosition()
    Sprite Rotation=&NormalizedAngleToDegrees(NormalizedAngle=&UniformRangedFloat(Minimum=0, Maximum=1))
    Sprite Size=<2 2>
   ) {enabled=false}
   &TorusLocation(
    Align Torus With Axis=true
    Coordinate Space:ENiagaraCoordinateSpace=::World
    Large Radius:NiagaraFloat=&Lerp_Float(A=$User.TorusInnerRadius, B=&Multiply_Float(A=$User.TorusInnerRadius, B=2), Alpha=$User.Altitude)
    MaskSpawnGroup=false
    Radius Position=0.2752
    Torus Axis=$User.ImpactNormal
    Torus Distribution Mode:ENiagaraTorusDistributionMode=::NewEnumerator1
    Torus Origin:NiagaraPosition=/Niagara/DynamicInputs/Helpers/SimulationPosition()
    U Position:NiagaraFloat=&Add_Float(
     A=$User.Theta
     B=&DegreesToRadians(AngleInDegrees:NiagaraFloat=&UniformRangedFloat(Minimum=-2, Maximum=2))
    )
    V Position=0.05
   )
   &FN_AlignTorusPosition()
   &AddVelocity(Velocity=&Multiply_VectorByFloat(
     Vector=&TransformVector(Vector=$Output.TorusLocation.TorusNormal, Source Space=::World, Destination Space=::Local)
     Float=&Multiply_Float(A=&UniformRangedFloat(Minimum=250, Maximum=500), B=$User.RotorPlaySpeed)
    ))
   /Niagara/Modules/Spawn/Initialization/V2/InitializeParticle(
    Lifetime=&UniformRangedFloat(Minimum=2, Maximum=3)
    Mass=&UniformRangedFloat(Minimum=0.6, Maximum=1.2)
    Mass Mode=::NewEnumerator1
    Mesh Scale Mode=::NewEnumerator3
    Mesh Uniform Scale=100
    Sprite Rotation Angle=&UniformRangedFloat(Minimum=0, Maximum=1)
    Sprite Rotation Mode=::NewEnumerator1
    UsePositionOffset=false
    Write Material Random=true
   ) {module=InitializeParticle001}
   &InitialMeshOrientation(
    Orientation Coordinate Space:ENiagaraCoordinateSpace=::World
    Orientation Vector=<0 0 0>
    Rotation=&UniformRangedVector(Minimum=<0 0 0>, Maximum=<0 0 1>)
    Use Orientation Vector=false
    Use Rotation Vector=true
   ) {enabled=false}
   /Niagara/Modules/Spawn/Orientation/V3/InitialMeshOrientation(
    Mesh Orientation Mode=::NewEnumerator2
    Orientation Vector=<0 0 0>
    Rotation=&UniformRangedVector(Minimum=<0 0 0>, Maximum=<0 0 1>)
    Use Rotation Vector=true
   ) {module=InitialMeshOrientation001}
   &DynamicMaterialParameters(
    Index 1 Param 3=750
    Module.Param0WriteEnabled1=false
    Module.Param1WriteEnabled1=false
    Module.Param2WriteEnabled1=true
    Module.Param3WriteEnabled1=false
    Write Parameter Index 0=false
    Write Parameter Index 1=true
   )
   &PointForce(Force Origin=/Niagara/Functions/Localspace/SimulationPosition(), Force Strength=200, Module.Use Falloff=false, Module.UseLinearForce=false) {enabled=false}
   &ApplyInitialForces()
   &InitializeWaterSnapToOwner()
   &InheritVelocity(Module.Limit Speed=true, Velocity Scale=<0.8 0.8 0.8>)
  ]
  particleUpdateScript [
   &ParticleState()
   &DynamicMaterialParameters(
    Index 0 Param 1=&FloatFromCurve(FloatCurve=DataInterface() {propertyValues {Curve {keys [<at -0.0046 2>,<at 1.0011 4>]}}})
    Index 0 Param 2=&FloatFromCurve(FloatCurve=DataInterface() {propertyValues {Curve {keys [<at -0.0046 0.4>,<at 1.0011 1.1>]}}})
    Index 0 Param 3=&FloatFromCurve(FloatCurve=DataInterface() {propertyValues {Curve {keys [<at -0.0046 1>,<at 1.0011 0.5>]}}})
    Index 2 Param 1=&FloatFromCurve(FloatCurve=DataInterface() {propertyValues {Curve {keys [<at -0.0046 0.4>,<at 1.0011 1.5>]}}})
    Index 2 Param 2=&FloatFromCurve(FloatCurve=DataInterface() {propertyValues {Curve {keys [<at -0.0046 0.1>,<at 0.2854 2.6>,<at 1.0011 0.3>]}}})
    Index 2 Param 3=0.8
    Index 2 Param 4=&FloatFromCurve(FloatCurve=DataInterface() {propertyValues {Curve {keys [
         {arriveTangent=-1.9991, interpMode=RCIM_Cubic, time=0, value=1}
         {arriveTangent=-2.7143, interpMode=RCIM_Cubic, leaveTangent=-2.7143, tangentMode=RCTM_User, time=0.2095, value=0.6901}
         <at 1 0 RCIM_Cubic>
        ]}}})
    Module.Param0WriteEnabled=true
    Module.Param0WriteEnabled2=true
    Module.Param1WriteEnabled=true
    Module.Param1WriteEnabled2=true
    Module.Param2WriteEnabled=true
    Module.Param2WriteEnabled2=true
    Module.Param3WriteEnabled=true
    Module.Param3WriteEnabled2=true
    Write Parameter Index 2=true
   ) {module=DynamicMaterialParameters001}
   &ScaleColor(
    Scale Alpha=&FloatFromCurve(FloatCurve=DataInterface() {propertyValues {Curve {keys [<at -0.0046 1.0113 RCIM_Cubic>,<at 1.0011 0.002 RCIM_Cubic>]}}})
    ScaleA=true
    ScaleRGB=true
   ) {enabled=false}
   &Drag(Drag=2, Use Linear Drag=true, Use Rotational Drag=false)
   &SolveForcesAndVelocity(Clamp Velocity=false, Limit Acceleration=false)
   &ScaleMeshSize() {enabled=false, module=MeshSizeScale}
   &ConstrainPositionToPlane(Plane Normal:Vector3f=$Particles.WaterPlaneNormal, Plane Position:Vector3f=$Particles.WaterPlaneLocation) {enabled=false}
  ]
  rendererClasses [&NiagaraMeshRendererProperties]
  renderers [{rendererClass=&NiagaraMeshRendererProperties, rendererIndex=0}]
 }
 Generic_Rotor_Clouds {
  emitterUpdateScript [
   &EmitterState(Life Cycle Mode=::NewEnumerator1, Loop Duration=1, UseLoopDelay=false)
   &SpawnRate(
    Spawn Probability=&Multiply_Float(A=$User.RotorPlaySpeed, B=&OneMinusFloat(Float=$User.Altitude))
    SpawnRate=100
    Use Spawn Probability=true
   )
  ]
  particleSpawnScript [
   &InitializeParticle(
    Color=<rgba 1 1 1 0.15>
    Lifetime=&UniformRangedFloat(Minimum=0.4, Maximum=0.75)
    Sprite Rotation Mode=::NewEnumerator2
    Sprite Size Mode=::NewEnumerator3
    Uniform Sprite Size=&RandomRangeFloat(Minimum=400, Maximum=500)
    UsePositionOffset=false
    Write Material Random=true
   ) {module=InitializeParticle001}
   &TorusLocation(
    Align Torus With Axis=true
    Coordinate Space:ENiagaraCoordinateSpace=::World
    Large Radius:NiagaraFloat=&Lerp_Float(A=$User.TorusInnerRadius, B=&Multiply_Float(A=$User.TorusInnerRadius, B=2), Alpha=$User.Altitude)
    MaskSpawnGroup=false
    Radius Position=0.25
    Torus Axis=$User.ImpactNormal
    Torus Distribution Mode:ENiagaraTorusDistributionMode=::NewEnumerator1
    Torus Origin:NiagaraPosition=/Niagara/DynamicInputs/Helpers/SimulationPosition()
    U Position:NiagaraFloat=&Add_Float(
     A=$User.Theta
     B=&AngleConversion(
      Angle:NiagaraFloat=&UniformRangedFloat(Minimum=-2, Maximum=2)
      Angle Input:ENiagara_AngleInput=::NewEnumerator0
      Angle Output:ENiagara_AngleInput=::NewEnumerator2
     )
    )
    V Position=0.05
   )
   &FN_AlignTorusPosition() {module=AlignTorusPosition}
   &AddVelocity(Velocity=&Multiply_Vector(
     A=&Multiply_VectorByFloat(
      Vector=&TransformVector(Vector=$Output.TorusLocation.TorusNormal, Source Space=::World, Destination Space=::Local)
      Float=&Multiply_Float(A=&UniformRangedFloat(Minimum=300, Maximum=3000), B=$User.RotorPlaySpeed)
     )
     B=<1 1 0>
    ))
   &CameraOffset(Camera Offset Amount=&UniformRangedFloat(Minimum=10, Maximum=30))
  ]
  particleUpdateScript [
   &ParticleState()
   &Drag(Drag=6, Use Linear Drag=true, Use Rotational Drag=false)
   &SolveForcesAndVelocity(Clamp Velocity=false, Limit Acceleration=false)
   /SetVariables(Particles.SubImageIndex=&FloatFromCurve(FloatCurve=DataInterface() {propertyValues {Curve {keys [{arriveTangent=84.4827, interpMode=RCIM_Cubic, time=0, value=0}, <at 1 63 RCIM_Cubic>]}}})) {enabled=false, isSetParameters=true}
   &SpriteRotationRate(Rotation Rate=&UniformRangedFloat(Minimum=-1, Maximum=1))
   &ScaleColor(
    Scale Alpha=&FloatFromCurve(FloatCurve=DataInterface() {propertyValues {Curve {keys [<at 0 0 RCIM_Cubic>, {interpMode=RCIM_Cubic, tangentMode=RCTM_User, time=0.4, value=1}, <at 1 0 RCIM_Cubic>]}}})
    ScaleA=true
    ScaleRGB=true
   )
   &DynamicMaterialParameters(
    Index 0 Param 1=&Add_Float(
     A=&FloatFromCurve(FloatCurve=DataInterface() {propertyValues {Curve {keys [{arriveTangent=2.1549, time=0, value=0.8}, <at 1 1>]}}})
     B=&Multiply_Float(A=$Particles.Initial.MaterialRandom, B=0.5)
    )
    Index 1 Param 1=1
    Index 1 Param 2=1
    Index 1 Param 3=1
    Index 1 Param 4=1
    Index 2 Param 1=&FloatFromCurve(FloatCurve=DataInterface() {propertyValues {Curve {keys [
         {arriveTangent=1.8288, interpMode=RCIM_Cubic, leaveTangent=1.8288, tangentMode=RCTM_User, time=0, value=0.3}
         {arriveTangent=-0.0004, interpMode=RCIM_Cubic, leaveTangent=-0.0004, tangentMode=RCTM_User, time=1, value=1}
        ]}}})
    Index 2 Param 2=&RandomRangeFloat(Minimum=0.5, Maximum=1)
    Index 2 Param 3=4
    Index 2 Param 4=&RandomRangeFloat(Minimum=0.8, Maximum=1.2)
    Module.Param0WriteEnabled=true
    Module.Param0WriteEnabled1=true
    Module.Param0WriteEnabled2=true
    Module.Param0WriteEnabled3=true
    Module.Param1WriteEnabled=true
    Module.Param1WriteEnabled1=true
    Module.Param1WriteEnabled2=true
    Module.Param1WriteEnabled3=true
    Module.Param2WriteEnabled=true
    Module.Param2WriteEnabled1=true
    Module.Param2WriteEnabled2=true
    Module.Param2WriteEnabled3=true
    Module.Param3WriteEnabled=true
    Module.Param3WriteEnabled1=true
    Module.Param3WriteEnabled2=true
    Module.Param3WriteEnabled3=true
    Write Parameter Index 1=true
    Write Parameter Index 2=true
    Write Parameter Index 3=true
   ) {module=DynamicMaterialParameters002}
  ]
  rendererClasses [&NiagaraSpriteRendererProperties]
  renderers [
   {rendererClass=&NiagaraSpriteRendererProperties}
   {rendererClass=&NiagaraSpriteRendererProperties, rendererIndex=1}
  ]
 }
 MistClouds_withRayTraceTorus {
  emitterUpdateScript [
   &EmitterState()
   &SpawnRate(
    Spawn Probability=&MakeCustomFloatFromBool(
     Bool=&Boolean_AndOperation(
      Bool1=&FN_PhysicalMaterialBoolComparison(Index:NiagaraInt32=2, Physical Material Index:NiagaraInt32=$User.PhysMatIndex)
      Bool2=&Boolean_AndOperation(Bool1=$User.hasDriver, Bool2=$User.isRotorWashActive)
     )
     True Float=1
     False Float=0
    )
    SpawnRate=&Multiply_Float(A=$User.RotorPlaySpeed, B=25)
    Use Spawn Probability=true
   )
   &SpawnPerFrame(
    Spawn:NiagaraBool=&Boolean_AndOperation(
     Bool1=&FN_PhysicalMaterialBoolComparison(Index:NiagaraInt32=2, Physical Material Index:NiagaraInt32=$User.PhysMatIndex)
     Bool2=&Boolean_AndOperation(Bool1=$User.hasDriver, Bool2=$User.isRotorWashActive)
    )
    Spawn Count:NiagaraInt32=&MakeIntFromFloat(Float=&Multiply_Float(
      A=$User.RotorPlaySpeed
      B=&Multiply_Float(A=1, B=&OneMinusFloat(Float=$User.Altitude))
     ))
    Use Spawn Probability=false
   ) {enabled=false}
  ]
  particleSpawnScript [
   /Niagara/Modules/Spawn/Initialization/InitializeParticle(
    Color=&MakeLinearColorFromVectorAndFloat(
     "Vector (RGB)"=&MakeVectorFromLinearColorRGB(Color=<rgba 1 1 1 1>)
     "Float (Alpha)"=&UniformRangedFloat(Minimum=0.5, Maximum=1)
    )
    Lifetime=&UniformRangedFloat(Minimum=0.4, Maximum=0.75)
    Module.Write Color=true
    Module.Write Lifetime=true
    Module.Write Mass=true
    Module.Write Material Random=true
    Module.Write Position=false
    Module.Write Scale=false
    Module.Write SpriteRotation=true
    Module.Write SpriteSize=true
    Position=/Niagara/Functions/Localspace/SimulationPosition()
    Sprite Rotation=&NormalizedAngleToDegrees(NormalizedAngle=&UniformRangedFloat(Minimum=0, Maximum=1))
    Sprite Size=&Vector2DFromFloat(Value=&UniformRangedFloat(Minimum=200, Maximum=500))
   )
   &TorusLocation(
    Align Torus With Axis=true
    Coordinate Space:ENiagaraCoordinateSpace=::World
    Large Radius:NiagaraFloat=&Lerp_Float(A=$User.TorusInnerRadius, B=&Multiply_Float(A=$User.TorusInnerRadius, B=2), Alpha=$User.Altitude)
    MaskSpawnGroup=false
    Radius Position=0.2752
    Torus Axis=$User.ImpactNormal
    Torus Distribution Mode:ENiagaraTorusDistributionMode=::NewEnumerator1
    Torus Origin:NiagaraPosition=/Niagara/DynamicInputs/Helpers/SimulationPosition()
    U Position:NiagaraFloat=&Add_Float(
     A=$User.Theta
     B=&DegreesToRadians(AngleInDegrees:NiagaraFloat=&UniformRangedFloat(Minimum=-2, Maximum=2))
    )
    V Position=0.05
   )
   &FN_AlignTorusPosition() {module=AlignTorusPosition}
   &AddVelocity(Velocity=&Multiply_VectorByFloat(
     Vector=&TransformVector(Vector=$Output.TorusLocation.TorusNormal, Source Space=::World, Destination Space=::Local)
     Float=&Multiply_Float(A=&UniformRangedFloat(Minimum=750, Maximum=3000), B=$User.RotorPlaySpeed)
    ))
   &CameraOffset(Camera Offset Amount=25)
   &DynamicMaterialParameters(
    Index 0 Param 1=&UniformRangedFloat(Minimum=1, Maximum=2)
    Index 0 Param 2=&UniformRangedFloat(Minimum=1, Maximum=3)
    Index 0 Param 3=1
    Module.Param0WriteEnabled=true
    Module.Param1WriteEnabled=true
    Module.Param2WriteEnabled=true
    Module.Param3WriteEnabled=true
   ) {module=DynamicMaterialParameters001}
  ]
  particleUpdateScript [
   &ParticleState()
   &Drag(Drag=&UniformRangedFloat(Minimum=6, Maximum=7), Use Linear Drag=true, Use Rotational Drag=false)
   &SolveForcesAndVelocity(Clamp Velocity=false, Limit Acceleration=false)
   &ScaleColor(
    Scale Alpha=&FloatFromCurve(FloatCurve=DataInterface() {propertyValues {Curve {keys [<at 0 0 RCIM_Cubic>,<at 0.5 1 RCIM_Cubic>,<at 1 0 RCIM_Cubic>]}}})
    Scale RGB=&MakeVectorFromLinearColorRGB(Color=&ColorFromCurve(ColorCurve=DataInterface() {propertyValues {
        AlphaCurve {keys [<at -0.0023 1 RCIM_Cubic>]}
        BlueCurve {keys [<at -0.002 1 RCIM_Cubic>,<at 0.5 1 RCIM_Cubic>]}
        GreenCurve {keys [<at -0.002 0.679 RCIM_Cubic>, {arriveTangent=2.3186, interpMode=RCIM_Cubic, leaveTangent=2.3186, tangentMode=RCTM_User, time=0.5, value=1}]}
        RedCurve {keys [<at -0.002 0.3021 RCIM_Cubic>, {arriveTangent=4.4899, interpMode=RCIM_Cubic, leaveTangent=4.4899, tangentMode=RCTM_User, time=0.5, value=1}]}
       }}))
    ScaleA=true
    ScaleRGB=true
   )
   &ScaleSpriteSize(Scale Factor=&Vector2DFromFloat(Value=&FloatFromCurve(FloatCurve=DataInterface() {propertyValues {Curve {keys [
          {arriveTangent=4.8646, interpMode=RCIM_Cubic, leaveTangent=4.8646, tangentMode=RCTM_User, time=0, value=0}
          {arriveTangent=1.2691, interpMode=RCIM_Cubic, leaveTangent=1.2691, tangentMode=RCTM_User, time=0.3098, value=0.9716}
          <at 1 1.2476 RCIM_Cubic>
         ]}}})))
   /SetVariables(Particles.SubImageIndex=&Add_Float(A=$Particles.SubImageIndex, B=&Multiply_Float(A=$Engine.DeltaTime, B=10))) {isSetParameters=true}
   &DynamicMaterialParameters(
    Index 0 Param 2=4
    Index 0 Param 3=0.75
    Index 0 Param 4=0.01
    Module.Param0WriteEnabled=false
    Module.Param1WriteEnabled=true
    Module.Param2WriteEnabled=true
    Module.Param3WriteEnabled=true
   )
  ]
  rendererClasses [&NiagaraSpriteRendererProperties]
  renderers [{rendererClass=&NiagaraSpriteRendererProperties}]
 }
 Splashes {
  emitterUpdateScript [
   &EmitterState()
   &EmitterScalability(
    MaxDistance=12000
    Module.bApplySpawnCountScale=true
    Module.bApplySpawnCountScaleByDistanceFraction=true
    Module.bUseMaxDistance=true
    Module.bUseMinDistance=true
    SpawnCountScaleByDistanceFraction=DataInterface() {propertyValues {Curve {keys [<at 0.1 0.5>,<at 0.5 1>,<at 1 0>]}}}
   ) {enabled=false}
   &EmitterLifeCycle() {enabled=false}
   &SpawnPerFrame(
    Spawn:NiagaraBool=&Boolean_AndOperation(
     Bool1=&FN_PhysicalMaterialBoolComparison(Index:NiagaraInt32=2, Physical Material Index:NiagaraInt32=$User.PhysMatIndex)
     Bool2=&Boolean_AndOperation(Bool1=$User.hasDriver, Bool2=$User.isRotorWashActive)
    )
    Spawn Count:NiagaraInt32=&MakeIntFromFloat(Float=&Multiply_Float(
      A=$User.RotorPlaySpeed
      B=&Multiply_Float(A=5, B=&OneMinusFloat(Float=$User.Altitude))
     ))
    Use Spawn Probability=false
   ) {enabled=false}
   &SpawnRate(
    Spawn Probability=&MakeCustomFloatFromBool(
     Bool=&Boolean_AndOperation(
      Bool1=&FN_PhysicalMaterialBoolComparison(Index:NiagaraInt32=2, Physical Material Index:NiagaraInt32=$User.PhysMatIndex)
      Bool2=&Boolean_AndOperation(Bool1=$User.hasDriver, Bool2=$User.isRotorWashActive)
     )
     True Float=1
     False Float=0
    )
    SpawnRate=&Multiply_Float(
     A=$User.RotorPlaySpeed
     B=&Multiply_Float(
      A=&UniformRangedFloat(Minimum=60, Maximum=80)
      B=&OneMinusFloat(Float=$User.Altitude)
     )
    )
    Use Spawn Probability=true
   )
  ]
  particleSpawnScript [
   &InitializeParticle(
    Lifetime=&UniformRangedFloat(Minimum=1, Maximum=1.5)
    Mass=&UniformRangedFloat(Minimum=0.6, Maximum=1.2)
    Mesh Scale=&VectorFromFloat(Value=&UniformRangedFloat(Minimum=0.2, Maximum=0.6))
    Module.Write Color=true
    Module.Write Lifetime=true
    Module.Write Mass=true
    Module.Write Material Random=true
    Module.Write Position=true
    Module.Write Scale=true
    Module.Write SpriteRotation=false
    Module.Write SpriteSize=false
    Position=/Niagara/DynamicInputs/Transforms/TransformPosition(
     Position=&Multiply_VectorByFloat(Vector=&UniformRangedVector(Minimum=<-200 50 -20>, Maximum=<200 50 -20>), Float=$User.Width)
     Source Space=::Local
     Destination Space=::World
    )
    Sprite Rotation=&NormalizedAngleToDegrees(NormalizedAngle=&UniformRangedFloat(Minimum=0, Maximum=1))
   )
   &TorusLocation(
    Align Torus With Axis=true
    Coordinate Space:ENiagaraCoordinateSpace=::World
    Large Radius:NiagaraFloat=&Lerp_Float(A=$User.TorusInnerRadius, B=&Multiply_Float(A=$User.TorusInnerRadius, B=2), Alpha=$User.Altitude)
    MaskSpawnGroup=false
    Radius Position=0.2752
    Torus Axis=$User.ImpactNormal
    Torus Distribution Mode:ENiagaraTorusDistributionMode=::NewEnumerator1
    Torus Origin:NiagaraPosition=&SimulationPosition()
    U Position:NiagaraFloat=&Add_Float(
     A=$User.Theta
     B=&DegreesToRadians(AngleInDegrees:NiagaraFloat=&UniformRangedFloat(Minimum=-2, Maximum=2))
    )
    V Position=0.05
   )
   &FN_AlignTorusPosition()
   &AddVelocity(Velocity=&Multiply_VectorByFloat(
     Vector=&TransformVector(Vector=$Output.TorusLocation.TorusNormal, Source Space=::World, Destination Space=::Local)
     Float=&Multiply_Float(A=&UniformRangedFloat(Minimum=50, Maximum=100), B=$User.RotorPlaySpeed)
    ))
   /Niagara/Modules/Spawn/Orientation/V2/InitialMeshOrientation(
    Orientation Vector=&NormalizeVector(Vector To Normalize=$Particles.Velocity)
    Rotation=&UniformRangedVector(Minimum=<0 0 0>, Maximum=<0 0 1>)
    Use Orientation Vector=true
    Use Rotation Vector=false
   )
   &DynamicMaterialParameters(Write Parameter Index 0=false)
   &InitializeWaterSnapToOwner()
  ]
  particleUpdateScript [
   &ParticleState()
   &DynamicMaterialParameters(
    Index 0 Param 1=&FloatFromCurve(FloatCurve=DataInterface() {propertyValues {Curve {keys [
         {arriveTangent=1.4694, interpMode=RCIM_Cubic, leaveTangent=1.4694, tangentMode=RCTM_User, time=-0.0005, value=-1.9726}
         {arriveTangent=1.9693, interpMode=RCIM_Cubic, leaveTangent=1.9693, time=0.7887, value=-0.7711}
         {arriveTangent=3.1029, interpMode=RCIM_Cubic, leaveTangent=3.1029, tangentMode=RCTM_User, time=1.0011, value=0}
        ]}}})
    Index 0 Param 2=&FloatFromCurve(FloatCurve=DataInterface() {propertyValues {Curve {keys [<at -0.0046 4>,<at 1.0011 0>]}}})
    Index 0 Param 3=&FloatFromCurve(
     FloatCurve=DataInterface() {propertyValues {Curve {keys [
         {arriveTangent=-0.0014, interpMode=RCIM_Cubic, leaveTangent=-0.0014, tangentMode=RCTM_User, time=-0.0046, value=1}
         {arriveTangent=-1.0017, interpMode=RCIM_Cubic, leaveTangent=-1.0017, time=0.8759, value=0.5}
         <at 0.9937 0>
        ]}}}
     Scale Curve=&MakeFloatFromVector(Vector=$Particles.Scale, Channel=::NewEnumerator0)
    )
    Index 0 Param 4=1
    Module.Param0WriteEnabled=true
    Module.Param1WriteEnabled=true
    Module.Param2WriteEnabled=true
    Module.Param3WriteEnabled=true
   ) {module=DynamicMaterialParameters001}
   &ScaleColor(
    Scale Alpha=&FloatFromCurve(FloatCurve=DataInterface() {propertyValues {Curve {keys [<at -0.0046 1.0113 RCIM_Cubic>,<at 1.0011 0.002 RCIM_Cubic>]}}})
    ScaleA=true
    ScaleRGB=false
   )
   &SolveForcesAndVelocity(Clamp Velocity=false, Limit Acceleration=false)
   &ScaleMeshSize(Scale Factor=&VectorFromCurve(VectorCurve=DataInterface() {propertyValues {
       XCurve {keys [<at -0.0011 1.0067>,<at 0.9989 3>]}
       YCurve {keys [<at -0.0011 1.0067>,<at 0.9989 3>]}
       ZCurve {keys [<at -0.0011 1.1211>,<at 0.9989 1.5622>]}
      }})) {enabled=false, module=MeshSizeScale}
   &ConstrainPositionToPlane(Plane Normal:Vector3f=$Particles.WaterPlaneNormal, Plane Position:Vector3f=$Particles.WaterPlaneLocation) {enabled=false}
  ]
  rendererClasses [&NiagaraMeshRendererProperties]
  renderers [{rendererClass=&NiagaraMeshRendererProperties, rendererIndex=0}]
 }
}
summary {
 systemName=NS_TidalCrane_Hoagie_RotorWashBurst
 emitters [
  {emitterName=Debris_Stone, rendererClasses [&NiagaraSpriteRendererProperties]}
  {emitterName=MistClouds_withRayTraceTorus, rendererClasses [&NiagaraSpriteRendererProperties]}
  {emitterName=Bubbles, rendererClasses [&NiagaraMeshRendererProperties]}
  {emitterName=Foam_Detail, rendererClasses [&NiagaraMeshRendererProperties]}
  {emitterName=Splashes, rendererClasses [&NiagaraMeshRendererProperties]}
  {emitterName=DebirsDirt, rendererClasses [&NiagaraSpriteRendererProperties]}
  {emitterName=Debris_Grass, rendererClasses [&NiagaraMeshRendererProperties]}
  {emitterName=Generic_Rotor_Clouds, rendererClasses [&NiagaraSpriteRendererProperties]}
 ]
 userVariables [
  User.Altitude {type=NiagaraFloat, defaultValue=0, description=""}
  User.hasDriver {type=NiagaraBool, defaultValue=true}
  User.ImpactNormal {type=Vector3f, description="", defaultValue {struct=Vector3f, value=<0 0 1>}}
  User.isRotorWashActive {type=NiagaraBool, defaultValue=true}
  User.PhysMatIndex {type=NiagaraInt32}
  User.RotorPlaySpeed {type=NiagaraFloat, defaultValue=1, description=""}
  User.Theta {type=NiagaraFloat, defaultValue=0, description=""}
  User.TorusInnerRadius {type=NiagaraFloat, defaultValue=500, description=""}
  User.WaterBodyIndex {type=NiagaraFloat, defaultValue=0, description=""}
  User.WaterDepth {type=NiagaraFloat, defaultValue=0, description=""}
  User.WaterPlaneLocation {type=Vector3f, description="", defaultValue {struct=Vector3f, value=<0 0 0>}}
  User.WaterPlaneNormal {type=Vector3f, description="", defaultValue {struct=Vector3f, value=<0 0 0>}}
  User.WaveAttenuationFactor {type=NiagaraFloat, defaultValue=0, description=""}
  User.WaveTime {type=NiagaraFloat, defaultValue=0, description=""}
  User.Width {type=NiagaraFloat, defaultValue=0, description=""}
 ]
}
