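# NS_TidalCrane_Scraps is a Niagara effect for the Tidal Crane, throwing hot orange sparks with flash, glow and smoke.
# Sparks fall under gravity and short streaks fly outward, both using speed-stretched sprites with randomized HDR brightness.
# A panning-material mesh flash shares System.VecDir with the spark cone, alongside eroding flipbook smoke and a light-plus-sprite glow.
# All emitters burst, all but the glow also spawn continuously, and no user parameters are exposed.
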NiagaraSystem /TidalCraneHoagie/Assets/FX/NS_TidalCrane_Scraps
dependencies {
 usedDataInterfaces [&NiagaraDataInterfaceCurve, &NiagaraDataInterfaceVectorCurve]
 usedDynamicInputs [&RandomRangeVector]
 usedModules [
  &SystemState
  &self:SystemSpawnScript.SystemScriptSource.SystemScriptGraph.NiagaraNodeAssignment_0.SetVariables_#1
 ]
}
emitters {
 Air_Glow {
  emitterUpdateScript [
   &EmitterState()
   &SpawnBurst_Instantaneous(Spawn Count=1, Use Spawn Probability=false)
  ]
  particleSpawnScript [
   /Niagara/Modules/Spawn/Initialization/InitializeParticle(
    Lifetime=&UniformRangedFloat(Minimum=0.12, Maximum=0.2)
    Module.Write Color=true
    Module.Write Lifetime=true
    Module.Write Mass=true
    Module.Write Material Random=true
    Module.Write Position=false
    Module.Write Scale=false
    Module.Write SpriteRotation=true
    Module.Write SpriteSize=true
    Position=&SimulationPosition()
    Sprite Rotation=&UniformRangedFloat(Minimum=0, Maximum=360)
    Sprite Size=&Vector2DFromFloat(Value=&UniformRangedFloat(Minimum=300, Maximum=600))
   )
   &CameraOffset(Camera Offset Amount=75)
   /SetVariables(Particles.LightExponent=2) {isSetParameters=true}
  ]
  particleUpdateScript [
   &ParticleState() {module=UpdateAge}
   &ScaleSpriteSize(Scale Factor=&Vector2DFromFloat(Value=&FloatFromCurve(FloatCurve=DataInterface() {propertyValues {Curve {keys [
          <at 0.0013 0.2913>
          {arriveTangent=1.5081, interpMode=RCIM_Cubic, leaveTangent=1.5081, tangentMode=RCTM_User, time=0.1578, value=0.6688}
          <at 0.9941 1.0318>
         ]}}})))
   &Color(Color=<rgba 15 3 0.3 0.05>)
   &DynamicMaterialParameters(
    Index 0 Param 1=&FloatFromCurve(FloatCurve=DataInterface() {propertyValues {Curve {keys [<at 0 5>,<at 1 15>]}}})
    Index 0 Param 2=1
    Module.Param0WriteEnabled=true
    Module.Param1WriteEnabled=true
    Module.Param2WriteEnabled=true
    Module.Param3WriteEnabled=true
   )
  ]
  rendererClasses [&NiagaraLightRendererProperties, &NiagaraSpriteRendererProperties]
  renderers [
   {rendererClass=&NiagaraLightRendererProperties, rendererIndex=0}
   {rendererClass=&NiagaraSpriteRendererProperties, rendererIndex=1}
  ]
 }
 Smoke {
  emitterUpdateScript [
   &EmitterState()
   &SpawnRate(SpawnRate=&UniformRangedFloat(Minimum=20, Maximum=30), Use Spawn Probability=false)
   &SpawnBurst_Instantaneous(Spawn Count=1, Use Spawn Probability=false)
  ]
  particleSpawnScript [
   &InitializeParticle(
    Color=&MakeLinearColorFromVectorAndFloat(
     "Vector (RGB)"=&MakeVectorFromLinearColorRGB(Color=<rgba 1 0.6248 0.1506 1>)
     "Float (Alpha)"=&UniformRangedFloat(Minimum=0.1, Maximum=0.2)
    )
    Lifetime=&UniformRangedFloat(Minimum=0.33, Maximum=0.44)
    Module.Write Color=true
    Module.Write Lifetime=true
    Module.Write Mass=true
    Module.Write Material Random=true
    Module.Write Position=false
    Module.Write Scale=false
    Module.Write SpriteRotation=true
    Module.Write SpriteSize=true
    Position=&SimulationPosition()
    Sprite Rotation=&NormalizedAngleToDegrees(NormalizedAngle=&UniformRangedFloat(Minimum=0, Maximum=1))
    Sprite Size=&Vector2DFromFloat(Value=&UniformRangedFloat(Minimum=200, Maximum=400))
   )
   /SetVariables(Particles.InitSpriteRotRate=&UniformRangedFloat(Minimum=1, Maximum=2)) {isSetParameters=true}
   &AddVelocityInCone(
    Module.Use Velocity Falloff On Cone Axis=false
    Velocity Strength=&UniformRangedFloat(Minimum=100, Maximum=400)
   )
   &CameraOffset(Camera Offset Amount=50)
  ]
  particleUpdateScript [
   &ParticleState() {module=UpdateAge}
   &SpriteRotationRate(Rotation Rate=&Multiply_Float(
     A=$Particles.InitSpriteRotRate
     B=&FloatFromCurve(FloatCurve=DataInterface() {propertyValues {Curve {keys [
          <at -0.0046 1.0113 RCIM_Cubic>
          {arriveTangent=-0.0811, interpMode=RCIM_Cubic, leaveTangent=-0.0811, tangentMode=RCTM_User, time=1.0015, value=0.0948}
         ]}}})
    ))
   /Niagara/Modules/Update/SubUV/SubUVAnimation(Number Of Frames=64)
   &SolveForcesAndVelocity(Clamp Velocity=false, Limit Acceleration=false)
   &Color(
    Color=$Particles.Initial.Color
    Scale Alpha=&FloatFromCurve(FloatCurve=DataInterface() {propertyValues {Curve {keys [
         <at 0 0>
         {arriveTangent=1.4566, interpMode=RCIM_Cubic, leaveTangent=1.4566, time=0.1629, value=0.7449}
         {arriveTangent=-0.8863, interpMode=RCIM_Cubic, leaveTangent=-0.8863, time=0.6194, value=0.9022}
         <at 1.0011 0.002>
        ]}}})
    Scale Color=&VectorFromFloat(Value=&FloatFromCurve(FloatCurve=DataInterface() {propertyValues {Curve {keys [
          {arriveTangent=-6.2795, interpMode=RCIM_Cubic, leaveTangent=-6.2795, tangentMode=RCTM_User, time=0, value=1}
          {arriveTangent=-0.3209, interpMode=RCIM_Cubic, leaveTangent=-0.3209, tangentMode=RCTM_User, time=0.15, value=0.2}
          <at 1 0.05 RCIM_Cubic>
         ]}}}))
   )
   &ScaleSpriteSize(Scale Factor=&Vector2DFromFloat(Value=&FloatFromCurve(FloatCurve=DataInterface() {propertyValues {Curve {keys [
          {arriveTangent=2.357, interpMode=RCIM_Cubic, leaveTangent=2.357, tangentMode=RCTM_User, time=-0.0056, value=0.4822}
          {arriveTangent=0.5334, interpMode=RCIM_Cubic, leaveTangent=0.5334, time=0.1744, value=0.8307}
          <at 1.0016 1.0194>
         ]}}})))
   /Game/Effects/Niagara/Modules/Materials/MaterialSmoke(Erode:NiagaraFloat=&FloatFromCurve(FloatCurve=DataInterface() {propertyValues {Curve {keys [<at -0.0046 1.0113 RCIM_Cubic>,<at 1.0027 -0.9316 RCIM_Cubic>]}}}))
  ]
  rendererClasses [&NiagaraSpriteRendererProperties]
  renderers [{rendererClass=&NiagaraSpriteRendererProperties}]
 }
 Sparks {
  emitterUpdateScript [
   &EmitterState()
   &SpawnBurst_Instantaneous(Spawn Count=&UniformRangedInt(Minimum=32, Maximum=64), Spawn Time=0.03, Use Spawn Probability=false)
   &SpawnRate(SpawnRate=100, Use Spawn Probability=false)
  ]
  particleSpawnScript [
   &InitializeParticle(
    Color=<rgba 1 0.2 0.02 1>
    Lifetime=&UniformRangedFloat(Minimum=0.2, Maximum=1.4)
    Module.Write Color=true
    Module.Write Lifetime=true
    Module.Write Mass=true
    Module.Write Material Random=true
    Module.Write Position=false
    Module.Write Scale=false
    Module.Write SpriteRotation=false
    Module.Write SpriteSize=true
    Position=&SimulationPosition()
    Sprite Size=&UniformRangedVector2D(Minimum=<3 8>, Maximum=<5 16>)
   )
   /Niagara/Modules/Spawn/Location/SphereLocation(Hemisphere X=true, MaskSpawnGroup=false, Module.Override Randomness=false, Sphere Radius=5, Surface Only=false)
   &AddVelocity(
    Cone Angle=25
    Cone Axis=$System.VecDir
    Module.UseConeFalloff=false
    Module.UseVelDistribution=false
    Override Seed=false
    Velocity Mode=::NewEnumerator2
    Velocity Speed=&Lerp_Float(
     A=500
     B=2500
     Alpha=&FloatFromCurve(
      FloatCurve=DataInterface() {propertyValues {Curve {keys [<at 0 0 RCIM_Cubic>, {arriveTangent=3.3672, interpMode=RCIM_Cubic, leaveTangent=3.3672, tangentMode=RCTM_User, time=1, value=1}]}}}
      CurveIndex=&RandomRangeFloat(Minimum=0, Maximum=1)
     )
    )
   ) {module=AddVelocity001}
   &ApplyInitialForces()
   &CameraOffset(Camera Offset Amount=50)
  ]
  particleUpdateScript [
   &ParticleState()
   &AccelerationForce(Acceleration=<0 0 -1000>)
   &SolveForcesAndVelocity(Clamp Velocity=false, Limit Acceleration=false)
   &ScaleColor(
    Scale RGB=&Multiply_VectorByFloat(
     Vector=&Lerp_Vector(A=<1 1 1>, B=<0.3333 0.026 0.1205>, Alpha=$Particles.NormalizedAge)
     Float=&Lerp_Float(
      A=10
      B=50
      Alpha=&FloatFromCurve(
       FloatCurve=DataInterface() {propertyValues {Curve {keys [<at 0 0 RCIM_Cubic>, {arriveTangent=2.3837, interpMode=RCIM_Cubic, leaveTangent=2.3837, tangentMode=RCTM_User, time=1, value=1}]}}}
       CurveIndex=&RandomRangeFloat(Minimum=0, Maximum=1)
      )
     )
    )
    ScaleA=true
    ScaleRGB=true
   )
   &ScaleSpriteSizeBySpeed(Max Scale Factor=&MakeVector2D(
     X=1
     Y=&Lerp_Float(
      A=1
      B=4
      Alpha=&FloatFromCurve(
       FloatCurve=DataInterface() {propertyValues {Curve {keys [<at 0 0 RCIM_Cubic>, {arriveTangent=2.3837, interpMode=RCIM_Cubic, leaveTangent=2.3837, tangentMode=RCTM_User, time=1, value=1}]}}}
       CurveIndex=&RandomRangeFloat(Minimum=0, Maximum=1)
      )
     )
    ))
  ]
  rendererClasses [&NiagaraSpriteRendererProperties]
  renderers [{rendererClass=&NiagaraSpriteRendererProperties}]
 }
 Streaks {
  emitterUpdateScript [
   &EmitterState()
   &SpawnBurst_Instantaneous(Spawn Count=&UniformRangedInt(Minimum=10, Maximum=20), Spawn Time=0.03, Use Spawn Probability=false)
   &SpawnRate(SpawnRate=50, Use Spawn Probability=false)
  ]
  particleSpawnScript [
   &InitializeParticle(
    Color=<rgba 1 0.2 0.02 1>
    Lifetime=&UniformRangedFloat(Minimum=0.08, Maximum=0.15)
    Module.Write Color=true
    Module.Write Lifetime=true
    Module.Write Mass=true
    Module.Write Material Random=true
    Module.Write Position=false
    Module.Write Scale=false
    Module.Write SpriteRotation=false
    Module.Write SpriteSize=true
    Position=&SimulationPosition()
    Sprite Size=&UniformRangedVector2D(Minimum=<2 35>, Maximum=<4 120>)
   )
   &SphereLocation(Hemisphere X=true, MaskSpawnGroup=false, Module.Override Randomness=false, Sphere Radius=3, Surface Only=false)
   &AddVelocity(Velocity=&UniformRangedVector(Minimum=<0 0 0>, Maximum=<100 0 0>))
   &AddVelocityFromPoint(Module.Use Falloff Distance=false, Velocity Origin:Vector3f=&SimulationPosition(), Velocity Strength=320)
   &CameraOffset(Camera Offset Amount=50)
  ]
  particleUpdateScript [
   &ParticleState()
   &Drag(Drag=2, Use Linear Drag=true, Use Rotational Drag=false)
   &SolveForcesAndVelocity(Clamp Velocity=false, Limit Acceleration=false)
   &ScaleSpriteSize(Scale Factor=&MakeVector2D(
     X=&FloatFromCurve(FloatCurve=DataInterface() {propertyValues {Curve {keys [<at 0 1.5 RCIM_Cubic>, {arriveTangent=-0.5, interpMode=RCIM_Cubic, leaveTangent=-0.5, time=0.3019, value=1.2624}, <at 1 1 RCIM_Cubic>]}}})
     Y=&FloatFromCurve(FloatCurve=DataInterface() {propertyValues {Curve {keys [
          {arriveTangent=2.108, interpMode=RCIM_Cubic, leaveTangent=2.108, tangentMode=RCTM_User, time=0, value=1}
          {arriveTangent=-1, interpMode=RCIM_Cubic, leaveTangent=-1, time=0.3019, value=1.1}
          <at 1 0>
         ]}}})
    ))
   &ScaleSpriteSizeBySpeed(Max Scale Factor=<1 2>)
   &ScaleColor(
    Scale RGB=&VectorFromFloat(Value=&Lerp_Float(
      A=2
      B=30
      Alpha=&FloatFromCurve(
       FloatCurve=DataInterface() {propertyValues {Curve {keys [<at 0 0 RCIM_Cubic>, {arriveTangent=2.3837, interpMode=RCIM_Cubic, leaveTangent=2.3837, tangentMode=RCTM_User, time=1, value=1}]}}}
       CurveIndex=&RandomRangeFloat(Minimum=0, Maximum=1)
      )
     ))
    ScaleA=false
    ScaleRGB=true
   )
  ]
  rendererClasses [&NiagaraSpriteRendererProperties]
  renderers [{rendererClass=&NiagaraSpriteRendererProperties}]
 }
 Wide_Flash {
  emitterUpdateScript [
   &EmitterState()
   &SpawnRate(SpawnRate=20, Use Spawn Probability=false)
   &SpawnBurst_Instantaneous(Spawn Count=1, Use Spawn Probability=false)
  ]
  particleSpawnScript [
   &InitializeParticle(
    Color=<rgba 1 0.2 0.02 1>
    Lifetime=&UniformRangedFloat(Minimum=0.08, Maximum=0.1)
    Mesh Scale=&UniformRangedVector(Minimum=<1 1 1>, Maximum=<2 2 2>)
    Module.Write Color=true
    Module.Write Lifetime=true
    Module.Write Mass=true
    Module.Write Material Random=true
    Module.Write Position=false
    Module.Write Scale=true
    Module.Write SpriteRotation=false
    Module.Write SpriteSize=false
    Position=&SimulationPosition()
   )
   /SetVariables(
    Particles.MeshOrientation=&MakeQuatFromAxisAngle(Axis=$System.VecDir, AngleInDegrees=&UniformRangedFloat(Minimum=0, Maximum=360))
    Particles.RandomFloat=&UniformRangedFloat(Minimum=0, Maximum=1)
   ) {isSetParameters=true}
   &CameraOffset(Camera Offset Amount=50)
  ]
  particleUpdateScript [
   &ParticleState() {module=UpdateAge}
   &ScaleMeshSize(Scale Factor=&MakeVector(
     X=&Lerp_Float(
      A=0.7
      B=1
      Alpha=&FloatFromCurve(
       FloatCurve=DataInterface() {propertyValues {Curve {keys [<at 0 0 RCIM_Cubic>, {arriveTangent=3.3672, interpMode=RCIM_Cubic, leaveTangent=3.3672, tangentMode=RCTM_User, time=1, value=1}]}}}
       CurveIndex=&RandomRangeFloat(Minimum=0, Maximum=1)
      )
     )
     Y=1
     Z=1
    )) {module=ScaleMeshSize001}
   &ScaleMeshSize(Scale Factor=&VectorFromCurve(VectorCurve=DataInterface() {propertyValues {
       XCurve {keys [{arriveTangent=5.045, interpMode=RCIM_Cubic, leaveTangent=5.045, tangentMode=RCTM_User, time=0.0043, value=0.4742}, <at 0.9792 1.2888>]}
       YCurve {keys [
         <at -0.0011 1.0067 RCIM_Cubic>
         {arriveTangent=-1.6901, interpMode=RCIM_Cubic, leaveTangent=-1.6901, tangentMode=RCTM_User, time=0.9932, value=0.2813}
        ]}
       ZCurve {keys [
         <at -0.0011 1.0067 RCIM_Cubic>
         {arriveTangent=-1.6791, interpMode=RCIM_Cubic, leaveTangent=-1.6791, tangentMode=RCTM_User, time=0.9932, value=0.2813}
        ]}
      }}))
   /Game/Effects/Niagara/Modules/Materials/MaterialPanner(
    Jitter Pan by SpawnGroups=false
    Offset:NiagaraFloat=&Add_Float(A=$Emitter.Age, B=$Particles.RandomFloat)
    Speed=-5
   )
   &ScaleColor(
    Scale RGB=&VectorFromFloat(Value=&Lerp_Float(
      A=1
      B=30
      Alpha=&FloatFromCurve(
       FloatCurve=DataInterface() {propertyValues {Curve {keys [<at 0 0 RCIM_Cubic>, {arriveTangent=3.1145, interpMode=RCIM_Cubic, leaveTangent=3.1145, tangentMode=RCTM_User, time=1, value=1}]}}}
       CurveIndex=&RandomRangeFloat(Minimum=0, Maximum=1)
      )
     ))
    ScaleA=true
    ScaleRGB=true
   )
  ]
  rendererClasses [&NiagaraMeshRendererProperties]
  renderers [{rendererClass=&NiagaraMeshRendererProperties, rendererIndex=0}]
 }
}
summary {
 systemName=NS_TidalCrane_Scraps
 emitters [
  {emitterName=Sparks, rendererClasses [&NiagaraSpriteRendererProperties]}
  {emitterName=Streaks, rendererClasses [&NiagaraSpriteRendererProperties]}
  {emitterName=Wide_Flash, rendererClasses [&NiagaraMeshRendererProperties]}
  {emitterName=Air_Glow, rendererClasses [&NiagaraLightRendererProperties, &NiagaraSpriteRendererProperties]}
  {emitterName=Smoke, rendererClasses [&NiagaraSpriteRendererProperties]}
 ]
 userVariables []
}
