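# Material function that dithers a cone-shaped see-through cutout around the camera in vehicle materials.
# It transforms local position to camera space, building an elliptical, round-tipped cone mask from offset, size, aperture and contrast parameters.
# Cone length shortens with code-driven PlayerIsADS and steep camera pitch.
# Temporal-AA dither combines the mask with incoming Opacity and OpacityMask, ShadowReplace keeps shadows solid, and several earlier sphere, box and screen-space experiments remain unconnected.
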MaterialFunction /Game/Packages/Fortress_SharedMaterials/Base_Material_Functions/BaseCustomizationFunctions/MF_FN_Vehicle_CameraDither_2025
graph {nodes [
  n1 = LocalPosition()
  n2 = ComponentMask(None=@n1.XYZ) {b=true}
  n3 = Multiply(A=Subtract(A=@n2, B=Constant() {r=-50}), B=Constant() {r=1})
  n4 = AppendVector(A=Divide(A=ComponentMask(None=@n1.XYZ) {g=true, r=true}, B=Max(A=@n3)), B=@n2)
  n5 = Constant()
  n6 = Constant() {r=-15}
  n7 = Constant()
  n8 = Multiply(
   A=Subtract(
    A=TransformPosition(None=@n4) {transformType=TRANSFORMPOSSOURCE_Camera}
    B=MaterialFunctionCall(R=@n5, G=@n6, B=@n7) {materialFunction=&AppendMany}.RGB
   )
   B=MaterialFunctionCall(R=NamedRerouteUsage(), G=NamedRerouteUsage(), B=Constant() {r=1}) {materialFunction=&AppendMany}.RGB
  )
  Abs(None=Divide(A=@n8) {constB=200})
  n9 = CameraPositionWS()
  n10 = CameraVectorWS()
  n11 = Constant()
  n12 = SphereMask(
   A=Subtract(A=@n9, B=Multiply(A=@n10, B=@n11) {constB=100})
   B=WorldPosition().XYZ
   Radius=LinearInterpolate(A=Constant() {r=170}, B=Constant() {r=50}, Alpha=Constant())
   Hardness=Constant() {r=0.9}
  )
  n13 = OneMinus(None=@n12)
  n14 = NamedRerouteDeclaration(None=Constant() {r=1}) {name=Height, nodeColor=<rgba 0.2422 1 0 1>}
  n15 = MaterialFunctionCall(Float2=MaterialFunctionCall() {materialFunction=&ScreenResolution}.Visible Resolution) {materialFunction=&BreakOutFloat2Components}
  n16 = Multiply(A=Divide(B=Divide(A=@n15.G, B=@n15.G)), B=NamedRerouteDeclaration(None=Constant() {r=0.66}) {name=Width, nodeColor=<rgba 0 1 0.1562 1>})
  n17 = MaterialFunctionCall(
   UVs=ScreenPosition().ViewportUV
   Texture Scale=NamedRerouteDeclaration(None=AppendVector(A=@n14, B=@n16)) {name=AspectRatio, nodeColor=<rgba 0 0.625 1 1>}
  ) {materialFunction=&ScaleUVsByCenter}
  n18 = ComponentMask(None=Constant3Vector() {constant=<rgba 0.5 0.5 0 1>}) {g=true, r=true}
  n19 = MaterialFunctionCall(In=Subtract(A=Distance(A=@n17.UVs, B=@n18), B=Constant() {r=-0.1}), Contrast=Constant() {r=3}) {materialFunction=&CheapContrast}
  n20 = Multiply(A=@n19.Result, B=@n19.Result)
  Clamp(None=MaterialFunctionCall(Alpha Threshold=Max(A=@n13, B=@n20), Random=Constant()) {materialFunction=&DitherTemporalAA_Fortnite}.Result)
  n21 = LocalPosition()
  n22 = MaterialFunctionCall!(UV Coord=ComponentMask(None=@n21.XYZ) {b=true, g=true}, LenghtScale, RadialScale) {materialFunction=/Game/Global/MaterialFunctions/TechArt/MF_PolarCoordinates}
  ComponentMask(None=@n22.Result) {g=true}
  ComponentMask(None=@n22.Result) {r=true}
  Constant() {r=0.2}
  n23 = ComponentMask(None=OneMinus(None=@n8)) {b=true}
  n24 = Constant() {r=1}
  Distance(A=Divide(A=ComponentMask(None=@n8) {g=true, r=true}, B=Multiply(A=@n23, B=@n24) {constB=1.2}), B=Constant2Vector())
  n25 = GetMaterialAttributes(None=FunctionInput() {inputType=FunctionInput_MaterialAttributes}) {attributeGetTypes [#1, #2]}
  n26 = TransformPosition(None=LocalPosition().XYZ) {transformType=TRANSFORMPOSSOURCE_Camera}
  n27 = ScalarParameter() {group=Vehicle Camera Dithering, parameterName=Cutout_ConeOffset_Vertical, sortPriority=10}
  n28 = Abs(None=DotProduct(
    A=MaterialFunctionCall() {materialFunction=&CameraDirectionVector}.Result
    B=Constant3Vector() {constant=<rgba 0 0 1 0>}
   ))
  n29 = Add(
   A=ComponentMask(None=@n26) {desc=RB if local space, g=true, r=true}
   B=AppendVector(
    A=ScalarParameter() {group=Vehicle Camera Dithering, parameterName=Cutout_ConeOffset_Horizontal, sortPriority=10}
    B=LinearInterpolate(A=@n27, Alpha=@n28) {constB=0}
   )
  )
  n30 = AppendVector(
   A=ScalarParameter() {defaultValue=0.66, group=Vehicle Camera Dithering, parameterName=Cutout_Ratio_Width, sortPriority=5}
   B=ScalarParameter() {defaultValue=1, group=Vehicle Camera Dithering, parameterName=Cutout_Ratio_Height, sortPriority=5}
  )
  n31 = ScalarParameter() {
   defaultValue=1.5
   desc=A single parameter to tweak cutout size without messing with cone position or lenght
   group=Vehicle Camera Dithering
   parameterName=Cutout_Size
   sortPriority=0
  }
  n32 = ScalarParameter() {defaultValue=20, desc=AffectsCutoutSize, group=Vehicle Camera Dithering, parameterName=Cutout_ConeOffset_Depht, sortPriority=10}
  n33 = ComponentMask(None=Add(A=@n26, B=@n32)) {b=true, desc=G if local space}
  n34 = ScalarParameter() {
   defaultValue=50
   desc="Shorten Lenght on ADS to compensate camera movement. "
   group=Vehicle Camera Dithering
   parameterName=Cutout_ConeLenght_ADS
  }
  n35 = ScalarParameter() {
   defaultValue=0.7
   desc="Usefull to avoid holes on floor and ceiling.\r\nThis value is a lenght multiplier for when camera is pointing up/down"
   group=Vehicle Camera Dithering
   parameterName=Cutout_ConeLenghtMultFacingUP/Down
  }
  n36 = Subtract(A=@n28, B=Constant() {r=-0.15})
  n37 = Constant() {r=0.4}
  n38 = Multiply(
   A=LinearInterpolate(
    A=ScalarParameter() {defaultValue=170, group=Vehicle Camera Dithering, parameterName=Cutout_ConeLenght}
    B=@n34
    Alpha=ScalarParameter() {desc=Value overriten by Code, group=Vehicle Camera Dithering, parameterName=PlayerIsADS, sortPriority=10000}
   )
   B=LinearInterpolate(
    A=Constant() {r=1}
    B=@n35
    Alpha=MaterialFunctionCall(In=@n36, Contrast=@n37) {materialFunction=&CheapContrast}.Result
   )
  )
  n39 = Subtract(A=@n33, B=@n38)
  n40 = ScalarParameter() {
   defaultValue=0.3
   desc="Rounded Soft Tip (base of cone) in front of the player.\r\nThe lenght is a % of Cone Lenght"
   group=Vehicle Camera Dithering
   parameterName=Cutout_ConeTipFalloffLenght
  }
  n41 = Add(A=Multiply(A=@n39) {constB=-1}, B=Multiply(A=@n38, B=@n40))
  n42 = OneMinus(None=Divide(A=Divide(A=@n41, B=@n40), B=@n38))
  n43 = Multiply(A=@n42, B=@n42)
  n44 = LinearInterpolate(A=@n33, B=Multiply(A=OneMinus(None=@n43), B=@n38), Alpha=Max(A=Sign(None=@n39)) {constB=0})
  n45 = ScalarParameter() {
   defaultValue=0.3
   desc="Controls the aperture angle of the cone\r\nBigger values = bigger angles"
   group=Vehicle Camera Dithering
   parameterName=Cutout_ConeApertureAngleMult
   sortPriority=15
  }
  n46 = Divide(A=Divide(A=Multiply(A=@n29, B=@n30), B=@n31), B=Max(A=Multiply(A=@n44, B=@n45)) {constB=0})
  n47 = Constant2Vector()
  n48 = ScalarParameter() {
   defaultValue=1.5
   desc=modulatesgradient contrast ont he edges of the cutout hole
   group=Vehicle Camera Dithering
   parameterName=Cutout_Contrast
   sortPriority=5
  }
  n49 = MaterialFunctionCall(Alpha Threshold=MaterialFunctionCall(In=Distance(A=@n46, B=@n47), Contrast=@n48) {materialFunction=&CheapContrast}.Result) {materialFunction=&DitherTemporalAA_Fortnite}
  n50 = Min(A=@n25.OpacityMask, B=@n49.Result)
  n51 = Constant() {r=1}
  n52 = Min(A=@n25.Opacity, B=@n49.Result)
  FunctionOutput(None=SetMaterialAttributes!(
    MaterialAttributes=@n25.MaterialAttributes
    OpacityMask=ShadowReplace(Default=@n50, Shadow=@n51)
    Opacity=ShadowReplace(Default=@n52, Shadow=@n51)
   ) {attributeSetTypes [#1, #2]})
  LinearInterpolate(A=@n25.OpacityMask, B=@n20, Alpha=@n12) {constB=0}
  MaterialFunctionCall(Number=@n3) {materialFunction=&DebugScalarValues}
  n53 = Constant()
  n54 = Constant()
  n55 = Multiply(A=LinearInterpolate(A=Constant() {r=120}, B=Constant() {r=50}, Alpha=Constant())) {constB=2}
  n56 = OneMinus(None=MaterialFunctionCall!(
    A=@n8
    B=Constant3Vector()
    Bounds=MaterialFunctionCall(R=@n53, G=@n54, B=@n55) {materialFunction=&AppendMany}.RGB
    Edge Falloff=Constant() {r=100}
   ) {materialFunction=/Engine/Functions/Engine_MaterialFunctions01/Gradient/BoxMask-3D}.Result)
  n57 = MaterialFunctionCall(In=Subtract(A=@n56, B=Constant() {r=-0.1}), Contrast=Constant() {r=3}) {materialFunction=&CheapContrast}
  Max(A=Multiply(A=@n57.Result, B=@n57.Result), B=Constant())
  Step(Y=@n9) {constX=0}
  Step(X=@n56) {constX=0.001, constY=1}
  TransformPosition(None=AppendVector(A=@n22.Result, B=@n21.Z)) {transformType=TRANSFORMPOSSOURCE_Camera}
 ]}
