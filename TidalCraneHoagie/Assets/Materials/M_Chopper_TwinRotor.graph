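# Twin-rotor chopper vehicle material that adds an animated jet-heat glow to shared vehicle master, emissive, damage and camera-dither functions.
# A Stage parameter and a Blueprint-supplied start time drive timed stage lerps for heating and cooling.
# Vertex color, bounding-box gradient and pre-skinning sphere masks at both jet pivots shape the heat mask.
# The mask samples a gradient texture for emissive, while masked lerps adjust base color, specular and roughness.
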Material /TidalCraneHoagie/Assets/Materials/M_Chopper_TwinRotor
graph {
 properties {
  bAutomaticallySetUsageInEditor=false
  bUseMaterialAttributes=true
  bUsedWithInstancedStaticMeshes=true
  bUsedWithMeshDeformer=true
  bUsedWithSkeletalMesh=true
  pixelDepthOffsetMode=PDOM_Legacy
  refractionMethod=RM_IndexOfRefraction
  usedShadingModels=Default Lit
 }
 nodes [
  n1 = ScalarParameter() {group=TwinRotor, parameterName=Stage, sortPriority=3}
  n2 = LinearInterpolate(A=Subtract(A=@n1), B=@n1, Alpha=NamedRerouteUsage()) {desc=Native lerp toward}
  n3 = Constant4Vector() {desc=Stage 0 to 1, constant=<rgba 1 2 0 1>}
  NamedRerouteDeclaration(None=Saturate(None=MaterialFunctionCall(Input=@n2, Input Low=@n3.R, Input High=@n3.R, Target Low=@n3.R, Target High=@n3.R) {materialFunction=&RemapValueRange}.Result)) {name=Stage2_Lerp, nodeColor=<rgba 0.1893 1 0 1>}
  n4 = Subtract(A=Time(), B=ScalarParameter() {defaultValue=-99999, group=TwinRotor, parameterName=GameTimeFromBP, sortPriority=100})
  n5 = Add(A=ScalarParameter() {defaultValue=1, group=TwinRotor, parameterName=Animation_Duration, sortPriority=101}) {constB=0}
  NamedRerouteDeclaration(None=Saturate(None=Divide(A=@n4, B=@n5))) {name=TimeNative, nodeColor=<rgba 0.2891 1 0 1>}
  n6 = Constant4Vector() {desc=Stage 0 to 1, constant=<rgba 0 1 0 1>}
  NamedRerouteDeclaration(None=Saturate(None=MaterialFunctionCall(Input=@n2, Input Low=@n6.R, Input High=@n6.R, Target Low=@n6.R, Target High=@n6.R) {materialFunction=&RemapValueRange}.Result)) {name=Stage1_Lerp, nodeColor=<rgba 0.1893 1 0 1>}
  n7 = VertexColor()
  NamedRerouteDeclaration(None=Add(A=@n7.G, B=@n7.G)) {name=Jet_Msk, nodeColor=<rgba 0 1 1 1>}
  n8 = MaterialFunctionCall() {materialFunction=&BoundingBoxBased_0-1_UVW}
  n9 = VertexInterpolator(VS=LocalPosition() {localOrigin=InstancePreSkinning}.XYZ)
  n10 = Add(
   A=SphereMask(A=Constant3Vector() {desc="R Jet  Pviot", constant=<rgba -330 90 255 0>}, B=@n9.PS) {attenuationRadius=160, hardnessPercent=0.01}
   B=SphereMask(A=Constant3Vector() {desc="L Jet  Pviot", constant=<rgba -330 -90 255 0>}, B=@n9.PS) {attenuationRadius=160, hardnessPercent=0.01}
  )
  n11 = Multiply(A=Power(Base=Add(A=@n8.R, B=@n10)) {constExponent=8}) {constB=30}
  n12 = ScalarParameter() {group=TwinRotor, parameterName=Jet_HeatStartAmount}
  n13 = LinearInterpolate(
   A=LinearInterpolate(A=@n12, B=ScalarParameter() {group=TwinRotor, parameterName=Jet_HeatTargetAmount}, Alpha=NamedRerouteUsage())
   B=@n12
   Alpha=NamedRerouteUsage()
  ) {constB=0}
  n14 = OneMinus(None=@n13)
  n15 = Constant4Vector() {constant=<rgba 0 0.4 0 0.001>}
  n16 = Multiply(
   A=Multiply(
    A=MaterialFunctionCall!(Mask In=@n11, Amount=@n14, Range=@n15.G, Start Offset=@n15.G, Sharpness=@n15.G) {materialFunction=/Game/Global/MaterialFunctions/MF_MaskShifting}.Result
    B=@n13
   ) {constB=0.5}
   B=NamedRerouteUsage()
  )
  n17 = TextureSample(UVs=@n16) {samplerSource=SSM_Clamp_WorldGroupSettings, texture=/TidalCraneHoagie/Assets/Textures/T_Chopper_TwinRotor_JetHeat_Gradient}
  n18 = ScalarParameter() {defaultValue=8, group=TwinRotor, parameterName=Jet_HeatIntensity}
  NamedRerouteDeclaration(None=Multiply(A=Power(Base=@n17.RGB) {constExponent=1}, B=LinearInterpolate(B=@n18, Alpha=@n13) {constB=10}) {constB=10}) {name=JetHeat_Emissive, nodeColor=<rgba 1 0.3984 0 1>}
  NamedRerouteDeclaration(None=@n16) {name=Jet_Heat_Msk, nodeColor=<rgba 0.3594 0 1 1>}
  n19 = MaterialFunctionCall() {materialFunction=/Game/Packages/Fortress_SharedMaterials/Base_Material_Functions/BaseCustomizationFunctions/MF_FN_Vehicle_Master.MF_FN_Vehicle_MASTER}
  n20 = MaterialFunctionCall!(
   Material=@n19.Material Result
   "Tail Pipe Mask (_VL R)"=@n19.Material Result
   "Head Light Mask (_VL G)"=@n19.Material Result
   "Tail Light Mask (_VL B)"=@n19.Material Result
  ) {materialFunction=/Game/Packages/Fortress_SharedMaterials/Base_Material_Functions/BaseCustomizationFunctions/MF_BaseVehicle_Emissive}
  n21 = MaterialFunctionCall!(Material=@n20.Result, Cropped Emissive Mask=@n20.Result) {materialFunction=/Game/Packages/Fortress_SharedMaterials/Base_Material_Functions/BaseCustomizationFunctions/MF_Base_Customization_CroppedEmissive}
  n22 = MaterialFunctionCall(In=MaterialFunctionCall(Material=@n21.Result, "Clear Coat/Damage Mask (_M B)"=@n19."CC/Damage Mask (_M B)") {materialFunction=&MF_BaseVehicle_Damage}.Result) {materialFunction=&MF_TechArtOverrides_accessories}
  n23 = MaterialFunctionCall!(In=@n22.Result) {materialFunction=/Game/Packages/Fortress_SharedMaterials/Base_Material_Functions/BaseCustomizationFunctions/MF_FN_Vehicle_CameraDither_2025}
  n24 = GetMaterialAttributes(None=MaterialFunctionCall(MA In=@n23.Result) {materialFunction=&MF_CreativeMoveTool}.MA Out) {attributeGetTypes [
    #1
    #2
    #3
    #4
   ]}
  n25 = NamedRerouteUsage()
  SetMaterialAttributes!(
   MaterialAttributes=@n24.MaterialAttributes
   Base Color=LinearInterpolate(A=@n24.Base Color, Alpha=@n25) {constB=0}
   Specular=LinearInterpolate(A=@n24.Specular, Alpha=@n25) {constB=0}
   Roughness=LinearInterpolate(A=@n24.Roughness, Alpha=@n25) {constA=1}
   Emissive Color=Add(A=@n24.Emissive Color, B=NamedRerouteUsage())
  ) {attributeSetTypes [
    #1
    #2
    #3
    #4
   ]}
 ]
}
