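# Looping stereo music source for the TidalCrane Hoagie horn, using distant horn attenuation and a horn music concurrency set.
# On play, an ADSR envelope fades in over a short attack to full sustain, and its attack trigger starts the Wave Player on the music loop wave.
# The envelope is multiplied by a gain converted from a decibel variable, then scales the left and right outputs.
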MetaSoundSource /TidalCraneHoagie/Assets/Sounds/MSS_TidalCrane_Hoagie_Music
inputs [UE.Source.OnPlay {type=Trigger}]
interfaces [UE.OutputFormat.Stereo:1.0, UE.Source:1.1]
outputs [
 "UE.OutputFormat.Stereo.Audio:0" {type=Audio}
 "UE.OutputFormat.Stereo.Audio:1" {type=Audio}
]
pages [Default {
  input_defaults {UE.Source.OnPlay=null}
  nodes [
   ADSREnvelopeADSREnvelope_0 = ADSR Envelope.ADSR Envelope:Float(Trigger Attack=^UE.Source.OnPlay) {also {defaults {Attack Time=0.3, Sustain Level=1}}}
   DecibelstoLinearGain_0 = UE.Decibels to Linear Gain:Float(Decibels=VariableAccessor.Float(Variable=InitVariable.Float().Variable).Value)
   n1 = UE.Multiply:Float(PrimaryOperand=@ADSREnvelopeADSREnvelope_0.Out Envelope, AdditionalOperands=@DecibelstoLinearGain_0.Linear Gain)
   WavePlayer_0 = UE.Wave Player:Stereo(Play=@ADSREnvelopeADSREnvelope_0.On Attack Triggered) {also {
     defaults {Loop="<MetaSoundBuilderResult.SUCCEEDED: 0>", Wave Asset=/TidalCraneHoagie/Assets/Sounds/Wavs/MX_S37_TidalCraneHoagie_LP}
     version [1,1]
    }}
   Multiply_0 = UE.Multiply:Audio by Float(AdditionalOperands=@n1.Out, PrimaryOperand=@WavePlayer_0.Out Left)
   Multiply_1 = UE.Multiply:Audio by Float(AdditionalOperands=@n1.Out, PrimaryOperand=@WavePlayer_0.Out Right)
  ]
  links [{dst="@UE.OutputFormat.Stereo.Audio:0", src=Multiply_0.Out}, {dst="@UE.OutputFormat.Stereo.Audio:1", src=Multiply_1.Out}]
 }]
properties {
 assetImportData=&self:AssetImportData
 attenuationSettings=/TidalCraneHoagie/Assets/Sounds/Settings/ATT_TidalCrane_Hoagie_Horn_Distant
 bLooping=true
 duration=10000
 maxDistance=23500
 numChannels=2
 outputFormat=Stereo
 sampleRate=48000
 volume=1.25
 concurrencySet [/Hoagie/Audio/Assets/Horn/Hoagie_Horn_Music_Conc]
}
