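# Sound cue that plays the Tidal Crane Hoagie vehicle's boost-ready sound, reusing the base Hoagie's boost-ready wave.
# A fixed short delay feeds the single wave player, played at reduced volume in the vehicle exterior sound class with the Hoagie engine boost attenuation.
# Overridden concurrency allows one instance per owner, stopping the oldest when a new one starts.
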SoundCue /TidalCraneHoagie/Assets/Sounds/SC_TidalCraneHoagie_Boost_Ready
root=Delay_0
nodes [
 {id=Delay_0, children [WavePlayer_0], properties {delayMax=0.2, delayMin=0.2}}
 {id=WavePlayer_0, properties {soundWaveAssetPtr=/Hoagie/Audio/Assets/Waves/Veh_Hoagie_Boost_Ready_01}}
]
properties {
 attenuationSettings=/Hoagie/Audio/Hoagie_Engine_Boost_Att
 bOverrideConcurrency=true
 maxDistance=30000
 soundClassObject=&Fort_Vehicles_Exterior
 volumeMultiplier=0.5
 concurrencyOverrides {bLimitToOwner=true, maxCount=1, resolutionRule=StopOldest, voiceStealReleaseTime=1}
}
