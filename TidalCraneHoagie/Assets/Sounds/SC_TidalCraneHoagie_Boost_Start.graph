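# Sound cue for the TidalCrane Hoagie vehicle's boost-start sound, in the exterior vehicles sound class.
# A Random node picks one of three equally weighted wave variations, played through the shared Hoagie engine boost attenuation.
# A concurrency override allows one instance per owner, stopping the oldest with a release fade.
# It also sends to an engine sidechain volume audio bus, likely for ducking.
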SoundCue /TidalCraneHoagie/Assets/Sounds/SC_TidalCraneHoagie_Boost_Start
root=Random_1
nodes [
 {id=Random_1, children [WavePlayer_0,WavePlayer_4,WavePlayer_5], properties {weights [1,1,1]}}
 {id=WavePlayer_0, properties {soundWaveAssetPtr=/TidalCraneHoagie/Assets/Sounds/Wavs/TidalCrane_Hoagie_Boost_Start_01}}
 {id=WavePlayer_4, properties {soundWaveAssetPtr=/TidalCraneHoagie/Assets/Sounds/Wavs/TidalCrane_Hoagie_Boost_Start_02}}
 {id=WavePlayer_5, properties {soundWaveAssetPtr=/TidalCraneHoagie/Assets/Sounds/Wavs/TidalCrane_Hoagie_Boost_Start_03}}
]
properties {
 attenuationSettings=/Hoagie/Audio/Hoagie_Engine_Boost_Att
 bOverrideConcurrency=true
 maxDistance=30000
 soundClassObject=&Fort_Vehicles_Exterior
 volumeMultiplier=2
 busSends [{
   audioBus=/TidalCraneHoagie/Assets/Sounds/Modulation/AB_TidalCraneHoagie_Engine_SideChain_Vol
   minSendLevel=0
   sendLevel=1
   soundSourceBus=null
  }]
 concurrencyOverrides {bLimitToOwner=true, maxCount=1, resolutionRule=StopOldest, voiceStealReleaseTime=1}
}
