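# Sound cue for light collision impacts on the TidalCraneHoagie vehicle.
# A Random node picks among six equally weighted Hoagie light-impact waves and feeds a Modulator at the root.
# It plays in the exterior vehicles sound class with a borrowed close-range golf cart attenuation.
# A full-level bus send feeds the vehicle's engine side-chain volume audio bus, so impacts can drive engine volume modulation.
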SoundCue /TidalCraneHoagie/Assets/Sounds/SC_TidalCraneHoagie_Collision_01_Light
root=Modulator_0
nodes [
 {id=Modulator_0, children [Random_1]}
 {id=Random_1, children [WavePlayer_0,WavePlayer_5,WavePlayer_6,WavePlayer_7,WavePlayer_8,WavePlayer_9], properties {weights [1,1,1,1,1,1]}}
 {id=WavePlayer_0, properties {soundWaveAssetPtr=/Hoagie/Audio/Assets/Waves/Veh_Hoagie_Impact_Lt_02}}
 {id=WavePlayer_5, properties {soundWaveAssetPtr=/Hoagie/Audio/Assets/Waves/Veh_Hoagie_Impact_Lt_04}}
 {id=WavePlayer_6, properties {soundWaveAssetPtr=/Hoagie/Audio/Assets/Waves/Veh_Hoagie_Impact_Lt_06}}
 {id=WavePlayer_7, properties {soundWaveAssetPtr=/Hoagie/Audio/Assets/Waves/Veh_Hoagie_Impact_Lt_01}}
 {id=WavePlayer_8, properties {soundWaveAssetPtr=/Hoagie/Audio/Assets/Waves/Veh_Hoagie_Impact_Lt_03}}
 {id=WavePlayer_9, properties {soundWaveAssetPtr=/Hoagie/Audio/Assets/Waves/Veh_Hoagie_Impact_Lt_05}}
]
properties {
 attenuationSettings=/Game/Sounds/Attenuation/Fort_GolfCart_Movement_Close_Att
 maxDistance=5300
 soundClassObject=&Fort_Vehicles_Exterior
 volumeMultiplier=1
 busSends [{
   audioBus=/TidalCraneHoagie/Assets/Sounds/Modulation/AB_TidalCraneHoagie_Engine_SideChain_Vol
   minSendLevel=0
   sendLevel=1
   soundSourceBus=null
  }]
}
