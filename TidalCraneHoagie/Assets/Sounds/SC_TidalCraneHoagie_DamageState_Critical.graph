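# Sound cue for the TidalCrane Hoagie vehicle's critical damage state, layering a death-dive sound with a damage alarm.
# The dive one-shot uses distant critical attenuation and an envelope that dips after a second.
# The looping alarm uses close engine attenuation, fades in then slowly out over fifteen seconds, and is mixed at double the dive's input volume.
# Concurrency allows one instance per owner, stopping the oldest.
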SoundCue /TidalCraneHoagie/Assets/Sounds/SC_TidalCraneHoagie_DamageState_Critical
root=Mixer_0
nodes [
 {id=Attenuation_0, children [Enveloper_0], properties {attenuationSettings=/Hoagie/Audio/Hoagie_Engine_Close_Att}}
 {id=Attenuation_1, children [WavePlayer_1], properties {attenuationSettings=/Hoagie/Audio/Hoagie_Engine_Distant_Critical_Att}}
 {
  id=Enveloper_0
  children [WavePlayer_0]
  properties {pitchCurve {editorCurveData {keys [{time=0}]}}, volumeCurve {editorCurveData {keys [<at 0 0>,<at 1.5 1>,<at 15 0>]}}}
 }
 {
  id=Enveloper_1
  children [Attenuation_1]
  properties {pitchCurve {editorCurveData {keys [{time=0}]}}, volumeCurve {editorCurveData {keys [<at 0 1>,<at 1 1>,<at 1.5 0.6>]}}}
 }
 {id=Mixer_0, children [Enveloper_1, Attenuation_0], properties {inputVolume [1,2]}}
 {id=WavePlayer_0, properties {bLooping=true, soundWaveAssetPtr=/TidalCraneHoagie/Assets/Sounds/Wavs/TidalCrane_Hoagie_DamageAlarm_Lp_01}}
 {id=WavePlayer_1, properties {soundWaveAssetPtr=/Hoagie/Audio/Assets/Waves/Veh_Hoagie_Dmg_Death_Dive_Start_01}}
]
properties {
 bOverrideConcurrency=true
 maxDistance=20000
 soundClassObject=&Fort_Vehicles_Exterior
 volumeMultiplier=1
 concurrencyOverrides {bLimitToOwner=true, maxCount=1, resolutionRule=StopOldest, voiceStealReleaseTime=1}
}
