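# Looping stereo MetaSound source for the TidalCraneHoagie engine (1p variant), driven by forward speed, backward movement and turn input.
# Flight loop patches layer front and rear engine loops under an on-play ADSR, while a bank patch adds turn loops and one-shots.
# The bank signal also feeds a sidechain audio bus, and a sidechain control bus modulates engine gain.
# The Mobile page drops banking and outputs a mono engine mix to both channels.
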MetaSoundSource /TidalCraneHoagie/Assets/Sounds/MSS_TidalCraneHoagie_Engine_1p
inputs [
 BackwardDot {type=Float}
 ForwardSpeed {type=Float}
 OutputVolume {type=Float}
 Sounds.BackwardsLoop {type=WaveAsset:Array}
 Sounds.BankLoop {type=WaveAsset:Array}
 Sounds.BankOneShots {type=WaveAsset:Array}
 Sounds.ForwardsLoop {type=WaveAsset:Array}
 TurnDelta {type=Float}
 UE.Source.OnPlay {type=Trigger}
]
interfaces [UE.OutputFormat.Stereo:1.0, UE.Source:1.1]
outputs [
 "UE.OutputFormat.Stereo.Audio:0" {type=Audio}
 "UE.OutputFormat.Stereo.Audio:1" {type=Audio}
]
pages [
 Default {
  is_build_graph=true
  input_defaults {
   BackwardDot=0
   ForwardSpeed=0
   OutputVolume=1
   TurnDelta=0
   UE.Source.OnPlay=null
   Sounds.BackwardsLoop [/TidalCraneHoagie/Assets/Sounds/Wavs/TidalCrane_Hoagie_Engine_Close_Rear_Loop_01]
   Sounds.BankLoop [/TidalCraneHoagie/Assets/Sounds/Wavs/TidalCrane_Hoagie_Bank_Loop_01, /TidalCraneHoagie/Assets/Sounds/Wavs/TidalCrane_Hoagie_Bank_Loop_02]
   Sounds.BankOneShots [
    /TidalCraneHoagie/Assets/Sounds/Wavs/TidalCrane_Hoagie_Turn_Close_01
    /TidalCraneHoagie/Assets/Sounds/Wavs/TidalCrane_Hoagie_Turn_Close_02
    /TidalCraneHoagie/Assets/Sounds/Wavs/TidalCrane_Hoagie_Turn_Close_03
   ]
   Sounds.ForwardsLoop [/TidalCraneHoagie/Assets/Sounds/Wavs/TidalCrane_Hoagie_Engine_Close_Front_Loop_01]
  }
  nodes [
   n1 = InitVariable.Float()
   n2 = MapRange.MapRange:Float(In=^ForwardSpeed) {also {defaults {Out Range A=0.7}}}
   n3 = VariableAccessor.Float(Variable=InitVariable.Float().Variable)
   n4 = /FortMetasound/Abilities/Flight/MSP_Ability_Flight_Bank(
    MovementSounds=^Sounds.BankLoop
    OneShotSounds=^Sounds.BankOneShots
    DirectionValue=UE.Subtract:Float(AdditionalOperands=^TurnDelta) {also {defaults {PrimaryOperand=1}}}.Out
    NormSpeed=VariableAccessor.Float(Variable=VariableMutator.Float(Variable=@n1.Variable, Value=@n2.Out Value).Variable).Value
    LoopThreshold=@n3.Value
    OneshotThreshold=VariableAccessor.Float(Variable=@n3.Variable).Value
   ) {also {defaults {EnterCondition=3, ExitCondition=4, Fade=1.5, LoopVolume=0.3, MinPitchMod=0.8, OneShotRetriggerDelay=4, OneShotVolume=0.7, SpeedVol=0.8}}}
   AudioBusWriter1_0 = "UE.Audio Bus Writer (1)"(In 0=UE.Multiply:Audio by Float(PrimaryOperand=UE.Add:Audio(PrimaryOperand=@n4.Out L, AdditionalOperands=@n4.Out R).Out) {also {defaults {AdditionalOperands=0.65}}}.Out) {also {defaults {Audio Bus=/TidalCraneHoagie/Assets/Sounds/Modulation/AB_TidalCraneHoagie_Engine_SideChain_Vol}}}
   ADSREnvelopeADSREnvelope_0 = ADSR Envelope.ADSR Envelope:Float(Trigger Attack=^UE.Source.OnPlay) {also {defaults {Attack Time=1, Decay Time=10, Sustain Level=0.65}}}
   n5 = InitVariable.Float()
   n6 = MapRange.MapRange:Float(In=^BackwardDot) {also {defaults {Out Range A=0.6}}}
   n7 = /FortMetasound/Abilities/Flight/MSP_Ability_Flight_Loop(
    NormSpeed=^BackwardDot
    MovementSounds=^Sounds.BackwardsLoop
    OnPlay=^UE.Source.OnPlay
    SpeedVol=VariableAccessor.Float(Variable=VariableMutator.Float(Variable=@n5.Variable, Value=@n6.Out Value).Variable).Value
   ) {also {defaults {FadeIn=1}}}
   n8 = VariableAccessor.Float(Variable=VariableMutator.Float(
     Variable=InitVariable.Float().Variable
     Value=MapRange.MapRange:Float(In=^ForwardSpeed) {also {defaults {Out Range A=1, Out Range B=1.25}}}.Out Value
    ).Variable)
   n9 = &MSP_Ability_Flight_Loop(NormSpeed=^ForwardSpeed, MovementSounds=^Sounds.ForwardsLoop, OnPlay=^UE.Source.OnPlay, SpeedVol=@n8.Value) {also {defaults {FadeIn=1}}}
   AudioMixerAudioMixerStereo2_1 = "AudioMixer.Audio Mixer (Stereo, 2)"(In 1 L=@n7.Out L, In 1 R=@n7.Out R, In 0 L=@n9.Out L, In 0 R=@n9.Out R)
   Multiply_0 = UE.Multiply:Audio by Float(AdditionalOperands=@ADSREnvelopeADSREnvelope_0.Out Envelope, PrimaryOperand=@AudioMixerAudioMixerStereo2_1.Out R)
   AudioMixerAudioMixerStereo2_0 = "AudioMixer.Audio Mixer (Stereo, 2)"(
    In 1 L=@n4.Out L
    In 1 R=@n4.Out R
    In 0 R=@Multiply_0.Out
    In 0 L=UE.Multiply:Audio by Float(AdditionalOperands=@ADSREnvelopeADSREnvelope_0.Out Envelope, PrimaryOperand=@AudioMixerAudioMixerStereo2_1.Out L).Out
    Gain 1=UE.Multiply:Float(PrimaryOperand=^OutputVolume, AdditionalOperands=VariableAccessor.Float(Variable=@n8.Variable).Value).Out
    Gain 0=UE.Multiply:Float(
     AdditionalOperands=^OutputVolume
     PrimaryOperand=Modulation.GetModulatorValue() {also {defaults {Modulator=/TidalCraneHoagie/Assets/Sounds/Modulation/CB_TidalCraneHoagie_Engine_SideChain_Vol}}}.Out
    ).Out
   )
  ]
  links [
   {dst="@UE.OutputFormat.Stereo.Audio:0", src=AudioMixerAudioMixerStereo2_0.Out L}
   {dst="@UE.OutputFormat.Stereo.Audio:1", src=AudioMixerAudioMixerStereo2_0.Out R}
  ]
 }
 Mobile {
  nodes [
   InitVariable.Float()
   ADSREnvelopeADSREnvelope_0 = ADSR Envelope.ADSR Envelope:Float(Trigger Attack=^UE.Source.OnPlay) {also {defaults {Attack Time=1, Decay Time=10, Sustain Level=0.65}}}
   n1 = VariableMutator.Float(Variable=InitVariable.Float().Variable, Value=MapRange.MapRange:Float(In=^BackwardDot) {also {defaults {Out Range A=0.6}}}.Out Value)
   n2 = VariableMutator.Float(
    Variable=InitVariable.Float().Variable
    Value=MapRange.MapRange:Float(In=^ForwardSpeed) {also {defaults {Out Range A=1, Out Range B=1.25}}}.Out Value
   )
   AudioMixerAudioMixerMono2_0 = "AudioMixer.Audio Mixer (Mono, 2)"(
    In 1=&MSP_Ability_Flight_Loop(
     NormSpeed=^BackwardDot
     MovementSounds=^Sounds.BackwardsLoop
     OnPlay=^UE.Source.OnPlay
     SpeedVol=VariableAccessor.Float(Variable=@n1.Variable).Value
    ) {also {defaults {FadeIn=1}}}.Out L
    In 0=&MSP_Ability_Flight_Loop(
     NormSpeed=^ForwardSpeed
     MovementSounds=^Sounds.ForwardsLoop
     OnPlay=^UE.Source.OnPlay
     SpeedVol=VariableAccessor.Float(Variable=@n2.Variable).Value
    ) {also {defaults {FadeIn=1}}}.Out L
   )
   n3 = Modulation.GetModulatorValue() {also {defaults {Modulator=&CB_TidalCraneHoagie_Engine_SideChain_Vol}}}
   Multiply_0 = UE.Multiply:Audio by Float(PrimaryOperand=UE.Multiply:Audio by Float(
     PrimaryOperand=UE.Multiply:Audio by Float(AdditionalOperands=@ADSREnvelopeADSREnvelope_0.Out Envelope, PrimaryOperand=@AudioMixerAudioMixerMono2_0.Out).Out
     AdditionalOperands=UE.Multiply:Float(AdditionalOperands=^OutputVolume, PrimaryOperand=@n3.Out).Out
    ).Out) {also {defaults {AdditionalOperands=0.7}}}
  ]
  links [{dst="@UE.OutputFormat.Stereo.Audio:0", src=Multiply_0.Out}, {dst="@UE.OutputFormat.Stereo.Audio:1", src=Multiply_0.Out}]
 }
]
properties {
 assetImportData=&self:AssetImportData
 bLooping=true
 duration=10000
 numChannels=2
 outputFormat=Stereo
 sampleRate=48000
 concurrencyOverrides {maxCount=2}
}
