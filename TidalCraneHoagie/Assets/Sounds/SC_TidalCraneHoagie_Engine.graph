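# Engine sound cue for the TidalCrane variant of the Hoagie rotor vehicle, driven by gameplay parameters such as rotor speed, lift, turning, forward speed, driver presence and critical damage.
# IsLocalPlayerInVehicle crossfades exterior close loops with an in-vehicle layer, alongside a distant layer.
# Facing dot parameters blend front and rear loops; quality switches substitute simpler mobile loops or drop the skid scrape and damage whine layers.
# EngineDuck lowers the whole mix.
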SoundCue /TidalCraneHoagie/Assets/Sounds/SC_TidalCraneHoagie_Engine
root=ModulatorContinuous_14
nodes [
 {
  id=Attenuation_1
  children [ModulatorContinuous_19]
  properties {attenuationSettings=&/TidalCrane_Hoagie_Engine_Close_Att}
 }
 {id=Attenuation_2, children [ModulatorContinuous_3], properties {attenuationSettings=/Hoagie/Audio/Hoagie_Engine_Close_Detail_Att}}
 {id=Attenuation_3, children [ModulatorContinuous_12], properties {attenuationSettings=/Hoagie/Audio/Hoagie_RotorWash_Att}}
 {
  id=Attenuation_4
  children [QualityLevel_3]
  properties {attenuationSettings=&/TidalCrane_Hoagie_Engine_Distant_Att}
 }
 {
  id=Attenuation_7
  children [ModulatorContinuous_18]
  properties {attenuationSettings=&TidalCrane_Hoagie_Engine_Close_Att}
 }
 {
  id=Enveloper_0
  children [Mixer_7]
  properties {loopEnd=30, pitchCurve {editorCurveData {keys [{time=0}]}}, volumeCurve {editorCurveData {keys [<at 0 1>,<at 30 0.65>]}}}
 }
 {id=Mixer_1, children [ModulatorContinuous_2, QualityLevel_2], properties {inputVolume [1,1.25]}}
 {id=Mixer_2, children [ParamCrossFade_1, ParamCrossFade_2], properties {inputVolume [1,2]}}
 {id=Mixer_4, children [ModulatorContinuous_4, ModulatorContinuous_5], properties {inputVolume [1.2,0.8]}}
 {id=Mixer_5, children [Attenuation_2, Attenuation_3], properties {inputVolume [1,0.5]}}
 {id=Mixer_6, children [ModulatorContinuous_13, ModulatorContinuous_15], properties {inputVolume [2,2]}}
 {id=Mixer_7, children [ModulatorContinuous_17, WavePlayer_13], properties {inputVolume [1,1]}}
 {
  id=ModulatorContinuous_0
  children [ModulatorContinuous_8]
  properties {pitchModulationParams {default=0, minOutput=0.6, parameterName=RotorSpeed}, volumeModulationParams {default=0, parameterName=RotorSpeed}}
 }
 {
  id=ModulatorContinuous_1
  children [ModulatorContinuous_0]
  properties {
   pitchModulationParams {default=0, maxOutput=1.05, minInput=-1, minOutput=0.95, parameterName=LiftDelta}
   volumeModulationParams {default=0, maxOutput=1.2, minInput=-1, minOutput=1, paramMode=MPM_Abs, parameterName=LiftDelta}
  }
 }
 {
  id=ModulatorContinuous_12
  children [WavePlayer_3]
  properties {pitchModulationParams {minOutput=1, parameterName=Damage}, volumeModulationParams {default=0, parameterName=Damage}}
 }
 {
  id=ModulatorContinuous_13
  children [WavePlayer_16]
  properties {
   pitchModulationParams {default=0, maxOutput=1.1, minInput=-0.3, minOutput=1, parameterName=ForwardDot}
   volumeModulationParams {default=0, minInput=-0.3, minOutput=0.2, parameterName=ForwardDot}
  }
 }
 {
  id=ModulatorContinuous_14
  children [Mixer_1]
  properties {volumeModulationParams {default=0, maxOutput=0.6, minOutput=1, parameterName=EngineDuck}}
 }
 {
  id=ModulatorContinuous_15
  children [WavePlayer_17]
  properties {
   pitchModulationParams {default=0, maxOutput=1.1, minInput=-0.3, minOutput=1, parameterName=BackwardDot}
   volumeModulationParams {default=0, minInput=-0.3, minOutput=0.2, parameterName=BackwardDot}
  }
 }
 {
  id=ModulatorContinuous_16
  children [Attenuation_4]
  properties {
   pitchModulationParams {default=0, maxOutput=1.1, minOutput=0.9, parameterName=ForwardSpeed}
   volumeModulationParams {default=0, maxOutput=1.25, minOutput=1, parameterName=ForwardSpeed}
  }
 }
 {
  id=ModulatorContinuous_17
  children [WavePlayer_4]
  properties {
   pitchModulationParams {default=0, maxOutput=1.1, minOutput=0.9, parameterName=ForwardSpeed}
   volumeModulationParams {default=0, maxOutput=1.25, minOutput=1, parameterName=ForwardSpeed}
  }
 }
 {
  id=ModulatorContinuous_18
  children [QualityLevel_0]
  properties {
   pitchModulationParams {default=0, minOutput=0.9, parameterName=TurnDelta}
   volumeModulationParams {default=0, minOutput=0.8, parameterName=TurnDelta}
  }
 }
 {
  id=ModulatorContinuous_19
  children [QualityLevel_1]
  properties {
   pitchModulationParams {default=0, minOutput=0.9, parameterName=TurnDelta}
   volumeModulationParams {default=0, maxOutput=1.25, minOutput=1, parameterName=TurnDelta}
  }
 }
 {
  id=ModulatorContinuous_2
  children [ModulatorContinuous_1]
  properties {
   pitchModulationParams {default=0, maxOutput=0.75, minOutput=1, parameterName=CriticalDamage}
   volumeModulationParams {default=0, maxOutput=0, minOutput=1, parameterName=CriticalDamage}
  }
 }
 {
  id=ModulatorContinuous_3
  children [ModulatorContinuous_9]
  properties {
   pitchModulationParams {default=0, minOutput=0.9, parameterName=ForwardSpeed}
   volumeModulationParams {default=0, maxOutput=1.25, minOutput=1, parameterName=ForwardSpeed}
  }
 }
 {
  id=ModulatorContinuous_4
  children [WavePlayer_0]
  properties {
   pitchModulationParams {default=0, maxOutput=1.1, minInput=-0.4, minOutput=1, parameterName=ForwardDot}
   volumeModulationParams {default=0, minInput=-0.4, minOutput=0.1, parameterName=ForwardDot}
  }
 }
 {
  id=ModulatorContinuous_5
  children [WavePlayer_1]
  properties {
   pitchModulationParams {default=0, maxOutput=1.1, minInput=-0.3, minOutput=1, parameterName=BackwardDot}
   volumeModulationParams {default=0, minInput=-0.3, minOutput=0.1, parameterName=BackwardDot}
  }
 }
 {id=ModulatorContinuous_8, children [Mixer_2], properties {pitchModulationParams {default=0, minOutput=0.75, parameterName=HasDriver}}}
 {
  id=ModulatorContinuous_9
  children [WavePlayer_2]
  properties {pitchModulationParams {default=0, minOutput=1}, volumeModulationParams {default=0, parameterName=IsSkidding}}
 }
 {
  id=ParamCrossFade_1
  children [Attenuation_7, Attenuation_1]
  properties {
   paramName=IsLocalPlayerInVehicle
   crossFadeInput [
    {fadeInDistanceEnd=0, fadeInDistanceStart=0, fadeOutDistanceEnd=1, fadeOutDistanceStart=0, volume=1}
    {fadeInDistanceEnd=1, fadeInDistanceStart=0, fadeOutDistanceEnd=1, fadeOutDistanceStart=1, volume=0.75}
   ]
  }
 }
 {
  id=ParamCrossFade_2
  children [null, ModulatorContinuous_16]
  properties {
   paramName=IsLocalPlayerInVehicle
   crossFadeInput [
    {fadeInDistanceEnd=1, fadeInDistanceStart=0, fadeOutDistanceEnd=1, fadeOutDistanceStart=0, volume=0}
    {fadeInDistanceEnd=0, fadeInDistanceStart=0, fadeOutDistanceEnd=1, fadeOutDistanceStart=1, volume=0.5}
   ]
  }
 }
 {id=QualityLevel_0, children [Mixer_4, Mixer_4, WavePlayer_12, WavePlayer_12]}
 {id=QualityLevel_1, children [WavePlayer_11, Enveloper_0, Enveloper_0, Enveloper_0]}
 {id=QualityLevel_2, children [Mixer_5,Mixer_5,null,null]}
 {id=QualityLevel_3, children [Mixer_6, Mixer_6, WavePlayer_10, WavePlayer_10]}
 {
  id=WavePlayer_0
  properties {bLooping=true, soundWaveAssetPtr=/TidalCraneHoagie/Assets/Sounds/Wavs/TidalCrane_Hoagie_Engine_Close_Front_Loop_01}
 }
 {
  id=WavePlayer_1
  properties {bLooping=true, soundWaveAssetPtr=/TidalCraneHoagie/Assets/Sounds/Wavs/TidalCrane_Hoagie_Engine_Close_Rear_Loop_01}
 }
 {id=WavePlayer_10, properties {bLooping=true, soundWaveAssetPtr=/Hoagie/Audio/Assets/Waves/Veh_Hoagie_Engine_Distant_Loop_Mobile}}
 {id=WavePlayer_11, properties {soundWaveAssetPtr=&/MSS_TidalCraneHoagie_Engine_1p}}
 {id=WavePlayer_12, properties {bLooping=true, soundWaveAssetPtr=/Hoagie/Audio/Assets/Waves/Veh_Hoagie_Engine_Close_Loop_Mobile}}
 {
  id=WavePlayer_13
  properties {bLooping=true, soundWaveAssetPtr=&TidalCrane_Hoagie_Engine_Close_Rear_Loop_01}
 }
 {
  id=WavePlayer_16
  properties {bLooping=true, soundWaveAssetPtr=/TidalCraneHoagie/Assets/Sounds/Wavs/TidalCrane_Hoagie_Engine_Distant_Front_Loop_01}
 }
 {
  id=WavePlayer_17
  properties {bLooping=true, soundWaveAssetPtr=/TidalCraneHoagie/Assets/Sounds/Wavs/TidalCrane_Hoagie_Engine_Distant_Rear_Loop_01}
 }
 {id=WavePlayer_2, properties {bLooping=true, soundWaveAssetPtr=/Hoagie/Audio/Assets/Waves/Veh_Hoagie_Dmg_Ski_Scrape_Loop_01}}
 {id=WavePlayer_3, properties {bLooping=true, soundWaveAssetPtr=/Hoagie/Audio/Assets/Waves/Veh_Hoagie_Blade_Impact_Whine_Loop_01}}
 {
  id=WavePlayer_4
  properties {bLooping=true, soundWaveAssetPtr=&TidalCrane_Hoagie_Engine_Close_Front_Loop_01}
 }
]
properties {
 maxDistance=28500
 soundClassObject=&Fort_Vehicles_Exterior
 volumeMultiplier=2
 attenuationOverrides {falloffDistance=5000, attenuationShapeExtents {x=500}}
}
