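# Sound cue for the close engine pass-by of the TidalCrane Hoagie vehicle, classed as a vehicle exterior sound.
# A random node picks one of four equally weighted pass-by wave variations, played with boosted volume and spatialized through a dedicated close pass-by attenuation asset.
# The cue also sends to an engine side-chain volume audio bus, so its signal is available to drive modulation.
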SoundCue /TidalCraneHoagie/Assets/Sounds/SC_TidalCraneHoagie_Engine_PassBy_Close
root=Random_1
nodes [
 {id=Random_1, children [WavePlayer_0,WavePlayer_5,WavePlayer_6,WavePlayer_7], properties {weights [1,1,1,1]}}
 {id=WavePlayer_0, properties {soundWaveAssetPtr=/TidalCraneHoagie/Assets/Sounds/Wavs/TidalCrane_Hoagie_PassBy_Close_01}}
 {id=WavePlayer_5, properties {soundWaveAssetPtr=/TidalCraneHoagie/Assets/Sounds/Wavs/TidalCrane_Hoagie_PassBy_Close_02}}
 {id=WavePlayer_6, properties {soundWaveAssetPtr=/TidalCraneHoagie/Assets/Sounds/Wavs/TidalCrane_Hoagie_PassBy_Close_03}}
 {id=WavePlayer_7, properties {soundWaveAssetPtr=/TidalCraneHoagie/Assets/Sounds/Wavs/TidalCrane_Hoagie_PassBy_Close_04}}
]
properties {
 attenuationSettings=&/TidalCrane_Hoagie_Passby_Close_Att
 maxDistance=9000
 soundClassObject=&Fort_Vehicles_Exterior
 volumeMultiplier=1.5
 attenuationOverrides {falloffDistance=5000, attenuationShapeExtents {x=500}}
 busSends [{
   audioBus=/TidalCraneHoagie/Assets/Sounds/Modulation/AB_TidalCraneHoagie_Engine_SideChain_Vol
   minSendLevel=0
   sendLevel=1
   soundSourceBus=null
  }]
}
