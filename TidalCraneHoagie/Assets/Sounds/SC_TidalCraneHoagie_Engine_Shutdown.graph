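# Sound cue for the engine shutdown sound of the TidalCraneHoagie vehicle.
# A single wave player plays one shutdown wav at reduced volume, routed to the exterior vehicle sound class.
# Spatialization uses the shared Hoagie close engine attenuation asset, with falloff and shape extent values also recorded as overrides.
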SoundCue /TidalCraneHoagie/Assets/Sounds/SC_TidalCraneHoagie_Engine_Shutdown
root=WavePlayer_0
nodes [{id=WavePlayer_0, properties {soundWaveAssetPtr=/TidalCraneHoagie/Assets/Sounds/Wavs/TidalCrane_Hoagie_Engine_Shutdown_01}}]
properties {
 attenuationSettings=/Hoagie/Audio/Hoagie_Engine_Close_Att
 maxDistance=12000
 soundClassObject=&Fort_Vehicles_Exterior
 volumeMultiplier=0.7
 attenuationOverrides {falloffDistance=5000, attenuationShapeExtents {x=500}}
}
