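# Sound cue for the engine startup sound of the TidalCraneHoagie vehicle.
# A single startup wave plays through a Mixer that lowers its input, with a boosted cue-level volume multiplier.
# It uses the Fortnite exterior vehicle sound class and references the shared Hoagie close engine attenuation.
# It sends to an engine audio bus named for sidechain volume modulation.
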SoundCue /TidalCraneHoagie/Assets/Sounds/SC_TidalCraneHoagie_Engine_Startup
root=Mixer_0
nodes [
 {id=Mixer_0, children [WavePlayer_0], properties {inputVolume [0.5]}}
 {id=WavePlayer_0, properties {soundWaveAssetPtr=/TidalCraneHoagie/Assets/Sounds/Wavs/TidalCrane_Hoagie_Engine_Startup_01}}
]
properties {
 attenuationSettings=/Hoagie/Audio/Hoagie_Engine_Close_Att
 maxDistance=12000
 soundClassObject=&Fort_Vehicles_Exterior
 volumeMultiplier=1.5
 attenuationOverrides {falloffDistance=5000, attenuationShapeExtents {x=500}}
 busSends [{
   audioBus=/TidalCraneHoagie/Assets/Sounds/Modulation/AB_TidalCraneHoagie_Engine_SideChain_Vol
   minSendLevel=0
   sendLevel=1
   soundSourceBus=null
  }]
}
