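# Third-person engine turn sound cue for the TidalCraneHoagie vehicle, audible only when IsLocalPlayerInVehicle is false.
# A Mixer layers randomized close and distant recordings, each with its own attenuation, weighting the distant layer louder.
# The branch feeds a slight random pitch modulation, wired to only the first two quality-level inputs.
# Output uses the exterior vehicle sound class and sends to an engine sidechain volume bus.
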SoundCue /TidalCraneHoagie/Assets/Sounds/SC_TidalCraneHoagie_Engine_Turn_3P
root=QualityLevel_0
nodes [
 {id=Attenuation_0, children [Random_0], properties {attenuationSettings=&/TidalCrane_Hoagie_Turn_Close_Att}}
 {id=Attenuation_1, children [Random_1], properties {attenuationSettings=&/TidalCrane_Hoagie_Turn_Distant_Att}}
 {id=Branch_0, children [null,Mixer_0,null], properties {boolParameterName=IsLocalPlayerInVehicle}}
 {id=Mixer_0, children [Attenuation_0, Attenuation_1], properties {inputVolume [1,2]}}
 {id=Modulator_1, children [Branch_0], properties {pitchMax=1.1, pitchMin=0.9}}
 {id=QualityLevel_0, children [Modulator_1,Modulator_1,null,null]}
 {id=Random_0, children [WavePlayer_1,WavePlayer_2,WavePlayer_3], properties {weights [1,1,1]}}
 {id=Random_1, children [WavePlayer_4, WavePlayer_10, WavePlayer_11, WavePlayer_12], properties {weights [1,1,1,1]}}
 {id=WavePlayer_1, properties {soundWaveAssetPtr=/TidalCraneHoagie/Assets/Sounds/Wavs/TidalCrane_Hoagie_Turn_Close_01}}
 {id=WavePlayer_10, properties {soundWaveAssetPtr=/TidalCraneHoagie/Assets/Sounds/Wavs/TidalCrane_Hoagie_Turn_Distant_02}}
 {id=WavePlayer_11, properties {soundWaveAssetPtr=/TidalCraneHoagie/Assets/Sounds/Wavs/TidalCrane_Hoagie_Turn_Distant_03}}
 {id=WavePlayer_12, properties {soundWaveAssetPtr=/TidalCraneHoagie/Assets/Sounds/Wavs/TidalCrane_Hoagie_Turn_Distant_04}}
 {id=WavePlayer_2, properties {soundWaveAssetPtr=/TidalCraneHoagie/Assets/Sounds/Wavs/TidalCrane_Hoagie_Turn_Close_02}}
 {id=WavePlayer_3, properties {soundWaveAssetPtr=/TidalCraneHoagie/Assets/Sounds/Wavs/TidalCrane_Hoagie_Turn_Close_03}}
 {id=WavePlayer_4, properties {soundWaveAssetPtr=/TidalCraneHoagie/Assets/Sounds/Wavs/TidalCrane_Hoagie_Turn_Distant_01}}
]
properties {
 maxDistance=27500
 soundClassObject=&Fort_Vehicles_Exterior
 volumeMultiplier=1
 attenuationOverrides {falloffDistance=5000, attenuationShapeExtents {x=500}}
 busSends [{
   audioBus=/TidalCraneHoagie/Assets/Sounds/Modulation/AB_TidalCraneHoagie_Engine_SideChain_Vol
   minSendLevel=0
   sendLevel=1
   soundSourceBus=null
  }]
}
