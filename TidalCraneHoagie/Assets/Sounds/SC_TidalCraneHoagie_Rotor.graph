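# Sound cue for the Tidal Crane Hoagie's rotor wash, a looping ground-wash wave played as an attenuated exterior vehicle sound.
# An inner continuous modulator scales volume by the Amount parameter, which defaults to silent.
# An outer modulator lowers volume when IsLocalPlayerInVehicle is set.
# The cue also sends to an engine volume side-chain audio bus.
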SoundCue /TidalCraneHoagie/Assets/Sounds/SC_TidalCraneHoagie_Rotor
root=ModulatorContinuous_2
nodes [
 {
  id=ModulatorContinuous_0
  children [WavePlayer_2]
  properties {pitchModulationParams {minOutput=1, parameterName=Amount}, volumeModulationParams {default=0, parameterName=Amount}}
 }
 {
  id=ModulatorContinuous_2
  children [ModulatorContinuous_0]
  properties {volumeModulationParams {default=0, maxOutput=0.3, minOutput=1, parameterName=IsLocalPlayerInVehicle}}
 }
 {id=WavePlayer_2, properties {bLooping=true, soundWaveAssetPtr=/Hoagie/Audio/Assets/Waves/Veh_Hoagie_Rotor_Wash_Ground_Loop_01}}
]
properties {
 attenuationSettings=&/TidalCrane_Hoagie_RotorWash_Att
 maxDistance=5000
 soundClassObject=&Fort_Vehicles_Exterior
 volumeMultiplier=1
 attenuationOverrides {falloffDistance=5000, attenuationShapeExtents {x=500}}
 busSends [{
   audioBus=/TidalCraneHoagie/Assets/Sounds/Modulation/AB_TidalCraneHoagie_Engine_SideChain_Vol
   minSendLevel=0
   sendLevel=2
   soundSourceBus=null
  }]
}
