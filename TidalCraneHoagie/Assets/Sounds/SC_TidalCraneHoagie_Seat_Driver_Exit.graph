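# Sound cue for a driver exiting the seat of the TidalCraneHoagie vehicle.
# A single wave player, reusing the base Hoagie seat-exit wave, feeds a Modulator node left at default settings.
# The cue routes to the vehicle interior sound class with boosted volume and borrows the golf cart close-movement attenuation.
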SoundCue /TidalCraneHoagie/Assets/Sounds/SC_TidalCraneHoagie_Seat_Driver_Exit
root=Modulator_0
nodes [
 {id=Modulator_0, children [WavePlayer_0]}
 {id=WavePlayer_0, properties {soundWaveAssetPtr=/Hoagie/Audio/Assets/Waves/Veh_Hoagie_Seat_Exit_01}}
]
properties {
 attenuationSettings=/Game/Sounds/Attenuation/Fort_GolfCart_Movement_Close_Att
 maxDistance=5300
 soundClassObject=&Fort_Vehicles_Interior
 volumeMultiplier=1.5
}
