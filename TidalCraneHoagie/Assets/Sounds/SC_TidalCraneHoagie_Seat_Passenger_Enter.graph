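# Sound cue for a passenger entering a seat in the TidalCraneHoagie vehicle.
# A Random node picks between two equally weighted seat-switch waves from the base Hoagie audio, then a default Modulator feeds the output.
# It plays through the vehicle interior sound class with raised volume and reuses a close-range golf cart movement attenuation.
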SoundCue /TidalCraneHoagie/Assets/Sounds/SC_TidalCraneHoagie_Seat_Passenger_Enter
root=Modulator_0
nodes [
 {id=Modulator_0, children [Random_0]}
 {id=Random_0, children [WavePlayer_0,WavePlayer_3], properties {weights [1,1]}}
 {id=WavePlayer_0, properties {soundWaveAssetPtr=/Hoagie/Audio/Assets/Waves/Veh_Hoagie_Seat_Switch_01}}
 {id=WavePlayer_3, properties {soundWaveAssetPtr=/Hoagie/Audio/Assets/Waves/Veh_Hoagie_Seat_Switch_02}}
]
properties {
 attenuationSettings=/Game/Sounds/Attenuation/Fort_GolfCart_Movement_Close_Att
 maxDistance=5300
 soundClassObject=&Fort_Vehicles_Interior
 volumeMultiplier=1.5
}
