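# Sound cue for the Hoagie vehicle's out-of-fuel warning in the TidalCrane Hoagie content.
# A single wave player plays the fuel-empty warning wave in the vehicle interior sound class with vehicle prop attenuation.
# Overridden concurrency allows one instance per owner and blocks new ones while it plays.
# It also sends to an engine side-chain volume audio bus.
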SoundCue /TidalCraneHoagie/Assets/Sounds/SC_TidalCraneVeh_Hoagie_OutOfFuel
root=WavePlayer_1
nodes [{id=WavePlayer_1, properties {soundWaveAssetPtr=/TidalCraneHoagie/Assets/Sounds/Wavs/TidalCrane_Hoagie_FuelEmpty_Warning_01}}]
properties {
 attenuationSettings=/Game/Sounds/Attenuation/Fort_Prop_Vehicles_Att
 bOverrideConcurrency=true
 maxDistance=3000
 soundClassObject=&Fort_Vehicles_Interior
 volumeMultiplier=1.5
 busSends [{
   audioBus=/TidalCraneHoagie/Assets/Sounds/Modulation/AB_TidalCraneHoagie_Engine_SideChain_Vol
   minSendLevel=0
   sendLevel=1
   soundSourceBus=null
  }]
 concurrencyOverrides {bLimitToOwner=true, maxCount=1, resolutionRule=PreventNew}
}
