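# Sound cue for the low-fuel warning on the TidalCrane Hoagie vehicle.
# A single wave player plays the warning sample through the vehicle sound class and a super-close player attenuation asset, with boosted volume.
# It sends at full level to an engine side-chain volume audio bus, presumably to duck engine audio.
# A concurrency override blocks new instances at the limit rather than cutting off playing ones.
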SoundCue /TidalCraneHoagie/Assets/Sounds/SC_TidalCraneVeh_Hoagie_Warning_LowFuel
root=WavePlayer_0
nodes [{id=WavePlayer_0, properties {soundWaveAssetPtr=/TidalCraneHoagie/Assets/Sounds/Wavs/TidalCrane_Hoagie_FuelLow_Warning_01}}]
properties {
 attenuationSettings=/Game/Sounds/Attenuation/Fort_Player_SuperClose_Att
 bOverrideConcurrency=true
 maxDistance=800
 soundClassObject=&Fort_Vehicles
 volumeMultiplier=1.5
 attenuationOverrides {distanceAlgorithm=NaturalSound, falloffDistance=4000, attenuationShapeExtents {x=1000}}
 busSends [{
   audioBus=/TidalCraneHoagie/Assets/Sounds/Modulation/AB_TidalCraneHoagie_Engine_SideChain_Vol
   minSendLevel=0
   sendLevel=1
   soundSourceBus=null
  }]
 concurrencyOverrides {resolutionRule=PreventNew}
}
