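# Sound cue for the periodic low-fuel warning reminder on the TidalCrane Hoagie vehicle.
# A single wave player feeds the reminder wave straight to the output with a boosted volume, under the vehicles sound class.
# It uses the shared super-close player attenuation asset.
# Its own concurrency override blocks new instances once the limit is reached, and the repeat timing is not defined in the cue.
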SoundCue /TidalCraneHoagie/Assets/Sounds/SC_TidalCraneVeh_Hoagie_Warning_LowFuel_Periodic
root=WavePlayer_0
nodes [{id=WavePlayer_0, properties {soundWaveAssetPtr=/TidalCraneHoagie/Assets/Sounds/Wavs/TidalCrane_Hoagie_FuelLow_Warning_Remind}}]
properties {
 attenuationSettings=/Game/Sounds/Attenuation/Fort_Player_SuperClose_Att
 bOverrideConcurrency=true
 maxDistance=800
 soundClassObject=&Fort_Vehicles
 volumeMultiplier=1.5
 attenuationOverrides {distanceAlgorithm=NaturalSound, falloffDistance=4000, attenuationShapeExtents {x=1000}}
 concurrencyOverrides {resolutionRule=PreventNew}
}
