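# Sound cue for switching on music in the TidalCrane Hoagie vehicle, judging by its name and its vehicle interior sound class.
# A single wave player plays the music switch-on wave with no randomization or modulation.
# It routes through the Fort vehicle interior sound class with a slight volume boost and shares a music toggle concurrency setting that governs overlapping plays.
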SoundCue /TidalCraneHoagie/Assets/Sounds/SC_TidalCrane_Hoagie_Music_On
root=WavePlayer_0
nodes [{id=WavePlayer_0, properties {soundWaveAssetPtr=/TidalCraneHoagie/Assets/Sounds/Wavs/TidalCrane_Hoagie_Music_Switch_On_01}}]
properties {
 maxDistance=2097152
 soundClassObject=&Fort_Vehicles_Interior
 volumeMultiplier=1.059
 concurrencySet [/TidalCraneHoagie/Assets/Sounds/Settings/CONC_TidalCrane_Hoagie_Music_Toggle]
}
